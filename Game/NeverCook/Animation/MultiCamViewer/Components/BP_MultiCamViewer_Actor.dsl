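; Multi-camera character viewer actor for animation review, with front, back, side, three-quarter, face and game scene-capture views on a Fortnite pawn.
; Construction picks the game camera's render target and previews on a dummy body, which BeginPlay destroys.
; A looping timeline finds the controller and pawn, head-attaches the face cam, follows the player, polls numpad keys and forces time of day by cheat command.
; Custom events attach by pawn index and set FOV, arm length and height.
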(event Custom|AttachToCharacter
  (params (CharacteIndex :type Integer))
  (Utilities|FlowControl|Sequence
    (:then_0
      (bind _outactors (Actor|GetAllActorsOfClass "/Script/FortniteGame.FortPlayerPawnAthena"))
      (bind _output_get (Variables|Default|SetAllAthenaPawns _outactors))
      (bind _output_get_2 (Variables|Default|SetActivePawn (Utilities|Array|Get(acopy) _output_get CharacteIndex)))
      (Utilities|IsValid _output_get_2
        (:Is%20Valid
          (Transformation|AttachActorToActor self _output_get_2 "None" "SnapToTarget" "SnapToTarget" "KeepRelative" true))))))

(event Custom|DetachFromCharacter
  (Transformation|DetachFromActor self "KeepWorld" "KeepWorld" "KeepWorld"))

(event Custom|Set%20FOV%20and%20distance
  (params (FOV :type Float%20%28single-precision%29) (Distance :type Float%20%28single-precision%29) (Height :type Float%20%28double-precision%29))
  (Utilities|FlowControl|Sequence
    (:then_0
      (bind _output_get (Variables|Default|SetField%20Of%20View FOV))
      (bind _output_get_2 (Variables|/Script/Engine.SceneCaptureComponent2D|SetFOVAngle _output_get (Variables|Default|GetsceneBack)))
      (bind _output_get_3 (Variables|/Script/Engine.SceneCaptureComponent2D|SetFOVAngle _output_get_2 (Variables|Default|Getscene3/4L)))
      (bind _output_get_4 (Variables|/Script/Engine.SceneCaptureComponent2D|SetFOVAngle _output_get_3 (Variables|Default|GetsceneRight)))
      (bind _output_get_5 (Variables|/Script/Engine.SceneCaptureComponent2D|SetFOVAngle _output_get_4 (Variables|Default|GetsceneFrontPlayer)))
      (bind _output_get_6 (Variables|/Script/Engine.SceneCaptureComponent2D|SetFOVAngle _output_get_5 (Variables|Default|Getscene3/4R)))
      (Variables|/Script/Engine.SceneCaptureComponent2D|SetFOVAngle _output_get_6 (Variables|Default|GetsceneLeft)))
    (:then_1
      (bind _output_get_7 (Variables|Default|SetDistance Distance))
      (bind _output_get_8 (Variables|/Script/Engine.SpringArmComponent|SetTargetArmLength _output_get_7 (Variables|Default|GetSpringArm_Front)))
      (bind _output_get_9 (Variables|/Script/Engine.SpringArmComponent|SetTargetArmLength _output_get_8 (Variables|Default|GetSpringArm_Left)))
      (bind _output_get_10 (Variables|/Script/Engine.SpringArmComponent|SetTargetArmLength _output_get_9 (Variables|Default|GetSpringArm_Right)))
      (bind _output_get_11 (Variables|/Script/Engine.SpringArmComponent|SetTargetArmLength _output_get_10 (Variables|Default|GetSpringArm_3/4L)))
      (bind _output_get_12 (Variables|/Script/Engine.SpringArmComponent|SetTargetArmLength _output_get_11 (Variables|Default|GetSpringArm_Back)))
      (Variables|/Script/Engine.SpringArmComponent|SetTargetArmLength _output_get_12 (Variables|Default|GetSpringArm_3/4R)))
    (:then_2
      (bind _output_get_13 (Variables|Default|SetHeight Height))
      (Transformation|SetRelativeLocation (Variables|Default|GetCameraOffsets) (Math|Vector|MakeVector 0.0 0.0 _output_get_13) false false))))

(event Custom|SetIsFar
  (params (Index :type Integer))
  (Variables|Default|SetDistance (Utilities|Select :Index Index :Option%200 2.6e+02 :Option%201 305.0 :Option%202 3.5e+02 :Option%203 5e+02))
  (bind _output_get (Variables|/Script/Engine.SpringArmComponent|SetTargetArmLength (stale Custom|Set%20FOV%20and%20distance _output_get_7) (Variables|Default|GetSpringArm_Front)))
  (bind _output_get_2 (Variables|/Script/Engine.SpringArmComponent|SetTargetArmLength _output_get (Variables|Default|GetSpringArm_Left)))
  (bind _output_get_3 (Variables|/Script/Engine.SpringArmComponent|SetTargetArmLength _output_get_2 (Variables|Default|GetSpringArm_Right)))
  (bind _output_get_4 (Variables|/Script/Engine.SpringArmComponent|SetTargetArmLength _output_get_3 (Variables|Default|GetSpringArm_3/4L)))
  (bind _output_get_5 (Variables|/Script/Engine.SpringArmComponent|SetTargetArmLength _output_get_4 (Variables|Default|GetSpringArm_Back)))
  (Variables|/Script/Engine.SpringArmComponent|SetTargetArmLength _output_get_5 (Variables|Default|GetSpringArm_3/4R)))

(event EventBeginPlay
  ;[1 1 1] on start, remove the preview setup.
  (Components|DestroyComponent (Variables|Default|GetcharBody))
  (Transformation|SetRelativeLocation (Variables|Default|GetcameraParent) "0, 0, 0" false false)
  (Transformation|SetRelativeLocation (Variables|Default|GetOffsetGameCam) (-5e+02 0.0 72.0) false false)
  (Timeline|tick Play
    (:Update
      ;[0.4844 0.4844 0.4844] on tick, set player cam to player location, update widget information
      ;;[1 1 1] watch for button presses.
      (CallFunction|ButtonPresses)
      (Utilities|FlowControl|Sequence
        (:then_0
          ;;[1 1 1] if PC or Actor are invalid, set these variables first. Safety net.
          (bind _playercontroller (Variables|Default|GetplayerController)
            (:then
              (bind _actor (Variables|Default|Getactor)
                (:then
                  ;;[1 1 1] if is NOT facecam, snap the camera rig to the characters root.
                  (if (not (Variables|Properties|GetisFaceCam))
                    (Transformation|SetWorldTransform (Variables|Default|GetScene) (Math|Transform|MakeTransform (Transformation|GetActorLocation _actor) (Pawn|GetControlRotation _playercontroller :!class /Script/Engine.Controller)) false false)))
                (:else
                  (CallFunction|SetCamera))))
            (:else
              (CallFunction|SetVariables))))
        (:then_1
          (Development|ExecuteConsoleCommand (Utilities|String|Append "Cheat SetTimeOfDay " (Utilities|String|ToString(Integer) (Variables|Properties|GettimeOfDay)))))))))

(event Custom|toggleFacecam)

(event Custom|setVariables
  ;[0.6042 0.6042 0.6042] on start, establish variables then set self location based on settings
  ;;[1 1 1] set the player controller variable
  (bind _returnvalue (Actor|GetActorOfClass "/Game/Athena/Athena_PlayerController.Athena_PlayerController_C"))
  (Utilities|IsValid _returnvalue
    (:Is%20Valid
      (Variables|Default|SetplayerController _returnvalue))))

(event Custom|increaseCamDistance
  (bind _cameradistanceoverride (Variables|Properties|GetcameraDistanceOverride))
  ;[0.6042 0.6042 0.6042] on start, establish variables then set self location based on settings
  ;;[1 0 0 0.5] Legcy
  (Variables|Properties|SetcameraDistanceOverride 15.0)
  (Utilities|FlowControl|Sequence
    (:then_0
      (CallFunction|SetCamDistance (Variables|Default|GetOffsetFront) _cameradistanceoverride))
    (:then_1
      (CallFunction|SetCamDistance (Variables|Default|GetOffset3/4R) _cameradistanceoverride))
    (:then_2
      (CallFunction|SetCamDistance (Variables|Default|GetOffset3/4L) _cameradistanceoverride))))

(event Custom|setCamera
  (bind _self self)
  (bind _facecamparent (Variables|Default|GetfaceCamParent))
  (bind _cameradistanceoverride (Variables|Properties|GetcameraDistanceOverride))
  ;[0.6042 0.6042 0.6042] on start, establish variables then set self location based on settings
  ;;[1 1 1] assign Actor variable, and ignore the rest of graph if we ONLY want gamecamera.
  (Utilities|IsValid (Pawn|GetControlledPawn (stale Custom|setVariables _returnvalue))
    (:Is%20Valid
      (bind _output_get (Variables|Default|Setactor (Pawn|GetControlledPawn (stale Custom|setVariables _returnvalue))))
      (Utilities|IsValid _output_get
        (:Is%20Valid
          (if (Variables|Properties|GetcharacterOnly)
            _
            (elif (Variables|Properties|GetsnapToPlayer)
              ;;[1 1 1] if we want the cameras to stick to the character in world space or not
              (Transformation|AttachActorToActor _self (Pawn|GetControlledPawn (stale Custom|setVariables _returnvalue)) "None" "SnapToTarget" "SnapToTarget" "KeepRelative" true)
              (else
                (Transformation|SetActorTransform _self (Transformation|GetActorTransform (Pawn|GetControlledPawn (stale Custom|setVariables _returnvalue))) false true))))
          ;;[1 1 1] adjust the camera position if facecam on start
          (if (Variables|Properties|GetisFaceCam)
            (CallFunction|ToggleFacecam))
          ;;[1 1 1] adjust position and parent of the face camera
          (Transformation|AttachComponentToComponent _facecamparent (Actor|GetComponentbyClass (Variables|Default|Getactor) "/Script/Engine.SkeletalMeshComponent") "head" "SnapToTarget" "SnapToTarget" "SnapToTarget" true)
          (Transformation|SetRelativeRotation _facecamparent "180.000000,90.000000,-90.000000" false false)
          (Transformation|SetRelativeLocation _facecamparent "5.000000,100.000000,0.000000" false false)
          ;;[1 0 0 0.5] Legcy
          (Utilities|FlowControl|Sequence
            (:then_0
              (CallFunction|SetCamDistance (Variables|Default|GetOffsetFront) _cameradistanceoverride))
            (:then_1
              (CallFunction|SetCamDistance (Variables|Default|GetOffset3/4R) _cameradistanceoverride))
            (:then_2
              (CallFunction|SetCamDistance (Variables|Default|GetOffset3/4L) _cameradistanceoverride))))))))

(event Custom|decreaseCamDistance
  (bind _cameradistanceoverride (Variables|Properties|GetcameraDistanceOverride))
  ;[0.6042 0.6042 0.6042] on start, establish variables then set self location based on settings
  ;;[1 0 0 0.5] Legcy
  (Variables|Properties|SetcameraDistanceOverride -15.0)
  (Utilities|FlowControl|Sequence
    (:then_0
      (CallFunction|SetCamDistance (Variables|Default|GetOffsetFront) _cameradistanceoverride))
    (:then_1
      (CallFunction|SetCamDistance (Variables|Default|GetOffset3/4R) _cameradistanceoverride))
    (:then_2
      (CallFunction|SetCamDistance (Variables|Default|GetOffset3/4L) _cameradistanceoverride))))

(event Custom|buttonPresses
  ;[1 1 1] overwrite current skin with selected from array
  (Variables|Default|GetplayerController
    (:then
      (Utilities|FlowControl|Sequence
        (:then_11
          ;[1 1 1] Comment
          (if (not (Game|Player|WasInputKeyJustPressed (Variables|Default|GetplayerController) "Multiply"))
            (goto _label Reset))
          (Utilities|FlowControl|DoOnce false :!label _label))
        (:then_12
          (if (not (Game|Player|WasInputKeyJustPressed (Variables|Default|GetplayerController) "Add"))
            (goto _label_2 Reset))
          (Utilities|FlowControl|DoOnce false :!label _label_2
            (:Completed
              (CallFunction|IncreaseCamDistance))))
        (:then_13
          (if (not (Game|Player|WasInputKeyJustPressed (Variables|Default|GetplayerController) "Subtract"))
            (goto _label_3 Reset))
          (Utilities|FlowControl|DoOnce false :!label _label_3
            (:Completed
              (CallFunction|DecreaseCamDistance))))
        (:then_16
          ; Press Num / to remove face accessories (hat, sunglasses, headphones but also hair and beard)
          (if (not (Game|Player|WasInputKeyJustPressed (Variables|Default|GetplayerController) "Divide"))
            (goto _label_4 Reset))
          (Utilities|FlowControl|DoOnce false :!label _label_4
            (:Completed
              (Development|ExecuteConsoleCommand "TogglePawnCharacterPartVisibility Face"))))))
    (:else)))

(fn setCamDistance
  (params (camera :type Scene%20Component%20Object%20Reference) (offset :type Float%20%28double-precision%29))
  (Transformation|AddLocalOffset camera (Math|Vector|MakeVector offset) false false))

(fn ConstructionScript
  (bind _facecamparent (Variables|Default|GetfaceCamParent))
  ;[1 1 1] on construction, remove the bottom right 3/4R camera if desired.
  (Variables|/Script/Engine.SceneCaptureComponent2D|SetTextureTarget (select (Variables|Properties|Getadditional3/4) "/Game/NeverCook/Animation/MultiCamViewer/Components/RT_MultiCamViewer_gameCamHalf.RT_MultiCamViewer_gameCamHalf" "/Game/NeverCook/Animation/MultiCamViewer/Components/RT_MultiCamViewer_Game.RT_MultiCamViewer_Game") (Variables|Default|GetgameCam))
  ;[1 1 1] set location of camera parent for preview
  (Transformation|SetRelativeLocation (Variables|Default|GetcameraParent) "0.000000,0.000000,60.000000" false false)
  ;[1 1 1] attach facecam to dummy for preview setup.
  (Transformation|AttachComponentToComponent _facecamparent (Variables|Default|GetcharBody) "head" "SnapToTarget" "SnapToTarget" "SnapToTarget" true)
  (Transformation|SetRelativeRotation _facecamparent "180.000000,90.000000,-90.000000" false false)
  (Transformation|SetRelativeLocation _facecamparent "4.000000,120.000000,0.000000" false false))
