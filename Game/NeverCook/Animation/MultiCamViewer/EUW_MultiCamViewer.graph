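# Editor utility widget for previewing PIE characters, emotes and cosmetics across multiple camera views.
# It spawns or reuses a BP_MultiCamViewer_Actor attached to the chosen player, feeding render targets through dynamic materials into configurable screens and face cam.
# Saved ID button lists cover emotes, characters, Sparks and pickaxes; emotes can auto-fill by release season.
# Cosmetics apply through Python or console commands, and editable text overlays and layout settings persist via SaveGame.
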EditorUtilityWidgetBlueprint /Game/NeverCook/Animation/MultiCamViewer/EUW_MultiCamViewer
graphs [
 Box_CIDs_Visibility {nodes [
   Box_CIDs_Visibility() {is=title K2Node_FunctionEntry}
   > Return Node!(execute, ReturnValue:ESlateVisibility Enum=Select(Option 0=Collapsed, Option 1=Visible, Index=Get ShowCIDs() -> (ShowCIDs:bool)))
  ]}
 Box_EIDs_Visibility {nodes [
   Box_EIDs_Visibility() {is=title K2Node_FunctionEntry}
   > Return Node(ReturnValue=Select(Option 0=Collapsed, Option 1=Visible, Index=Get ShowEIDs() -> (ShowEIDs:bool)))
  ]}
 Box_PickaxeID_Visibility {nodes [
   Box_PickaxeID_Visibility() {is=title K2Node_FunctionEntry}
   > Return Node(ReturnValue=Select(Option 0=Collapsed, Option 1=Visible, Index=Get ShowPickaxes() -> (ShowPickaxes:bool)))
  ]}
 Box_SparksID_Visibility {nodes [
   Box_SparksID_Visibility() {is=title K2Node_FunctionEntry}
   > Return Node(ReturnValue=Select(Option 0=Collapsed, Option 1=Visible, Index=Get ShowSparksID() -> (ShowSparksID:bool)))
  ]}
 Box_TextOverlays_Visibility {nodes [
   Box_TextOverlays_Visibility() {is=title K2Node_FunctionEntry}
   > Return Node(ReturnValue=Select(Option 0=Collapsed, Option 1=Visible, Index=Get EditTextOverlays() -> (EditTextOverlays:bool)))
  ]}
 EventGraph {nodes [
   Clear(TargetArray=Get EIDs() -> (EIDs:Array of Strings))
   > SaveGame_Save(self:TA Base Editor Widget Object Reference)
   n1 = "On Clicked (EditorUtilityButton_73)"() {is=title K2Node_ComponentBoundEvent}
   Sequence(execute=[@n1.then, @n49.then])
   > n2 = GetPIEWorlds!(execute, bIncludeDedicatedServer:bool=false) -> (then, ReturnValue:Array of World Object References)
   n3 = "Get (a copy)"(Array:Array of World Object References=@n2) -> (Output:World Object Reference)
   > n4 = GetActorOfClass!(
    execute
    WorldContextObject:Object Reference=@n3
    ActorClass:Actor Class Reference=/Game/NeverCook/Animation/MultiCamViewer/Components/BP_MultiCamViewer_Actor.BP_MultiCamViewer_Actor_C
   ) -> (then, ReturnValue:BP Multi Cam Viewer Actor Object Reference)
   > n5 = Is Valid(InputObject=@n4)
   n6 = SpawnActor BP Multi Cam Viewer Actor!(
    execute:Exec=@n5.Is Not Valid
    WorldContextObject:Object Reference=@n3
    Class:Actor Class Reference=&BP_MultiCamViewer_Actor_C
    SpawnTransform_Location:Vector="0, 0, 0"
    SpawnTransform_Rotation:Rotator="0, 0, 0"
    SpawnTransform_Scale:Vector="1,1,1"
    CollisionHandlingOverride:ESpawnActorCollisionHandlingMethod Enum=Undefined
    TransformScaleMethod:ESpawnActorScaleMethod Enum=MultiplyWithRoot
    Owner:Actor Object Reference
    characterOnly:bool=False
    Instigator:Pawn Object Reference
   ) -> (then:Exec, ReturnValue:BP Multi Cam Viewer Actor Object Reference) {is=title K2Node_SpawnActorFromClass}
   > n7 = Set CameraActor(CameraActor:BP Multi Cam Viewer Actor Object Reference=@n6) -> (Output_Get:BP Multi Cam Viewer Actor Object Reference)
   AttachToCharacter!(
    execute=[@n8.then, @n7.then]
    self:BP Multi Cam Viewer Actor Object Reference=@n7
    CharacteIndex:int=Get Player Index() -> (Player Index:int=0)
   ) -> (then)
   > SetIsFar!(
    execute
    self:BP Multi Cam Viewer Actor Object Reference=Get CameraActor() -> (CameraActor:BP Multi Cam Viewer Actor Object Reference)
    Index:int=Get SetDistance() -> (SetDistance:int=0)
   ) -> (then)
   n8 = Set CameraActor(execute=@n5.Is Valid, CameraActor=@n4)
   "On Clicked (Button_StopEID_1)"() {is=title K2Node_ComponentBoundEvent}
   > RefreshCameras()
   > DetachFromCharacter(self:BP Multi Cam Viewer Actor Object Reference=Get CameraActor())
   "On Clicked (Button_AutoWalk)"() {is=title K2Node_ComponentBoundEvent}
   > n9 = Set AutoTravel(AutoTravel:bool=NOT Boolean(A=Get AutoTravel() -> (AutoTravel:bool))) -> (Output_Get:bool)
   > SetBrushTintColor!(
    execute
    self:Image Object Reference=Get Image_52() -> (Image_52:Image Object Reference)
    TintColor_SpecifiedColor:Linear Color Structure=Select(Option 0:Linear Color Structure=(R=1,G=1,B=1,A=1), Option 1:Linear Color Structure=(R=1,G=0.8495,B=0.0192,A=1), Index=@n9) -> (ReturnValue:Linear Color Structure)
    TintColor_ColorUseRule:ESlateColorStylingMode Enum
   ) -> (then)
   > SetVisibility(
    self=Get EditorUtilityCircularThrobber_26() -> (EditorUtilityCircularThrobber_26:Editor Utility Circular Throbber Object Reference)
    InVisibility=Select(Option 0:ESlateVisibility Enum=Collapsed, Option 1:ESlateVisibility Enum=Visible, Index=@n9) -> (ReturnValue:ESlateVisibility Enum)
   )
   "On Clicked (Button_StopEID)"() {is=title K2Node_ComponentBoundEvent}
   > n10 = RefreshCameraCheck!(execute, self) -> (then, playerController:Athena Player Controller Object Reference)
   n11 = GetPlayerPawn!(self:Fort Player Controller Object Reference=@n10) -> (ReturnValue:Fort Player Pawn Object Reference)
   > Montage_Stop!(
    execute
    self:Anim Instance Object Reference=GetAnimInstance(self=Get Mesh(self=@n11))
    InBlendOutTime:float=0.2
    Montage:Anim Montage Object Reference
   ) -> (then)
   "On Clicked (Button_AddEID)"() {is=title K2Node_ComponentBoundEvent}
   > Add ID!(execute, self, CosmeticType:int=0) -> (then)
   "On Clicked (EditorUtilityButton_1)"() {is=title K2Node_ComponentBoundEvent}
   > Clear(TargetArray=Get EIDs())
   > n12 = GetAssets!(
    execute
    self:Asset Registry Interface=GetAssetRegistry()
    Filter_PackageNames:Array of Names
    Filter_PackagePaths:Array of Names
    Filter_SoftObjectPaths:Array of Soft Object Path Structures
    Filter_ClassPaths:Array of Top Level Asset Path Structures=Make Array("[0]":Top Level Asset Path Structure="Get Class Path Name (Top Level Asset Path)"!(Class:Object Class Reference=/Script/FortniteGame.AthenaDanceItemDefinition) -> (ReturnValue:Top Level Asset Path Structure) {is=title K2Node_CallFunction}) -> (Array:Array of Top Level Asset Path Structures)
    Filter_RecursiveClassPathsExclusionSet:Set of Top Level Asset Path Structures
    Filter_ClassNames:Array of Names
    Filter_RecursiveClassesExclusionSet:Set of Names
    Filter_bRecursivePaths:bool=false
    Filter_bRecursiveClasses:bool=false
    Filter_bIncludeOnlyOnDiskAssets:bool=false
    bSkipARFilteredAssets:bool=true
   ) -> (then, OutAssetData:Array of Asset Data Structures, ReturnValue:bool)
   > n13 = For Each Loop(Array:Array of Asset Data Structures=@n12) -> (Array Element:Asset Data Structure)
   SaveGame_Save(execute=@n13.Completed)
   > Generate IDs!(execute, self, CosmeticType:int=0) -> (then)
   n14 = GetAsset(InAssetData=@n13)
   n15 = EditorOnly_GetItemDefinitionFortReleaseVersion!(
    execute=@n13.LoopBody
    ItemDefinition:Item Definition Base Object Reference=Cast To AthenaDanceItemDefinition!(Object=@n14) -> (AsAthena Dance Item Definition:Athena Dance Item Definition Object Reference, bSuccess:bool=~) {is=title K2Node_DynamicCast}
   ) -> (then, ReturnValue:Fort Release Version Structure)
   > n16 = EditorOnly_DoVersionSeasonsMatch!(
    execute
    VersionA:"Fort Release Version Structure (by ref)"=@n15
    VersionB:"Fort Release Version Structure (by ref)"=Get EID Release() -> (EID Release:Fort Release Version Structure)
   ) -> (then, ReturnValue:bool)
   > Branch(Condition=@n16)
   > Add(TargetArray=Get EIDs(), NewItem=Get Object Path String(Object=@n14)) {is=title K2Node_CallArrayFunction}
   n17 = "On Check State Changed (Checkbox_ShowEIDs)"!() -> (OutputDelegate, then, bIsChecked:bool) {is=title K2Node_ComponentBoundEvent}
   > Set ShowEIDs(ShowEIDs:bool=@n17.bIsChecked) -> (Output_Get:bool)
   > SaveGame_Save()
   "On Clicked (EditorUtilityButton_79)"() {is=title K2Node_ComponentBoundEvent}
   > n18 = Add(TargetArray=Get TextsText() -> (TextsText:Array of Strings), NewItem=New Text)
   > Set Array Elem!(
    execute
    TargetArray:"Array of Float (double-precision)s"=Get TextsSize() -> (TextsSize:"Array of Float (double-precision)s")
    Index:int=@n18
    Item:"Float (double-precision) (by ref)"=50
    bSizeToFit:bool=true
   ) -> (then)
   > Set Array Elem(
    TargetArray:Array of Vector 2D Structures=Get TextsPosition() -> (TextsPosition:Array of Vector 2D Structures)
    Index=@n18
    Item:"Vector 2D Structure (by ref)"
    bSizeToFit=true
   )
   > Set Array Elem(
    TargetArray:Array of Linear Color Structures=Get TextsColor() -> (TextsColor:Array of Linear Color Structures)
    Index=@n18
    Item:"Linear Color Structure (by ref)"=(R=0,G=0,B=0,A=1)
    bSizeToFit=true
   )
   > Set Array Elem(
    TargetArray:Array of Booleans=Get TextVisibility() -> (TextVisibility:Array of Booleans)
    Index=@n18
    Item:"Boolean (by ref)"=true
    bSizeToFit=true
   )
   > Set Array Elem(
    TargetArray:Array of Integers=Get TextOutlineSize() -> (TextOutlineSize:Array of Integers)
    Index=@n18
    Item:"Integer (by ref)"=2
    bSizeToFit=true
   )
   > Set Array Elem(
    TargetArray:Array of Linear Color Structures=Get TextOutlineColor() -> (TextOutlineColor:Array of Linear Color Structures)
    Index=@n18
    Item:"Linear Color Structure (by ref)"=(R=1,G=1,B=1,A=1)
    bSizeToFit=true
   )
   > SaveGame_Save()
   > Generate text overlays()
   "On Clicked (EditorUtilityButton)"() {is=title K2Node_ComponentBoundEvent}
   > ClearChildren(self=[
     Get VerticalBox_textoverlayproperty() -> (VerticalBox_textoverlayproperty:Vertical Box Object Reference)
     Get CanvasPanel_TextOverlays() -> (CanvasPanel_TextOverlays:Canvas Panel Object Reference)
    ])
   > Clear(TargetArray=Get TextsText())
   > Clear(TargetArray:"Array of Float (double-precision)s"=Get TextsSize())
   > Clear(TargetArray:Array of Linear Color Structures=Get TextsColor())
   > Clear(TargetArray:Array of Booleans=Get TextVisibility())
   > Clear(TargetArray:Array of Vector 2D Structures=Get TextsPosition())
   > SaveGame_Save()
   n19 = "On Check State Changed (Checkbox_ShowSparks)"!() -> (OutputDelegate, then, bIsChecked:bool) {is=title K2Node_ComponentBoundEvent}
   > Set ShowSparksID(ShowSparksID:bool=@n19.bIsChecked) -> (Output_Get:bool)
   > SaveGame_Save()
   "On Clicked (Button_AddPickaxes)"() {is=title K2Node_ComponentBoundEvent}
   > Add ID(CosmeticType=3)
   n20 = "On Property Changed (PropertyView_Cam_FOV)"!() -> (OutputDelegate, then, PropertyName:Name=None) {is=title K2Node_ComponentBoundEvent}
   RefreshCameraCheck(execute=[@n20.then, @n33.then, @n38.then])
   > Set FOV and distance!(
    execute
    self:BP Multi Cam Viewer Actor Object Reference=Get CameraActor()
    FOV:float=Get Cameras_FOV() -> (Cameras_FOV:double)
    Distance:float=Get Cameras_Distance() -> (Cameras_Distance:double)
    Height:double=Get Cameras_Height() -> (Cameras_Height:double)
   ) -> (then) {is=title K2Node_CallFunction}
   "On Clicked (Button_AddEID_1)"() {is=title K2Node_ComponentBoundEvent}
   > Add ID(CosmeticType=1)
   "On Clicked (EditorUtilityButton_2)"() {is=title K2Node_ComponentBoundEvent}
   > RemoveAllIDs!(execute, self, CosmeticType:int=1) -> (then)
   "On Property Changed (PropertyView_Montage_Visible)"!() -> (OutputDelegate, then, PropertyName:Name=None) {is=title K2Node_ComponentBoundEvent}
   n21 = Get MontageDetails_Visibility() -> (MontageDetails_Visibility:bool)
   > Set MontageDetails_Visibility(MontageDetails_Visibility:bool=NOT Boolean(A=@n21)) -> (Output_Get:bool)
   > n22 = SaveGame_Save()
   SetVisibility(
    execute=[@n22.then, @n100.then]
    self=[Get VerticalBox_1() -> (VerticalBox_1:Vertical Box Object Reference), Get VerticalBox_0() -> (VerticalBox_0:Vertical Box Object Reference)]
    InVisibility=Select(Option 0=Collapsed, Option 1=HitTestInvisible, Index=@n21)
   )
   "On Selection Changed (ComboBox_FaceCam)"!() -> (OutputDelegate, then, SelectedItem:String, SelectionType:ESelectInfo Enum) {is=title K2Node_ComponentBoundEvent}
   > Set FaceFormatIndex(FaceFormatIndex:int=GetSelectedIndex!(self:"Combo Box (String) Object Reference"=Get ComboBox_FaceCam() -> (ComboBox_FaceCam:Editor Utility Combo Box String Object Reference)) -> (ReturnValue:int=0)) -> (Output_Get:int=0)
   > n23 = SaveGame_Save()
   n24 = Get FaceFormatIndex() -> (FaceFormatIndex:int=0)
   n25 = Get facecamBorder_1() -> (facecamBorder_1:Border Object Reference)
   n26 = Get facecamBorder() -> (facecamBorder:Border Object Reference)
   SetVisibility(
    execute=[@n23.then, @n71.then]
    self=[@n25, @n26]
    InVisibility=Select(Option 0=Collapsed, Option 1=Visible, Index="Not Equal (Integer)"(A=@n24))
   )
   > SetRenderScale!(
    execute
    self:Widget Object Reference=[@n25, @n26]
    Scale:Vector 2D Structure=Select!(
     Option 0:Vector 2D Structure
     Option 1:Vector 2D Structure=(X=1,Y=1)
     Option 2:Vector 2D Structure=(X=1.5,Y=1.5)
     Option 3:Vector 2D Structure=(X=2,Y=2)
     Index:int=Get FaceFormatIndex() -> (FaceFormatIndex=0)
    ) -> (ReturnValue:Vector 2D Structure)
   ) -> (then)
   "On Selection Changed (ComboBox_ViewLayouts)"!() -> (OutputDelegate, then, SelectedItem:String, SelectionType:ESelectInfo Enum) {is=title K2Node_ComponentBoundEvent}
   > Set ViewLayoutIndex(ViewLayoutIndex:int=GetSelectedIndex(self=Get ComboBox_ViewLayouts() -> (ComboBox_ViewLayouts:Editor Utility Combo Box String Object Reference)) -> (ReturnValue=0)) -> (Output_Get:int=0)
   > n27 = SaveGame_Save()
   n28 = Get ViewLayoutIndex() -> (ViewLayoutIndex:int=0)
   SetVisibility(
    execute=[@n27.then, @n74.then]
    self=Get ScaleBox_FullView() -> (ScaleBox_FullView:Scale Box Object Reference)
    InVisibility=Select!(Option 0:ESlateVisibility Enum=Collapsed, Option 1:ESlateVisibility Enum=Visible, Index:bool="Equal (Integer)"(A=@n28, B=1)) -> (ReturnValue:ESlateVisibility Enum)
   )
   n29 = "Equal (Integer)"(A=@n28, B=2)
   n30 = "Equal (Integer)"(A=@n28, B=3)
   > SetVisibility(
    self=Get ScaleBox_FullLeft() -> (ScaleBox_FullLeft:Scale Box Object Reference)
    InVisibility=Select(Option 0=Hidden, Option 1=Visible, Index=OR Boolean(A=@n29, B=@n30))
   )
   n31 = "Equal (Integer)"(A=@n28, B=4)
   n32 = "Equal (Integer)"(A=@n28, B=2)
   > SetVisibility(
    self=Get ScaleBox_FullRight() -> (ScaleBox_FullRight:Scale Box Object Reference)
    InVisibility=Select(Option 0=Hidden, Option 1=Visible, Index=OR Boolean(A=@n32, B=@n31))
   )
   n33 = "On Property Changed (PropertyView_Cam_Distance_1)"!() -> (OutputDelegate, then, PropertyName:Name=None) {is=title K2Node_ComponentBoundEvent}
   "On Property Changed (PropertyView_Montage_Size)"!() -> (OutputDelegate, then, PropertyName:Name=None) {is=title K2Node_ComponentBoundEvent}
   n34 = Get Speed() -> (Speed:Text Object Reference)
   > n35 = Set members in Slate Font Info!(
    execute:Exec
    StructRef:"Slate Font Info Structure (by ref)"=Get Font!(self:Text Object Reference=@n34) -> (Font:Slate Font Info Structure)
    Size:float=Get MontageDetails_Size() -> (MontageDetails_Size:double)
   ) -> (then:Exec, StructOut:"Slate Font Info Structure (by ref)") {is=title K2Node_SetFieldsInStruct}
   > SetFont(
    self=[
     Get AnimName() -> (AnimName:Text Object Reference)
     Get Frame() -> (Frame:Text Object Reference)
     Get SectionName() -> (SectionName:Text Object Reference)
     @n34
    ]
    InFontInfo=@n35
   )
   "On Selection Changed (ComboBox_FaceCam_1)"!() -> (OutputDelegate, then, SelectedItem:String, SelectionType:ESelectInfo Enum) {is=title K2Node_ComponentBoundEvent}
   > n36 = Set SetDistance(SetDistance:int=GetSelectedIndex(self=Get ComboBox_FaceCam_1() -> (ComboBox_FaceCam_1:Editor Utility Combo Box String Object Reference)) -> (ReturnValue=0)) -> (Output_Get:int=0)
   > SaveGame_Save()
   > n37 = RefreshCameraCheck()
   SetIsFar(execute=[@n37.then, @n73.then], self=Get CameraActor(), Index=@n36)
   "On Clicked (Button_RemovePickaxes)"() {is=title K2Node_ComponentBoundEvent}
   > RemoveAllIDs(CosmeticType=3)
   n38 = "On Property Changed (PropertyView_Cam_Distance)"!() -> (OutputDelegate, then, PropertyName:Name=None) {is=title K2Node_ComponentBoundEvent}
   "On Clicked (Button_AddSparks)"() {is=title K2Node_ComponentBoundEvent}
   > Add ID(CosmeticType=2)
   "On Clicked (Button_RemoveSparks)"() {is=title K2Node_ComponentBoundEvent}
   > RemoveAllIDs(CosmeticType=2)
   n39 = "On Check State Changed (Checkbox_ShowPickaxes)"!() -> (OutputDelegate, then, bIsChecked:bool) {is=title K2Node_ComponentBoundEvent}
   > Set ShowPickaxes(ShowPickaxes:bool=@n39.bIsChecked) -> (Output_Get:bool)
   > SaveGame_Save()
   "On Clicked (Button_AddSparks_1)"() {is=title K2Node_ComponentBoundEvent}
   > n40 = GetPIEWorlds(bIncludeDedicatedServer=false)
   n41 = GetPlayerController!(WorldContextObject:Object Reference="Get (a copy)"(Array=@n40), PlayerIndex:int=0) -> (ReturnValue:Player Controller Object Reference)
   > ExecuteConsoleCommand(Command=Cheat ClearInventory, SpecificPlayer=@n41)
   > ExecuteConsoleCommand(Command=Cheat giveweapon wid_pilgrim_bandquickplay_guitar, SpecificPlayer=@n41)
   > ExecuteConsoleCommand(Command=Cheat giveweapon wid_pilgrim_bandquickplay_bass, SpecificPlayer=@n41)
   > ExecuteConsoleCommand(Command=Cheat giveweapon wid_pilgrim_bandquickplay_mic, SpecificPlayer=@n41)
   > ExecuteConsoleCommand(Command=Cheat giveweapon wid_pilgrim_bandquickplay_drum, SpecificPlayer=@n41)
   > ExecuteConsoleCommand(Command=Cheat giveweapon wid_pilgrim_bandquickplay_keytar, SpecificPlayer=@n41)
   n42 = "On Check State Changed (Checkbox_EditOverlays)"!() -> (OutputDelegate, then, bIsChecked:bool) {is=title K2Node_ComponentBoundEvent}
   > Set EditTextOverlays(EditTextOverlays:bool=@n42.bIsChecked) -> (Output_Get:bool)
   > SaveGame_Save()
   n43 = "On Check State Changed (Checkbox_ShowCIDs)"!() -> (OutputDelegate, then, bIsChecked:bool) {is=title K2Node_ComponentBoundEvent}
   > Set ShowCIDs(ShowCIDs:bool=@n43.bIsChecked) -> (Output_Get:bool)
   > SaveGame_Save()
   "On Clicked (EditorUtilityButton_4)"() {is=title K2Node_ComponentBoundEvent}
   > RemoveAllIDs(CosmeticType=0)
   "On Clicked (EditorUtilityButton_3)"() {is=title K2Node_ComponentBoundEvent}
   > n44 = RefreshCameraCheck()
   > AddYawInput!(
    execute
    self:Player Controller Object Reference=@n44
    Val:float=float / float(A=30, B=GetDeprecatedInputYawScale!(self:Player Controller Object Reference=@n44) -> (ReturnValue:float))
   ) -> (then)
   "On Clicked (EditorUtilityButton_5)"() {is=title K2Node_ComponentBoundEvent}
   > n45 = RefreshCameraCheck()
   > AddYawInput(self=@n45, Val=float / float(A=-30, B=GetDeprecatedInputYawScale(self=@n45)))
   "On Property Changed (SinglePropertyView)"!() -> (OutputDelegate, then, PropertyName:Name=None) {is=title K2Node_ComponentBoundEvent}
   > n46 = Set EditViews(EditViews:bool=NOT Boolean(A=Get EditViews() -> (EditViews:bool))) -> (Output_Get:bool)
   > n47 = SaveGame_Save()
   Call Hide Combo Boxes!(execute:Exec=[@n47.then, @n114.then], self:EUW Multi Cam Viewer Object Reference, IsVisible:bool=@n46) -> (then:Exec) {is=title K2Node_CallDelegate}
   n48 = SetCosmetic!() -> (OutputDelegate, then, ID:Fort Item Definition Object Reference, CosmeticType:CosmeticPicker_CosmeticTypes Enum=NewEnumerator0) {is=title K2Node_CustomEvent}
   > Execute Python Script!(
    execute
    PythonScript="import TechArt.CosmeticPicker.CosmeticPickerModule as CosmeticPicker\r\nCosmeticPicker.setOverride(id_name, id)"
    idName:String=Enum to String!(Enumerator:CosmeticPicker_CosmeticTypes Enum=@n48.CosmeticType) -> (ReturnValue:String)
    id:Fort Item Definition Object Reference=@n48.ID
   ) -> (then, ReturnValue) {inputs [idName,id]}
   > RefreshVariantWidget!(execute, self, id:Fort Item Definition Object Reference=@n48.ID) -> (then)
   n49 = RefreshCameras() {is=title K2Node_CustomEvent}
   n50 = PlayEmote!() -> (OutputDelegate, then, EmoteAsset:Fort Montage Item Definition Base Object Reference) {is=title K2Node_CustomEvent}
   > n51 = RefreshCameraCheck()
   > PlayEmoteItem!(
    execute
    self:Fort Player Controller Object Reference=@n51
    EmoteAsset:Fort Montage Item Definition Base Object Reference=@n50.EmoteAsset
    PlayMode:EFortEmotePlayMode Enum=ForcePlay
   ) -> (then)
   n52 = Generate IDs!() -> (OutputDelegate, then, CosmeticType:int) {is=title K2Node_CustomEvent}
   n53 = Get WrapBox_SparksID() -> (WrapBox_SparksID:Wrap Box Object Reference)
   n54 = Get WrapBox_CIDs() -> (WrapBox_CIDs:Wrap Box Object Reference)
   n55 = Get WrapBoxEID() -> (WrapBoxEID:Wrap Box Object Reference)
   n56 = Get WrapBox_PickaxeID() -> (WrapBox_PickaxeID:Wrap Box Object Reference)
   > n57 = Reverse for Each Loop!(
    Exec:Exec
    Array:Array of Widget Object References=GetAllChildren(self=Select!(
      Option 0:Panel Widget Object Reference=@n55
      Option 1:Panel Widget Object Reference=@n54
      Option 2:Panel Widget Object Reference=@n53
      Option 3:Panel Widget Object Reference=@n56
      Index:int=@n52.CosmeticType
     ) -> (ReturnValue:Panel Widget Object Reference))
   ) -> (LoopBody:Exec, ArrayIndex:int, ArrayElement:Widget Object Reference, Completed:Exec) {is=title K2Node_MacroInstance}
   n58 = For Each Loop(
    Exec=@n57.Completed
    Array:Array of Strings=Select(
     Option 0:Array of Strings=Get EIDs()
     Option 1:Array of Strings=Get CIDs() -> (CIDs:Array of Strings)
     Option 2:Array of Strings=Get SparksIDs() -> (SparksIDs:Array of Strings)
     Option 3:Array of Strings=Get PickaxesIDs() -> (PickaxesIDs:Array of Strings)
     Index=@n52.CosmeticType
    ) -> (ReturnValue:Array of Strings)
   ) -> (Array Element:String)
   > n59 = Get Asset By Object Path!(
    execute
    self:Asset Registry Interface=GetAssetRegistry()
    ObjectPath_PathString:String=@n58
    bIncludeOnlyOnDiskAssets:bool=false
    bSkipARFilteredAssets:bool=true
   ) -> (then, ReturnValue:Asset Data Structure) {is=title K2Node_CallFunction}
   n60 = GetAsset(InAssetData=@n59)
   > n61 = Construct Object from Class!(
    execute
    Class=/Game/NeverCook/Animation/MultiCamViewer/Components/EUW_MultiCamViewer_IdButton.EUW_MultiCamViewer_IdButton_C
    self
    Index:int=@n58.Array Index
    ItemDef:Item Definition Base Object Reference=Select!(
     Option 0:Item Definition Base Object Reference
     Option 1:Item Definition Base Object Reference=Cast To ItemDefinitionBase!(Object=@n60) -> (AsItem Definition Base:Item Definition Base Object Reference, bSuccess:bool=~) {is=title K2Node_DynamicCast}
     Index:bool=Is Valid Asset Data!(InAssetData:"Asset Data Structure (by ref)"=@n59) -> (ReturnValue:bool)
    ) -> (ReturnValue:Item Definition Base Object Reference)
    IsEditable:bool=True
    IsBeingEdited:bool=False
    ReplaceName:bool=False
    NewName:Text="LOCTEXT(\"CMM\")"
    CosmeticType:int=@n52.CosmeticType
   ) -> (then, ReturnValue:EUW Multi Cam Viewer Id Button Object Reference)
   > AddChild(
    self=Select!(
     Option 0:Panel Widget Object Reference=Get WrapBoxEID()
     Option 1:Panel Widget Object Reference=Get WrapBox_CIDs()
     Option 2:Panel Widget Object Reference=Get WrapBox_SparksID()
     Option 3:Panel Widget Object Reference=Get WrapBox_PickaxeID()
     Index:int=@n52.CosmeticType
    ) -> (ReturnValue:Panel Widget Object Reference)
    Content=@n61
   )
   Branch(
    execute=@n57.LoopBody
    Condition="integer > integer"(A=@n57, B=Select(Option 0:int=0, Option 1:int=9, Option 2:int=0, Option 3:int=0, Index=@n52.CosmeticType) -> (ReturnValue:int))
   )
   > RemoveFromParent(self=@n57.ArrayElement)
   n62 = Delete single text overlay!() -> (OutputDelegate, then, IndexToRemove:int) {is=title K2Node_CustomEvent}
   > Remove Index!(execute, TargetArray=Get TextsText(), IndexToRemove:int=@n62.IndexToRemove) -> (then)
   > Remove Index(TargetArray:"Array of Float (double-precision)s"=Get TextsSize(), IndexToRemove=@n62.IndexToRemove)
   > Remove Index(TargetArray:Array of Linear Color Structures=Get TextsColor(), IndexToRemove=@n62.IndexToRemove)
   > Remove Index(TargetArray:Array of Booleans=Get TextVisibility(), IndexToRemove=@n62.IndexToRemove)
   > Remove Index(TargetArray:Array of Vector 2D Structures=Get TextsPosition(), IndexToRemove=@n62.IndexToRemove)
   > SaveGame_Save()
   > Generate text overlays()
   CheckBoxes() {is=title K2Node_CustomEvent}
   > SetIsChecked!(
    execute
    self:Check Box Object Reference=Get Checkbox_EditOverlays() -> (Checkbox_EditOverlays:Editor Utility Check Box Object Reference)
    InIsChecked:bool=Get EditTextOverlays()
   ) -> (then)
   > SetIsChecked(self=Get Checkbox_ShowEIDs() -> (Checkbox_ShowEIDs:Editor Utility Check Box Object Reference), InIsChecked=Get ShowEIDs())
   > SetIsChecked(self=Get Checkbox_ShowCIDs() -> (Checkbox_ShowCIDs:Editor Utility Check Box Object Reference), InIsChecked=Get ShowCIDs())
   > SetIsChecked(self=Get Checkbox_ShowSparks() -> (Checkbox_ShowSparks:Editor Utility Check Box Object Reference), InIsChecked=Get ShowSparksID())
   > SetIsChecked(self=Get Checkbox_ShowPickaxes() -> (Checkbox_ShowPickaxes:Editor Utility Check Box Object Reference), InIsChecked=Get ShowPickaxes())
   n63 = Add ID!() -> (OutputDelegate, then, CosmeticType:int) {is=title K2Node_CustomEvent}
   > n64 = Add(TargetArray=Select(
     Option 0:Array of Strings=Get EIDs()
     Option 1:Array of Strings=Get CIDs()
     Option 2:Array of Strings=Get SparksIDs()
     Option 3:Array of Strings=Get PickaxesIDs()
     Index=@n63.CosmeticType
    ) -> (ReturnValue:Array of Strings))
   > n65 = Construct Object from Class(
    Class=&EUW_MultiCamViewer_IdButton_C
    Index=@n64
    IsEditable=True
    IsBeingEdited=true
    ReplaceName=False
    NewName="LOCTEXT(\"CMM\")"
    CosmeticType=@n63.CosmeticType
   )
   > AddChild(
    self=Select(
     Option 0:Panel Widget Object Reference=Get WrapBoxEID()
     Option 1:Panel Widget Object Reference=Get WrapBox_CIDs()
     Option 2:Panel Widget Object Reference=Get WrapBox_SparksID()
     Option 3:Panel Widget Object Reference=Get WrapBox_PickaxeID()
     Index=@n63.CosmeticType
    ) -> (ReturnValue:Panel Widget Object Reference)
    Content=@n65
   )
   LoadComboBoxes() {is=title K2Node_CustomEvent}
   > SetSelectedIndex!(execute, self:"Combo Box (String) Object Reference"=Get ComboBox_FaceCam(), Index:int=Get FaceFormatIndex() -> (FaceFormatIndex=0)) -> (then)
   > SetSelectedIndex(self=Get ComboBox_ViewLayouts(), Index=Get ViewLayoutIndex() -> (ViewLayoutIndex=0))
   > SetSelectedIndex(self=Get ComboBox_FaceCam_1(), Index=Get SetDistance() -> (SetDistance=0))
   UpdateAllViews() {is=title K2Node_CustomEvent}
   n66 = Get CurrentViews() -> (CurrentViews:Map of Strings to Strings)
   > n67 = Keys!(execute, TargetMap:Map of Strings to Strings=@n66) -> (then, Keys:Array of Strings)
   > n68 = For Each Loop(Array=@n67)
   > ChangeView!(execute, self, Option:String=Find(TargetMap=@n66, Key=@n68), ScreenId:String=@n68) -> (then)
   n69 = RefreshVariantWidget!() -> (OutputDelegate, then, id:Fort Item Definition Object Reference) {is=title K2Node_CustomEvent}
   > updateVariantWidgets!(
    execute
    self:EUW Cosmetic Pickerv 2 Variants Object Reference=Get EUW_CosmeticPickerv2_Variants() -> (EUW_CosmeticPickerv2_Variants:EUW Cosmetic Pickerv 2 Variants Object Reference)
    id:Fort Item Definition Object Reference=@n69.id
   ) -> (then)
   n70 = Set ID!() -> (OutputDelegate, then, Index:int, ID:Item Definition Base Object Reference, CosmeticType:int) {is=title K2Node_CustomEvent}
   > Set Array Elem(
    TargetArray:Array of Strings=Select(
     Option 0:Array of Strings=Get EIDs()
     Option 1:Array of Strings=Get CIDs()
     Option 2:Array of Strings=Get SparksIDs()
     Option 3:Array of Strings=Get PickaxesIDs()
     Index=@n70.CosmeticType
    ) -> (ReturnValue:Array of Strings)
    Index=@n70.Index
    Item:"String (by ref)"=Get Object Path String(Object=@n70.ID)
    bSizeToFit=false
   )
   > SaveGame_Save()
   n71 = RefreshFaceCamVisibility() {is=title K2Node_CustomEvent}
   n72 = RemoveAllIDs!() -> (OutputDelegate, then, CosmeticType:int) {is=title K2Node_CustomEvent}
   > Clear(TargetArray:Array of Strings=Select(Option 0=Get EIDs(), Option 1=Get CIDs(), Option 2=Get SparksIDs(), Option 3=Get PickaxesIDs(), Index=@n72.CosmeticType))
   > SaveGame_Save()
   > Generate IDs(CosmeticType=@n72.CosmeticType) {is=title K2Node_CallFunction}
   n73 = RefreshCameraActorsParams() {is=title K2Node_CustomEvent}
   n74 = RefreshFullView() {is=title K2Node_CustomEvent}
   n75 = Remove ID!() -> (OutputDelegate, then, Index:int, CosmeticType:int) {is=title K2Node_CustomEvent}
   > Remove Index(
    TargetArray:Array of Strings=Select(Option 0=Get EIDs(), Option 1=Get CIDs(), Option 2=Get SparksIDs(), Option 3=Get PickaxesIDs(), Index=@n75.CosmeticType)
    IndexToRemove=@n75.Index
   )
   > SaveGame_Save()
   > Generate IDs(CosmeticType=@n75.CosmeticType)
   n76 = ApplySparksID!() -> (OutputDelegate, then, Target:Item Definition Base Object Reference) {is=title K2Node_CustomEvent}
   n77 = Get ItemName!(self:Item Definition Base Object Reference=@n76.Target) -> (ItemName:Text)
   n78 = GetDisplayName(Object=@n76.Target)
   n79 = Select!(
    Option 0:Text=@n77
    Option 1:Text=Format Text!(Format="LOCTEXT(\"TBD ({item})\")", item:Text="To Text (String)"(InString=@n78)) -> (Result)
    Index:bool="Equal Exactly (Text)"!(A:"Text (by ref)"=@n77, B:"Text (by ref)"=MakeLiteralText(Value="LOCTEXT(\"TBD\")")) -> (ReturnValue:bool)
   ) -> (ReturnValue:Text)
   > PrintString(InString="To String (Text)"(InText=@n79))
   > n80 = GetPIEWorlds(bIncludeDedicatedServer=false)
   > ExecuteConsoleCommand(
    WorldContextObject="Get (a copy)"(Array=@n80)
    Command="To String (Text)"(InText=Format Text(Format="LOCTEXT(\"Sparks.SetCosmetic {item}\")", item=@n79))
   )
   n81 = ChangeView!() -> (OutputDelegate, then, Option:String, ScreenId:String) {is=title K2Node_CustomEvent}
   > Add!(execute, TargetMap:Map of Strings to Strings=Get CurrentViews(), Key:"String (by ref)"=@n81.ScreenId, Value:"String (by ref)"=@n81.Option) -> (then) {is=title K2Node_CallFunction}
   > SaveGame_Save()
   > SetTextureParameterValue(
    self=Find(
     TargetMap:Map of Strings to Material Instance Dynamic Object References=Get StoredMaterials() -> (StoredMaterials:Map of Strings to Material Instance Dynamic Object References)
     Key=@n81.ScreenId
    ) -> (Value:Material Instance Dynamic Object Reference)
    ParameterName=RenderTarget
    Value=Find(
     TargetMap:Map of Strings to Texture Object References=Get AvalibleViews() -> (AvalibleViews:Map of Strings to Texture Object References)
     Key=@n81.Option
    ) -> (Value:Texture Object Reference)
   )
   n82 = SaveTextEdit!() -> (
    OutputDelegate
    then
    Index:int
    Property:Name
    Text:String
    Size:double=~
    Position:Vector 2D Structure
    Color:Linear Color Structure
    Visibility:bool=~
    OutlineSize:int
    OutlineColor:Linear Color Structure
   ) {is=title K2Node_CustomEvent}
   > n83 = Switch on Name!(execute:Exec, Selection:Name=@n82.Property) -> (Text:Exec, Size:Exec, Position:Exec, Color:Exec, TextVisible:Exec, OutlineSize:Exec, OutlineColor:Exec) {is=title K2Node_SwitchName, bHasDefaultPin=false, pinNames [Text,Size,Position,Color,TextVisible,OutlineSize,OutlineColor]}
   n84 = Set Array Elem(
    execute=@n83.Color
    TargetArray:Array of Linear Color Structures=Get TextsColor()
    Index=@n82.Index
    Item:"Linear Color Structure (by ref)"=@n82.Color
    bSizeToFit=false
   )
   SaveGame_Save(execute=[@n85.then, @n90.then, @n86.then, @n84.then, @n89.then, @n88.then, @n87.then])
   n85 = Set Array Elem(execute=@n83.OutlineColor, TargetArray=Get TextOutlineColor(), Index=@n82.Index, Item=@n82.OutlineColor, bSizeToFit=false)
   n86 = Set Array Elem(
    execute=@n83.OutlineSize
    TargetArray:Array of Integers=Get TextOutlineSize()
    Index=@n82.Index
    Item:"Integer (by ref)"=@n82.OutlineSize
    bSizeToFit=false
   )
   n87 = Set Array Elem(
    execute=@n83.Position
    TargetArray:Array of Vector 2D Structures=Get TextsPosition()
    Index=@n82.Index
    Item:"Vector 2D Structure (by ref)"=@n82.Position
    bSizeToFit=false
   )
   n88 = Set Array Elem(
    execute=@n83.Size
    TargetArray:"Array of Float (double-precision)s"=Get TextsSize()
    Index=@n82.Index
    Item:"Float (double-precision) (by ref)"=@n82.Size
    bSizeToFit=false
   )
   n89 = Set Array Elem(execute=@n83.Text, TargetArray:Array of Strings=Get TextsText(), Index=@n82.Index, Item:"String (by ref)"=@n82.Text, bSizeToFit=false)
   n90 = Set Array Elem(
    execute=@n83.TextVisible
    TargetArray:Array of Booleans=Get TextVisibility()
    Index=@n82.Index
    Item:"Boolean (by ref)"=@n82.Visibility
    bSizeToFit=false
   )
   Reload Views() {is=title K2Node_CustomEvent}
   > n91 = Sequence()
   > n92 = Keys(TargetMap=Get CurrentViews())
   > n93 = For Each Loop(Array=@n92)
   > n94 = CreateDynamicMaterialInstance(Parent=/Game/NeverCook/Animation/MultiCamViewer/Components/M_MultiCamViewer_RenderTarget, OptionalName=None, CreationFlags=None)
   > Add(
    TargetMap:Map of Strings to Material Instance Dynamic Object References=Get StoredMaterials()
    Key=@n93
    Value:"Material Instance Dynamic Object Reference (by ref)"=@n94
   )
   > n95 = Keys(TargetMap:Map of Strings to Texture Object References=Get AvalibleViews())
   n96 = Find(TargetMap:Map of Strings to Strings=Get CurrentViews(), Key=@n93) -> (Value:String)
   > n97 = Create Widget!(
    execute:Exec
    WorldContextObject:Object Reference
    Class:User Widget Class Reference=/Game/NeverCook/Animation/MultiCamViewer/Components/EUW_MultiCamViewer_Screen.EUW_MultiCamViewer_Screen_C
    OwningPlayer:Player Controller Object Reference
    AvalibleViews:Array of Strings=@n95
    ScreenId:String=@n93
    Material:Material Interface Object Reference=@n94
    CurrentOption:String=@n96
    ParentWidget:EUW Multi Cam Viewer Object Reference=Self-Reference()
   ) -> (then:Exec, ReturnValue:EUW Multi Cam Viewer Screen Object Reference)
   n98 = "Get (a copy)"(
    Array:Array of Panel Widget Object References=Make Array!(
     "[0]":Panel Widget Object Reference=Get Border_CameraView_UpperLeft() -> (Border_CameraView_UpperLeft:Border Object Reference)
     "[1]":Panel Widget Object Reference=Get Border_CameraView_UpperRight() -> (Border_CameraView_UpperRight:Border Object Reference)
     "[2]":Panel Widget Object Reference=Get Border_CameraView_LowerLeft() -> (Border_CameraView_LowerLeft:Border Object Reference)
     "[3]":Panel Widget Object Reference=Get Border_CameraView_LowerRight() -> (Border_CameraView_LowerRight:Border Object Reference)
     "[4]":Panel Widget Object Reference=Get Border_CameraView_Full() -> (Border_CameraView_Full:Border Object Reference)
     "[5]":Panel Widget Object Reference=Get Border_CameraView_FullRight() -> (Border_CameraView_FullRight:Border Object Reference)
     "[6]":Panel Widget Object Reference=Get Border_CameraView_FullLeft() -> (Border_CameraView_FullLeft:Border Object Reference)
    ) -> (Array:Array of Panel Widget Object References)
    Dimension 1=@n93.Array Index
   ) -> (Output:Panel Widget Object Reference)
   > ClearChildren(self=@n98)
   > AddChild(self=@n98, Content=@n97)
   > Bind Event to View Changed(self:EUW Multi Cam Viewer Screen Object Reference=@n97, Delegate=@n81) {is=title K2Node_AddDelegate}
   > ChangeView!(execute, self:EUW Multi Cam Viewer Screen Object Reference=@n97, Option=@n96) -> (then) {is=title K2Node_CallFunction}
   n99 = CreateDynamicMaterialInstance(
    execute=@n91.then_1
    Parent=&M_MultiCamViewer_RenderTarget
    OptionalName=None
    CreationFlags=None
   )
   > SetTextureParameterValue(self=@n99, ParameterName=RenderTarget, Value=/Game/NeverCook/Animation/MultiCamViewer/Components/RT_MultiCamViewer_faceCam)
   > SetBrushFromMaterial(self=[Get faceCam_1() -> (faceCam_1:Image Object Reference), Get faceCam() -> (faceCam:Image Object Reference)], Material=@n99)
   n100 = RefreshMontageDetail() {is=title K2Node_CustomEvent}
   Generate text overlays() {is=title K2Node_CustomEvent}
   > ClearChildren(self=[Get CanvasPanel_TextOverlays(), Get VerticalBox_textoverlayproperty()])
   > n101 = For Each Loop(Array=Get TextsText())
   > n102 = Construct Object from Class!(execute, Class=&TextBlock, self) -> (then, ReturnValue:Text Object Reference)
   > n103 = AddChildToCanvas!(execute, self:Canvas Panel Object Reference=Get CanvasPanel_TextOverlays(), Content:Widget Object Reference=@n102) -> (then, ReturnValue:Canvas Panel Slot Object Reference)
   > SetZOrder!(execute, self:Canvas Panel Slot Object Reference=@n103, InZOrder:int=10) -> (then)
   > SetAlignment!(execute, self:Canvas Panel Slot Object Reference=@n103, InAlignment:Vector 2D Structure=(X=0.5,Y=1)) -> (then)
   n104 = To Vector2D!(InDouble:double=Make Literal Float(Value=0.5)) -> (ReturnValue:Vector 2D Structure)
   > SetAnchors!(
    execute
    self:Canvas Panel Slot Object Reference=@n103
    InAnchors_Minimum:Vector 2D Structure=@n104
    InAnchors_Maximum:Vector 2D Structure=@n104
   ) -> (then)
   > Set Justification!(execute, Justification:ETextJustify Enum=Center, self:Text Layout Widget Object Reference=@n102) -> (then, Output_Get:ETextJustify Enum=Left)
   > Set bAutoSize!(execute, bAutoSize:bool=true, self:Canvas Panel Slot Object Reference=@n103) -> (then, Output_Get:bool)
   n105 = "Get (a copy)"(Array:"Array of Float (double-precision)s"=Get TextsSize(), Dimension 1=@n101.Array Index) -> (Output:double)
   n106 = "Get (a copy)"(Array:Array of Vector 2D Structures=Get TextsPosition(), Dimension 1=@n101.Array Index) -> (Output:Vector 2D Structure)
   n107 = "Get (a copy)"(Array:Array of Linear Color Structures=Get TextsColor(), Dimension 1=@n101.Array Index) -> (Output:Linear Color Structure)
   n108 = "Get (a copy)"(Array:Array of Booleans=Get TextVisibility(), Dimension 1=@n101.Array Index) -> (Output:bool)
   n109 = "Get (a copy)"(Array:Array of Integers=Get TextOutlineSize(), Dimension 1=@n101.Array Index) -> (Output:int)
   n110 = "Get (a copy)"(Array:Array of Linear Color Structures=Get TextOutlineColor(), Dimension 1=@n101.Array Index) -> (Output:Linear Color Structure)
   n111 = "Get (a copy)"(Array:Array of Strings=Get TextsText(), Dimension 1=@n101.Array Index) -> (Output:String)
   > ModifyOverlayText!(
    execute
    self
    TextComponent:Text Object Reference=@n102
    Text:String=@n111
    Size:float=@n105
    Position:Vector 2D Structure=@n106
    Color:Linear Color Structure=@n107
    Visibility:bool=@n108
    Outline Size:int=@n109
    Outline Color:Linear Color Structure=@n110
   ) -> (then)
   > n112 = Construct Object from Class!(
    execute
    Class=/Game/NeverCook/Animation/MultiCamViewer/Components/EUW_MultiCamViewer_TextDetails.EUW_MultiCamViewer_TextDetails_C
    self
    Index:int=@n101.Array Index
    Parent:EUW Multi Cam Viewer Object Reference=Self-Reference()
    Size:double=@n105
    Position:Vector 2D Structure=@n106
    Color:Linear Color Structure=@n107
    Text:String=@n111
    TextVisible:bool=@n108
    TextComponentRef:Text Object Reference=@n102
    OutlineColor:Linear Color Structure=@n110
    OutlineSize:int=@n109
   ) -> (then, ReturnValue:EUW Multi Cam Viewer Text Details Object Reference)
   > n113 = AddChildToVerticalBox(self=Get VerticalBox_textoverlayproperty(), Content=@n112)
   > SetHorizontalAlignment(self:Vertical Box Slot Object Reference=@n113, InHorizontalAlignment=HAlign_Left)
   n114 = RefreshEditView() {is=title K2Node_CustomEvent}
   Event PreConstruct()
   > SaveGame_Init(self:TA Base Editor Widget Object Reference)
   > n115 = Sequence!(execute) -> (then_0, then_1, then_2:Exec, then_3:Exec, then_4:Exec)
   > SetObject(
    self=[
     Get SinglePropertyView_3() -> (SinglePropertyView_3:Single Property View Object Reference)
     Get SinglePropertyView_61() -> (SinglePropertyView_61:Single Property View Object Reference)
     Get PropertyView_Cam_Distance_1() -> (PropertyView_Cam_Distance_1:Single Property View Object Reference)
     Get SinglePropertyView() -> (SinglePropertyView:Single Property View Object Reference)
     Get SinglePropertyView_48() -> (SinglePropertyView_48:Single Property View Object Reference)
    ]
    NewObject=Self-Reference()
   )
   > SetObject(
    self=[
     Get PropertyView_ViewLables_Visible() -> (PropertyView_ViewLables_Visible:Single Property View Object Reference)
     Get PropertyView_Cam_Distance() -> (PropertyView_Cam_Distance:Single Property View Object Reference)
     Get PropertyView_ViewLables_Position() -> (PropertyView_ViewLables_Position:Single Property View Object Reference)
     Get PropertyView_ViewLables_Color() -> (PropertyView_ViewLables_Color:Single Property View Object Reference)
     Get PropertyView_ViewLables_Size() -> (PropertyView_ViewLables_Size:Single Property View Object Reference)
     Get PropertyView_Montage_Visible() -> (PropertyView_Montage_Visible:Single Property View Object Reference)
     Get PropertyView_Montage_Size() -> (PropertyView_Montage_Size:Single Property View Object Reference)
     Get PropertyView_Montage_Position() -> (PropertyView_Montage_Position:Single Property View Object Reference)
     Get PropertyView_Montage_Color() -> (PropertyView_Montage_Color:Single Property View Object Reference)
     Get PropertyView_Montage_Color()
     Get PropertyView_Cam_FOV() -> (PropertyView_Cam_FOV:Single Property View Object Reference)
    ]
    NewObject=Self-Reference()
   )
   Reload Views(execute=@n115.then_1) {is=title K2Node_CallFunction}
   > RefreshMontageDetail() {is=title K2Node_CallFunction}
   > RefreshEditView() {is=title K2Node_CallFunction}
   > RefreshFaceCamVisibility() {is=title K2Node_CallFunction}
   > RefreshFullView() {is=title K2Node_CallFunction}
   Generate text overlays(execute=@n115.then_2) {is=title K2Node_CallFunction}
   > n116 = RefreshCameraCheck()
   > RefreshVariantWidget(id=GetItemDefinitionFromLoadoutWithController!(Controller:Controller Object Reference=@n116, LoadoutSlot:Gameplay Tag Structure=(TagName="LoadoutSlot.Character")) -> (ReturnValue:Athena Cosmetic Item Definition Object Reference)) {is=title K2Node_CallFunction}
   CheckBoxes(execute=@n115.then_3) {is=title K2Node_CallFunction}
   > LoadComboBoxes() {is=title K2Node_CallFunction}
   n117 = For Loop(execute=@n115.then_4, LastIndex=3)
   > Generate IDs(CosmeticType=@n117)
   Event Tick()
   > n118 = Sequence!(execute) -> (then_0, then_1)
   n119 = Get ActivePawn!(self:BP Multi Cam Viewer Actor Object Reference=Get CameraActor()) -> (ActivePawn:Fort Player Pawn Athena Object Reference)
   > Is Valid(InputObject=@n119)
   n120 = GetAnimInstance(self=Get Mesh(self=@n119))
   n121 = GetCurrentActiveMontage!(self:Anim Instance Object Reference=@n120) -> (ReturnValue:Anim Montage Object Reference)
   n122 = GetDisplayName(Object=@n121)
   n123 = IsValid(Object=@n121)
   > "SetText (Text)"(self=Get AnimName(), InText=Select(Option 0="LOCTEXT(\"           \")", Option 1="To Text (String)"(InString=@n122), Index=@n123))
   n124 = Montage_GetPosition!(self:Anim Instance Object Reference=@n120, Montage:Anim Montage Object Reference=@n121) -> (ReturnValue:float)
   > "SetText (Text)"(
    self=Get Frame()
    InText="To Text (Float)"(Value=float * float(A=30, B=@n124), MinimumIntegralDigits=1, MaximumIntegralDigits=324, MinimumFractionalDigits=0, MaximumFractionalDigits=3)
   )
   n125 = GetVelocity!(self:Actor Object Reference=@n119) -> (ReturnValue:Vector)
   > "SetText (Text)"(
    self=Get Speed()
    InText="To Text (Float)"(Value=Vector Length(A=@n125), MinimumIntegralDigits=1, MaximumIntegralDigits=324, MinimumFractionalDigits=0, MaximumFractionalDigits=3)
   )
   n126 = Montage_GetCurrentSection!(self:Anim Instance Object Reference=@n120, Montage:Anim Montage Object Reference=@n121) -> (ReturnValue:Name=None)
   > "SetText (Text)"(
    self=Get SectionName()
    InText=Select(Option 0="LOCTEXT(\"           \")", Option 1="To Text (Name)"!(InName:Name=@n126) -> (ReturnValue:Text), Index=@n123)
   )
   Branch(execute=@n118.then_1, Condition=Get AutoTravel())
   > n127 = RefreshCameraCheck()
   n128 = GetPlayerPawn(self=@n127)
   > AddMovementInput!(
    execute
    self:Pawn Object Reference=@n128
    WorldDirection:Vector=GetActorForwardVector!(self:Actor Object Reference=@n128) -> (ReturnValue:Vector)
    ScaleValue:float=1
    bForce:bool=false
   ) -> (then)
   > SetCurrentMovementStyle!(
    execute
    self:Fort Pawn Object Reference=@n128
    NewMovementStyle:EFortMovementStyle Enum=Get WalkStyle() -> (WalkStyle:EFortMovementStyle Enum=Running)
   ) -> (then)
   "On Property Changed (PropertyView_ViewLables_Size)"!() -> (OutputDelegate, then, PropertyName:Name=None) {is=title K2Node_ComponentBoundEvent}
   Event Construct()
   "Not Equal (Integer)"(A=@n57)
   Get MontageDetails_Size()
   Get saveGame_SlotName() -> (saveGame_SlotName:String)
  ]}
 GetVisibility_0 {nodes [
   GetVisibility_0() {is=title K2Node_FunctionEntry}
   n1 = Get CameraActor()
   > Return Node(ReturnValue=Select(Option 0:ESlateVisibility Enum=Visible, Option 1:ESlateVisibility Enum=Collapsed, Index=IsValid(Object=@n1)) -> (ReturnValue:ESlateVisibility Enum))
  ]}
 Get_Full_Visibility {nodes [
   Get_Full_Visibility() {is=title K2Node_FunctionEntry}
   > Return Node(ReturnValue=Select(Option 0=Collapsed, Option 1=SelfHitTestInvisible, Index=Get FullView() -> (FullView:bool)))
  ]}
 Get_PropertyView_TextOverlay_Visibility {nodes [
   Get_PropertyView_TextOverlay_Visibility() {is=title K2Node_FunctionEntry}
   n1 = Get EditTextOverlays()
   > Return Node(ReturnValue=Select(Option 0=Collapsed, Option 1=Visible, Index=NOT Boolean(A=@n1)))
  ]}
 Get_faceCam_Visibility {nodes [
   Get_faceCam_Visibility() {is=title K2Node_FunctionEntry}
   > Return Node(ReturnValue=Select(Option 0=Visible, Option 1=Collapsed, Index=Get HideFaceCam() -> (HideFaceCam:bool)))
  ]}
 HideComboBoxes {nodes [HideComboBoxes!() -> (then, IsVisible:bool=~) {is=title K2Node_FunctionEntry}]}
 ModifyOverlayText {nodes [
   n1 = ModifyOverlayText!() -> (
    then
    TextComponent:Text Object Reference
    Text:String
    Size:float=~
    Position:Vector 2D Structure
    Color:Linear Color Structure
    Visibility:bool=~
    Outline Size:int
    Outline Color:Linear Color Structure
   ) {is=title K2Node_FunctionEntry}
   > n2 = Cast To CanvasPanelSlot!(execute, Object=Get Slot!(self:Widget Object Reference=@n1) -> (Slot:Panel Slot Object Reference)) -> (then, CastFailed, AsCanvas Panel Slot:Canvas Panel Slot Object Reference) {is=title K2Node_DynamicCast}
   > n3 = SetPosition!(execute, self:Canvas Panel Slot Object Reference=@n2, InPosition:Vector 2D Structure=@n1.Position) -> (then)
   n4 = Get Font(self=@n1)
   n5 = Set members in Font Outline Settings!(
    execute:Exec=[@n3.then, @n2.CastFailed]
    StructRef:"Font Outline Settings Structure (by ref)"=Break Slate Font Info!(SlateFontInfo:"Slate Font Info Structure (by ref)"=@n4) -> (OutlineSettings:Font Outline Settings Structure) {is=title K2Node_BreakStruct}
    OutlineSize:int=@n1.Outline Size
    OutlineColor:Linear Color Structure=@n1.Outline Color
   ) -> (then:Exec, StructOut:"Font Outline Settings Structure (by ref)") {is=title K2Node_SetFieldsInStruct}
   > n6 = Set members in Slate Font Info!(execute, StructRef=@n4, OutlineSettings:Font Outline Settings Structure=@n5, Size=@n1.Size) -> (then, StructOut)
   > SetFont(self=@n1, InFontInfo=@n6)
   > "SetText (Text)"(self=@n1, InText="To Text (String)"(InString=@n1.Text))
   > n7 = Set members in Slate Color!(
    execute:Exec
    StructRef:"Slate Color Structure (by ref)"=Get ColorAndOpacity!(self:Text Object Reference=@n1) -> (ColorAndOpacity:Slate Color Structure)
    SpecifiedColor:Linear Color Structure=@n1.Color
   ) -> (then:Exec, StructOut:"Slate Color Structure (by ref)") {is=title K2Node_SetFieldsInStruct}
   > SetColorAndOpacity(self=@n1, InColorAndOpacity=@n7)
   > SetVisibility(self=@n1, InVisibility=Select(Option 0=Collapsed, Option 1=SelfHitTestInvisible, Index=@n1.Visibility))
  ]}
 RefreshCameraCheck {nodes [
   RefreshCameraCheck() {is=title K2Node_FunctionEntry}
   > n1 = Sequence()
   n2 = Get playerController!(self:BP Multi Cam Viewer Actor Object Reference=Get CameraActor()) -> (playerController:Athena Player Controller Object Reference)
   > n3 = Is Valid(InputObject=@n2)
   RefreshCameras(execute=@n3.Is Not Valid) {is=title K2Node_CallFunction}
   Return Node!(execute=@n1.then_1, playerController:Athena Player Controller Object Reference=@n2)
  ]}
 SaveGame_PostLoad {nodes [
   n1 = SaveGame_PostLoad!() -> (then, saveGameExisted:bool, saveGame_Object:Save Game Object Reference) {is=title K2Node_FunctionEntry}
   > n2 = Cast To SG_MultiCamViewer_Save!(execute, Object=@n1.saveGame_Object) -> (then, CastFailed, AsSG Multi Cam Viewer Save:SG Multi Cam Viewer Save Object Reference) {is=title K2Node_DynamicCast}
   > n3 = Sequence!(execute) -> (then_0, then_1, then_2:Exec)
   n4 = Get CurrentViews!(self:SG Multi Cam Viewer Save Object Reference=@n2) -> (CurrentViews:Array of Strings)
   > Branch(Condition=Is Valid Index!(TargetArray=@n4, IndexToTest:int) -> (ReturnValue:bool))
   n5 = Get CurrentViews!() -> (CurrentViews:Map of Strings to Strings)
   > n6 = Keys(TargetMap:Map of Strings to Strings=@n5)
   > n7 = For Each Loop(Array=@n6)
   n8 = Make Array("[0]":String=Game, "[1]":String=Front, "[2]":String=3/4 L, "[3]":String=Right, "[4]":String=3/4 L, "[5]":String=3/4 L, "[6]":String=Game) -> (Array:Array of Strings)
   > Add(
    TargetMap:Map of Strings to Strings=@n5
    Key=@n7
    Value:"String (by ref)"=Select(
     Option 0:"String (by ref)"="Get (a copy)"(Array=@n8, Dimension 1=@n7.Array Index)
     Option 1:"String (by ref)"="Get (a copy)"(Array=@n4, Dimension 1=@n7.Array Index)
     Index=Is Valid Index(TargetArray=@n4, IndexToTest=@n7.Array Index)
    ) -> (ReturnValue:"String (by ref)")
   )
   Set EIDs(execute=@n3.then_2, EIDs:Array of Strings=Get EIDs!(self:SG Multi Cam Viewer Save Object Reference=@n2) -> (EIDs)) -> (Output_Get:Array of Strings)
   > Set CIDs(CIDs:Array of Strings=Get CIDs!(self:SG Multi Cam Viewer Save Object Reference=@n2) -> (CIDs)) -> (Output_Get:Array of Strings)
   > Set TextsText(TextsText:Array of Strings=Get OverlayText_Texts!(self:SG Multi Cam Viewer Save Object Reference=@n2) -> (OverlayText_Texts:Array of Strings)) -> (Output_Get:Array of Strings)
   > Set TextsSize(TextsSize:"Array of Float (double-precision)s"=Get OverlayText_Size!(self:SG Multi Cam Viewer Save Object Reference=@n2) -> (OverlayText_Size:"Array of Float (double-precision)s")) -> (Output_Get:"Array of Float (double-precision)s")
   > Set TextsPosition(TextsPosition:Array of Vector 2D Structures=Get OverlayText_Position!(self:SG Multi Cam Viewer Save Object Reference=@n2) -> (OverlayText_Position:Array of Vector 2D Structures)) -> (Output_Get:Array of Vector 2D Structures)
   > Set TextsColor(TextsColor:Array of Linear Color Structures=Get OverlayText_LinearColors!(self:SG Multi Cam Viewer Save Object Reference=@n2) -> (OverlayText_LinearColors:Array of Linear Color Structures)) -> (Output_Get:Array of Linear Color Structures)
   > Set TextVisibility(TextVisibility:Array of Booleans=Get OverlayText_Visibility!(self:SG Multi Cam Viewer Save Object Reference=@n2) -> (OverlayText_Visibility:Array of Booleans)) -> (Output_Get:Array of Booleans)
   > Set SparksIDs(SparksIDs:Array of Strings=Get SparksIDs!(self:SG Multi Cam Viewer Save Object Reference=@n2) -> (SparksIDs)) -> (Output_Get:Array of Strings)
   > Set TextOutlineSize(TextOutlineSize:Array of Integers=Get OverlayText_OutlineSize!(self:SG Multi Cam Viewer Save Object Reference=@n2) -> (OverlayText_OutlineSize:Array of Integers)) -> (Output_Get:Array of Integers)
   > Set TextOutlineColor(TextOutlineColor:Array of Linear Color Structures=Get OverlayText_OutlineColor!(self:SG Multi Cam Viewer Save Object Reference=@n2) -> (OverlayText_OutlineColor:Array of Linear Color Structures)) -> (Output_Get:Array of Linear Color Structures)
   > Set PickaxesIDs(PickaxesIDs:Array of Strings=Get PickaxeIDs!(self:SG Multi Cam Viewer Save Object Reference=@n2) -> (PickaxeIDs:Array of Strings)) -> (Output_Get:Array of Strings)
   n9 = Sequence!(execute=@n3.then_1) -> (then_0, then_1)
   n10 = Get indexes!(self:SG Multi Cam Viewer Save Object Reference=@n2) -> (indexes:Array of Integers)
   > Set ViewLayoutIndex(ViewLayoutIndex="Get (a copy)"(Array:Array of Integers=@n10) -> (Output:int)) -> (Output_Get=0)
   > Set FaceFormatIndex(FaceFormatIndex="Get (a copy)"(Array=@n10, Dimension 1=1)) -> (Output_Get=0)
   > Set SetDistance(SetDistance="Get (a copy)"(Array=@n10, Dimension 1=2)) -> (Output_Get=0)
   n11 = Get booleans!(self:SG Multi Cam Viewer Save Object Reference=@n2) -> (booleans:Array of Booleans)
   Set HideFaceCam(execute=@n9.then_1, HideFaceCam:bool="Get (a copy)"(Array:Array of Booleans=@n11) -> (Output:bool)) -> (Output_Get:bool)
   > Set FullView(FullView:bool="Get (a copy)"(Array=@n11, Dimension 1=1)) -> (Output_Get:bool)
   > Set EditViews(EditViews="Get (a copy)"(Array=@n11, Dimension 1=2))
   > Set MontageDetails_Visibility(MontageDetails_Visibility="Get (a copy)"(Array=@n11, Dimension 1=3))
   > Set EditTextOverlays(EditTextOverlays="Get (a copy)"(Array=@n11, Dimension 1=4))
   > Set ShowEIDs(ShowEIDs="Get (a copy)"(Array=@n11, Dimension 1=5))
   > Set ShowCIDs(ShowCIDs="Get (a copy)"(Array=@n11, Dimension 1=6))
   > Set ShowSparksID(ShowSparksID="Get (a copy)"(Array=@n11, Dimension 1=7))
   > Set ShowPickaxes(ShowPickaxes="Get (a copy)"(Array=@n11, Dimension 1=8))
   n12 = "Get (a copy)"!(Array:Array of Vectors=Get CameraValues!(self:SG Multi Cam Viewer Save Object Reference=@n2) -> (CameraValues:Array of Vectors), Dimension 1) -> (Output_X:double, Output_Y:double, Output_Z:double)
   Set Cameras_FOV(Cameras_FOV:double=@n12) -> (Output_Get:double)
   > Set Cameras_Distance(Cameras_Distance:double=@n12.Output_Y) -> (Output_Get:double)
   > Set Cameras_Height(Cameras_Height:double=@n12.Output_Z) -> (Output_Get:double)
   Return Node!(execute, " ":bool=false)
  ]}
 SaveGame_PrepareToSave {nodes [
   n1 = SaveGame_PrepareToSave!() -> (then, SaveGameObject:Save Game Object Reference) {is=title K2Node_FunctionEntry}
   > n2 = Cast To SG_MultiCamViewer_Save(Object=@n1)
   > n3 = Sequence!(execute) -> (then_0, then_1, then_2:Exec, then_3:Exec)
   > n4 = Values!(execute, TargetMap:Map of Strings to Strings=Get CurrentViews()) -> (then, Values:Array of Strings)
   > Set CurrentViews!(execute, CurrentViews:Array of Strings=@n4, self:SG Multi Cam Viewer Save Object Reference=@n2) -> (then, Output_Get:Array of Strings)
   > Set EIDs!(execute, EIDs=Get EIDs!() -> (EIDs), self:SG Multi Cam Viewer Save Object Reference=@n2) -> (then, Output_Get)
   > Set CIDs!(execute, CIDs=Get CIDs!() -> (CIDs), self:SG Multi Cam Viewer Save Object Reference=@n2) -> (then, Output_Get)
   > Set booleans!(
    execute
    booleans:Array of Booleans=Make Array!(
     "[0]":bool=Get HideFaceCam()
     "[1]":bool=Get FullView()
     "[2]":bool=Get EditViews()
     "[3]":bool=Get MontageDetails_Visibility()
     "[4]":bool=Get EditTextOverlays()
     "[5]":bool=Get ShowEIDs()
     "[6]":bool=Get ShowCIDs()
     "[7]":bool=Get ShowSparksID()
     "[8]":bool=Get ShowPickaxes()
    ) -> (Array:Array of Booleans)
    self:SG Multi Cam Viewer Save Object Reference=@n2
   ) -> (then, Output_Get:Array of Booleans)
   > Set SparksIDs!(execute, SparksIDs=Get SparksIDs!() -> (SparksIDs), self:SG Multi Cam Viewer Save Object Reference=@n2) -> (then, Output_Get)
   > Set indexes!(
    execute
    indexes:Array of Integers=Make Array!(
     "[0]":int=Get ViewLayoutIndex() -> (ViewLayoutIndex=0)
     "[1]":int=Get FaceFormatIndex() -> (FaceFormatIndex=0)
     "[2]":int=Get SetDistance() -> (SetDistance=0)
    ) -> (Array:Array of Integers)
    self:SG Multi Cam Viewer Save Object Reference=@n2
   ) -> (then, Output_Get:Array of Integers)
   > Set PickaxeIDs!(execute, PickaxeIDs:Array of Strings=Get PickaxesIDs(), self:SG Multi Cam Viewer Save Object Reference=@n2) -> (then, Output_Get:Array of Strings)
   Set OverlayText_LinearColors!(
    execute=@n3.then_1
    OverlayText_LinearColors:Array of Linear Color Structures=Get TextsColor()
    self:SG Multi Cam Viewer Save Object Reference=@n2
   ) -> (then, Output_Get:Array of Linear Color Structures)
   > Set OverlayText_Position!(execute, OverlayText_Position:Array of Vector 2D Structures=Get TextsPosition(), self:SG Multi Cam Viewer Save Object Reference=@n2) -> (then, Output_Get:Array of Vector 2D Structures)
   > Set OverlayText_Size!(execute, OverlayText_Size:"Array of Float (double-precision)s"=Get TextsSize(), self:SG Multi Cam Viewer Save Object Reference=@n2) -> (then, Output_Get:"Array of Float (double-precision)s")
   > Set OverlayText_Texts!(execute, OverlayText_Texts:Array of Strings=Get TextsText(), self:SG Multi Cam Viewer Save Object Reference=@n2) -> (then, Output_Get:Array of Strings)
   > Set OverlayText_Visibility!(execute, OverlayText_Visibility:Array of Booleans=Get TextVisibility(), self:SG Multi Cam Viewer Save Object Reference=@n2) -> (then, Output_Get:Array of Booleans)
   > Set OverlayText_OutlineSize!(execute, OverlayText_OutlineSize:Array of Integers=Get TextOutlineSize(), self:SG Multi Cam Viewer Save Object Reference=@n2) -> (then, Output_Get:Array of Integers)
   > Set OverlayText_OutlineColor!(
    execute
    OverlayText_OutlineColor:Array of Linear Color Structures=Get TextOutlineColor()
    self:SG Multi Cam Viewer Save Object Reference=@n2
   ) -> (then, Output_Get:Array of Linear Color Structures)
   Return Node(execute=@n3.then_2, " "=false)
   n5 = Get Cameras_FOV()
   n6 = Get Cameras_Distance()
   n7 = Get Cameras_Height()
   Set CameraValues!(
    execute
    CameraValues:Array of Vectors=Make Array!("[0]":Vector=MakeVector(X=@n5, Y=@n6, Z=@n7)) -> (Array:Array of Vectors)
    self:SG Multi Cam Viewer Save Object Reference=@n2
   ) -> (then, Output_Get:Array of Vectors)
  ]}
]
variables [
 AutoTravel {type=boolean}
 AvalibleViews {
  type="map<Texture>"
  default {
   3/4 L=/Game/NeverCook/Animation/MultiCamViewer/Components/RT_MultiCamViewer_threeQuartersL
   3/4 R=/Game/NeverCook/Animation/MultiCamViewer/Components/RT_MultiCamViewer_threeQuartersR
   Back=/Game/NeverCook/Animation/MultiCamViewer/Components/RT_MultiCamViewer_Back
   Free=/Game/NeverCook/Animation/MultiCamViewer/Components/RT_MultiCamViewer_Free
   Front=/Game/NeverCook/Animation/MultiCamViewer/Components/RT_MultiCamViewer_Front
   Game=/Game/NeverCook/Animation/MultiCamViewer/Components/RT_MultiCamViewer_gameCamHalf
   Left=/Game/NeverCook/Animation/MultiCamViewer/Components/RT_MultiCamViewer_Left
   Right=/Game/NeverCook/Animation/MultiCamViewer/Components/RT_MultiCamViewer_Right
  }
 }
 CameraActor {type=BP_MultiCamViewer_Actor_C, default=null}
 Cameras_Distance {type=number, default=260}
 Cameras_FOV {type=number, default=80}
 Cameras_Height {type=number}
 CIDs {type="array<string>", default []}
 CurrentViews {
  type="map<string>"
  default {Full=3/4 L, FullLeft=Game, FullRight=3/4 L, LowerLeft=3/4 L, LowerRight=Right, UpperLeft=Game, UpperRight=Front}
 }
 EditTextOverlays {type=boolean}
 EditViews {type=boolean}
 EID {type=AthenaDanceItemDefinition, default=/SparksCosmetics/Emotes/Drum/JapaneseSamba/EID_Drum_JapaneseSamba}
 EID Release {type=FortReleaseVersion, default {versionName=Unset}}
 EIDs {type="array<string>", default []}
 FaceFormatIndex {type=integer, default=1}
 FullView {type=boolean, default=true}
 HideComboBoxes {default=()}
 HideFaceCam {type=boolean}
 MontageDetails_Color {type=LinearColor, default=<rgba 0 0 0 1>}
 MontageDetails_Position {type=Vector2D, default=<0 0>}
 MontageDetails_Size {type=number, default=40}
 MontageDetails_Visibility {type=boolean, default=true}
 PickaxesIDs {type="array<string>", default []}
 Player Index {type=integer}
 RenderTargets {
  type="array<Texture>"
  default [
   &RT_MultiCamViewer_Front
   /Game/NeverCook/Animation/MultiCamViewer/Components/RT_MultiCamViewer_Game
   &RT_MultiCamViewer_gameCamHalf
   &RT_MultiCamViewer_threeQuartersL
   &RT_MultiCamViewer_threeQuartersR
   &RT_MultiCamViewer_faceCam
  ]
 }
 Selection Type {type=ESelectInfo, default=OnKeyPress}
 SetDistance {type=integer}
 ShowCIDs {type=boolean}
 ShowEIDs {type=boolean}
 ShowPickaxes {type=boolean}
 ShowSparksID {type=boolean}
 SparksID {type=SparksInstrumentItemDefBase, default=/SparksCosmetics/Instrument/Guitar/HydroBeam/Sparks_HydroBeam_Guitar}
 SparksIDs {type="array<string>", default []}
 StoredMaterials {type="map<MaterialInstanceDynamic>", default {Full=null, LowerLeft=null, LowerRight=null, UpperLeft=null, UpperRight=null}}
 TempCurentViews {
  type="array<Texture>"
  default [
   &RT_MultiCamViewer_Front
   &RT_MultiCamViewer_Game
   &RT_MultiCamViewer_gameCamHalf
   &RT_MultiCamViewer_threeQuartersL
   &RT_MultiCamViewer_threeQuartersR
  ]
 }
 TextOutlineColor {type="array<LinearColor>", default []}
 TextOutlineSize {type="array<integer>"}
 TextsColor {type="array<LinearColor>", default []}
 TextsPosition {type="array<Vector2D>", default []}
 TextsSize {type="array<number>", default []}
 TextsText {type="array<string>", default []}
 TextVisibility {type="array<boolean>", default []}
 ViewLabels_Color {type=LinearColor, default=<rgba 0 0 0 1>}
 ViewLabels_Position {type=Vector2D, default=<0 0>}
 ViewLabels_Size {type=number}
 ViewLabels_Visibility {type=boolean, default=true}
 ViewLayoutIndex {type=integer}
 WalkStyle {type=EFortMovementStyle, default=Charging}
]
widgets {
 info {
  parentClass=/Game/Blueprints/Blutilities/ForTechArt/Widgets/TA_Base_EditorWidget.TA_Base_EditorWidget_C
  rootWidgetClass=&EditorUtilityScrollBox
  widgetCount=179
 }
 widgets [
  EditorUtilityScrollBox_0 {widgetClassPath=&EditorUtilityScrollBox, widget=EditorUtilityScrollBox_0}
  ScaleBox_0 {
   widgetClassPath=&ScaleBox
   parent=EditorUtilityScrollBox_0
   widget=ScaleBox_0
   properties {stretch=ScaleToFitX}
   slot {class=ScrollBoxSlot}
  }
  CanvasPanel_1 {
   widgetClassPath=&CanvasPanel
   parent=ScaleBox_0
   widget=CanvasPanel_1
   slot {class=ScaleBoxSlot, properties {horizontalAlignment=HAlign_Fill, verticalAlignment=VAlign_Fill}}
  }
  CanvasPanel_TextOverlays {
   widgetClassPath=&CanvasPanel
   bIsVariable=true
   parent=CanvasPanel_1
   widget=CanvasPanel_TextOverlays
   slot {class=CanvasPanelSlot, properties {zOrder=10, layoutData {alignment=<0 0>, anchors {maximum=<1 1>, minimum=<0 0>}, offsets=<ltrb 0 0 0 0>}}}
  }
  UEFN_TextBlock_C_0 {
   widgetClassPath=&UEFN_TextBlock
   parent=CanvasPanel_TextOverlays
   widget=UEFN_TextBlock_C_0
   properties {text=Text Block}
   slot {
    class=CanvasPanelSlot
    properties {zOrder=10, layoutData {alignment=<0 0>, anchors {maximum=<0 0>, minimum=<0 0>}, offsets=<ltrb 644 592 165 41>}}
   }
  }
  HorizontalBox_HalfViews {
   widgetClassPath=&HorizontalBox
   bIsVariable=true
   parent=CanvasPanel_1
   widget=HorizontalBox_HalfViews
   slot {class=CanvasPanelSlot, properties {zOrder=1, layoutData {alignment=<0 0>, anchors {maximum=<1 1>, minimum=<0 0>}, offsets=<ltrb 0 0 0 0>}}}
  }
  ScaleBox_FullLeft {
   widgetClassPath=&ScaleBox
   bIsVariable=true
   parent=HorizontalBox_HalfViews
   widget=ScaleBox_FullLeft
   properties {clipping=ClipToBounds, stretch=ScaleToFitY}
   slot {class=HorizontalBoxSlot, properties {horizontalAlignment=HAlign_Center, padding=<ltrb 0 0 0.5 0>, size {sizeRule=Fill, value=1}}}
  }
  Border_CameraView_FullLeft {
   widgetClassPath=&Border
   bIsVariable=true
   parent=ScaleBox_FullLeft
   widget=Border_CameraView_FullLeft
   properties {
    background {
     bIsDynamicallyLoaded=false
     drawAs=NoDrawType
     imageType=NoImage
     mirroring=NoMirror
     resourceName=None
     resourceObject=null
     tiling=NoTile
     imageSize=<32 32>
     margin=<ltrb 0 0 0 0>
     outlineSettings {
      bUseBrushTransparency=false
      roundingType=HalfHeightRadius
      width=0
      color {colorUseRule=UseColor_Specified, specifiedColor=<rgba 0 0 0 0>}
      cornerRadii=<0 0 0 0>
     }
     tintColor {colorUseRule=UseColor_Specified, specifiedColor=<rgba 1 1 1 1>}
     uVRegion {bIsValid=false, max=<0 0>, min=<0 0>}
    }
    padding=<ltrb 0 0 0 0>
   }
   slot {class=ScaleBoxSlot, properties {horizontalAlignment=HAlign_Fill}}
  }
  ScaleBox_FullRight {
   widgetClassPath=&ScaleBox
   bIsVariable=true
   parent=HorizontalBox_HalfViews
   widget=ScaleBox_FullRight
   properties {clipping=ClipToBounds, stretch=ScaleToFitY}
   slot {class=HorizontalBoxSlot, properties {padding=<ltrb 0.5 0 0 0>, size {sizeRule=Fill, value=1}}}
  }
  Border_CameraView_FullRight {
   widgetClassPath=&Border
   bIsVariable=true
   parent=ScaleBox_FullRight
   widget=Border_CameraView_FullRight
   properties {
    background {
     bIsDynamicallyLoaded=false
     drawAs=NoDrawType
     imageType=NoImage
     mirroring=NoMirror
     resourceName=None
     resourceObject=null
     tiling=NoTile
     imageSize=<32 32>
     margin=<ltrb 0 0 0 0>
     outlineSettings {
      bUseBrushTransparency=false
      roundingType=HalfHeightRadius
      width=0
      color {colorUseRule=UseColor_Specified, specifiedColor=<rgba 0 0 0 0>}
      cornerRadii=<0 0 0 0>
     }
     tintColor {colorUseRule=UseColor_Specified, specifiedColor=<rgba 1 1 1 1>}
     uVRegion {bIsValid=false, max=<0 0>, min=<0 0>}
    }
    padding=<ltrb 0 0 0 0>
   }
   slot {class=ScaleBoxSlot, properties {horizontalAlignment=HAlign_Fill}}
  }
  GridPanel_0 {
   widgetClassPath=&GridPanel
   bIsVariable=true
   parent=CanvasPanel_1
   widget=GridPanel_0
   slot {class=CanvasPanelSlot, properties {bAutoSize=true, layoutData=~}}
  }
  Border_CameraView_UpperLeft {
   widgetClassPath=&Border
   bIsVariable=true
   parent=GridPanel_0
   widget=Border_CameraView_UpperLeft
   properties {
    background {
     bIsDynamicallyLoaded=false
     drawAs=NoDrawType
     imageType=NoImage
     mirroring=NoMirror
     resourceName=None
     resourceObject=null
     tiling=NoTile
     imageSize=<32 32>
     margin=<ltrb 0 0 0 0>
     outlineSettings {
      bUseBrushTransparency=false
      roundingType=HalfHeightRadius
      width=0
      color {colorUseRule=UseColor_Specified, specifiedColor=<rgba 0 0 0 0>}
      cornerRadii=<0 0 0 0>
     }
     tintColor {colorUseRule=UseColor_Specified, specifiedColor=<rgba 1 1 1 1>}
     uVRegion {bIsValid=false, max=<0 0>, min=<0 0>}
    }
    padding=<ltrb 0 0 0 0>
   }
   slot {class=GridSlot}
  }
  Border_CameraView_UpperRight {
   widgetClassPath=&Border
   bIsVariable=true
   parent=GridPanel_0
   widget=Border_CameraView_UpperRight
   properties {padding=<ltrb 0 0 0 0>}
   slot {class=GridSlot, properties {column=1}}
  }
  Border_CameraView_LowerLeft {
   widgetClassPath=&Border
   bIsVariable=true
   parent=GridPanel_0
   widget=Border_CameraView_LowerLeft
   properties {padding=<ltrb 0 0 0 0>}
   slot {class=GridSlot, properties {row=1}}
  }
  Border_CameraView_LowerRight {
   widgetClassPath=&Border
   bIsVariable=true
   parent=GridPanel_0
   widget=Border_CameraView_LowerRight
   properties {padding=<ltrb 0 0 0 0>}
   slot {class=GridSlot, properties {column=1, row=1}}
  }
  facecamBorder {
   widgetClassPath=&Border
   bIsVariable=true
   parent=CanvasPanel_1
   widget=facecamBorder
   properties {
    background {
     bIsDynamicallyLoaded=false
     drawAs=Box
     imageType=NoImage
     mirroring=NoMirror
     resourceName=None
     resourceObject=null
     tiling=NoTile
     imageSize=<32 32>
     margin=<ltrb 0 0 0 0>
     outlineSettings {
      bUseBrushTransparency=false
      roundingType=HalfHeightRadius
      width=0
      color {colorUseRule=UseColor_Specified, specifiedColor=<rgba 0 0 0 0>}
      cornerRadii=<0 0 0 0>
     }
     tintColor {colorUseRule=UseColor_Specified, specifiedColor=<rgba 0 0 0 1>}
     uVRegion {bIsValid=false, max=<0 0>, min=<0 0>}
    }
    padding=<ltrb 3 2 3 2>
   }
   slot {
    class=CanvasPanelSlot
    properties {zOrder=2, layoutData {alignment=<0.5 0.5>, anchors {maximum=<0.5 0.5>, minimum=<0.5 0.5>}, offsets=<ltrb 0 0 700 750>}}
   }
  }
  faceCam {
   widgetClassPath=&Image
   bIsVariable=true
   parent=facecamBorder
   widget=faceCam
   properties {brush {
     bIsDynamicallyLoaded=false
     drawAs=Image
     imageType=FullColor
     mirroring=NoMirror
     resourceName=None
     resourceObject=null
     tiling=NoTile
     imageSize=<32 32>
     margin=<ltrb 0 0 0 0>
     outlineSettings {
      bUseBrushTransparency=false
      roundingType=HalfHeightRadius
      width=0
      color {colorUseRule=UseColor_Specified, specifiedColor=<rgba 0 0 0 0>}
      cornerRadii=<0 0 0 0>
     }
     tintColor {colorUseRule=UseColor_Specified, specifiedColor=<rgba 1 1 1 1>}
     uVRegion {bIsValid=false, max=<0 0>, min=<0 0>}
    }}
   slot {class=BorderSlot, properties {padding=<ltrb 3 2 3 2>}}
  }
  ScaleBox_FullView {
   widgetClassPath=&ScaleBox
   bIsVariable=true
   parent=CanvasPanel_1
   widget=ScaleBox_FullView
   properties {stretch=ScaleToFitX, visibility=Collapsed}
   slot {class=CanvasPanelSlot, properties {zOrder=3, layoutData {alignment=<0 0>, anchors {maximum=<1 1>, minimum=<0 0>}, offsets=<ltrb 0 0 0 0>}}}
  }
  CanvasPanel_3 {
   widgetClassPath=&CanvasPanel
   parent=ScaleBox_FullView
   widget=CanvasPanel_3
   slot {class=ScaleBoxSlot, properties {horizontalAlignment=HAlign_Fill, verticalAlignment=VAlign_Fill}}
  }
  Border_CameraView_Full {
   widgetClassPath=&Border
   bIsVariable=true
   parent=CanvasPanel_3
   widget=Border_CameraView_Full
   properties {padding=<ltrb 0 0 0 0>}
   slot {class=CanvasPanelSlot, properties {bAutoSize=true, layoutData=~}}
  }
  facecamBorder_1 {
   widgetClassPath=&Border
   bIsVariable=true
   parent=CanvasPanel_3
   widget=facecamBorder_1
   properties {
    background {
     bIsDynamicallyLoaded=false
     drawAs=Box
     imageType=NoImage
     mirroring=NoMirror
     resourceName=None
     resourceObject=null
     tiling=NoTile
     imageSize=<32 32>
     margin=<ltrb 0 0 0 0>
     outlineSettings {
      bUseBrushTransparency=false
      roundingType=HalfHeightRadius
      width=0
      color {colorUseRule=UseColor_Specified, specifiedColor=<rgba 0 0 0 0>}
      cornerRadii=<0 0 0 0>
     }
     tintColor {colorUseRule=UseColor_Specified, specifiedColor=<rgba 0 0 0 1>}
     uVRegion {bIsValid=false, max=<0 0>, min=<0 0>}
    }
    padding=<ltrb 3 2 3 2>
    renderTransformPivot=<0 0>
   }
   slot {
    class=CanvasPanelSlot
    properties {zOrder=2, layoutData {alignment=<0 0>, anchors {maximum=<0 0>, minimum=<0 0>}, offsets=<ltrb 50 50 350 375>}}
   }
  }
  faceCam_1 {
   widgetClassPath=&Image
   bIsVariable=true
   parent=facecamBorder_1
   widget=faceCam_1
   properties {brush {
     bIsDynamicallyLoaded=false
     drawAs=Image
     imageType=FullColor
     mirroring=NoMirror
     resourceName=None
     resourceObject=&RT_MultiCamViewer_faceCam
     tiling=NoTile
     imageSize=<256.1955 277.4392>
     margin=<ltrb 0 0 0 0>
     outlineSettings {
      bUseBrushTransparency=false
      roundingType=HalfHeightRadius
      width=0
      color {colorUseRule=UseColor_Specified, specifiedColor=<rgba 0 0 0 0>}
      cornerRadii=<0 0 0 0>
     }
     tintColor {colorUseRule=UseColor_Specified, specifiedColor=<rgba 1 1 1 1>}
     uVRegion {bIsValid=false, max=<0 0>, min=<0 0>}
    }}
   slot {class=BorderSlot, properties {padding=<ltrb 3 2 3 2>}}
  }
  VerticalBox_0 {
   widgetClassPath=&VerticalBox
   bIsVariable=true
   parent=CanvasPanel_1
   widget=VerticalBox_0
   slot {
    class=CanvasPanelSlot
    properties {bAutoSize=true, zOrder=7, layoutData {alignment=<1 0>, anchors {maximum=<0.5 0>, minimum=<0.5 0>}, offsets=<ltrb -4 0 100 30>}}
   }
  }
  Border_0 {
   widgetClassPath=&Border
   parent=VerticalBox_0
   widget=Border_0
   properties {
    background {
     bIsDynamicallyLoaded=false
     drawAs=RoundedBox
     imageType=NoImage
     mirroring=NoMirror
     resourceName=None
     resourceObject=null
     tiling=NoTile
     imageSize=<32 32>
     margin=<ltrb 0 0 0 0>
     outlineSettings {
      bUseBrushTransparency=false
      roundingType=HalfHeightRadius
      width=0
      color {colorUseRule=UseColor_Specified, specifiedColor=<rgba 0 0 0 0>}
      cornerRadii=<0 0 0 0>
     }
     tintColor {colorUseRule=UseColor_Specified, specifiedColor=<rgba 1 1 1 1>}
     uVRegion {bIsValid=false, max=<0 0>, min=<0 0>}
    }
    brushColor=<rgba 1 1 1 0.5>
   }
   slot {class=VerticalBoxSlot}
  }
  AnimName {
   widgetClassPath=&TextBlock
   bIsVariable=true
   parent=Border_0
   widget=AnimName
   properties {
    text="           "
    colorAndOpacity {colorUseRule=UseColor_Specified, specifiedColor=<rgba 0 0 0 1>}
    font {
     bForceMonospaced=false
     bMaterialIsStencil=false
     fontMaterial=null
     fontObject=&Roboto
     letterSpacing=0
     monospacedWidth=1
     size=30
     skewAmount=0
     typefaceFontName=Bold
     outlineSettings {
      bApplyOutlineToDropShadows=false
      bMiteredCorners=false
      bSeparateFillAlpha=false
      outlineMaterial=null
      outlineSize=0
      outlineColor=<rgba 0 0 0 1>
     }
    }
   }
   slot {class=BorderSlot}
  }
  Border_1 {
   widgetClassPath=&Border
   parent=VerticalBox_0
   widget=Border_1
   properties {
    background {
     bIsDynamicallyLoaded=false
     drawAs=RoundedBox
     imageType=NoImage
     mirroring=NoMirror
     resourceName=None
     resourceObject=null
     tiling=NoTile
     imageSize=<32 32>
     margin=<ltrb 0 0 0 0>
     outlineSettings {
      bUseBrushTransparency=false
      roundingType=HalfHeightRadius
      width=0
      color {colorUseRule=UseColor_Specified, specifiedColor=<rgba 0 0 0 0>}
      cornerRadii=<0 0 0 0>
     }
     tintColor {colorUseRule=UseColor_Specified, specifiedColor=<rgba 1 1 1 1>}
     uVRegion {bIsValid=false, max=<0 0>, min=<0 0>}
    }
    brushColor=<rgba 1 1 1 0.5>
   }
   slot {class=VerticalBoxSlot}
  }
  Frame {
   widgetClassPath=&TextBlock
   bIsVariable=true
   parent=Border_1
   widget=Frame
   properties {
    text=0
    colorAndOpacity {colorUseRule=UseColor_Specified, specifiedColor=<rgba 0 0 0 1>}
    font {
     bForceMonospaced=false
     bMaterialIsStencil=false
     fontMaterial=null
     fontObject=&Roboto
     letterSpacing=0
     monospacedWidth=1
     size=30
     skewAmount=0
     typefaceFontName=Bold
     outlineSettings {
      bApplyOutlineToDropShadows=false
      bMiteredCorners=false
      bSeparateFillAlpha=false
      outlineMaterial=null
      outlineSize=0
      outlineColor=<rgba 0 0 0 1>
     }
    }
   }
   slot {class=BorderSlot}
  }
  VerticalBox_1 {
   widgetClassPath=&VerticalBox
   bIsVariable=true
   parent=CanvasPanel_1
   widget=VerticalBox_1
   slot {
    class=CanvasPanelSlot
    properties {bAutoSize=true, zOrder=7, layoutData {alignment=<0 0>, anchors {maximum=<0.5 0>, minimum=<0.5 0>}, offsets=<ltrb 4 0 100 30>}}
   }
  }
  Border_2 {
   widgetClassPath=&Border
   parent=VerticalBox_1
   widget=Border_2
   properties {
    background {
     bIsDynamicallyLoaded=false
     drawAs=RoundedBox
     imageType=NoImage
     mirroring=NoMirror
     resourceName=None
     resourceObject=null
     tiling=NoTile
     imageSize=<32 32>
     margin=<ltrb 0 0 0 0>
     outlineSettings {
      bUseBrushTransparency=false
      roundingType=HalfHeightRadius
      width=0
      color {colorUseRule=UseColor_Specified, specifiedColor=<rgba 0 0 0 0>}
      cornerRadii=<0 0 0 0>
     }
     tintColor {colorUseRule=UseColor_Specified, specifiedColor=<rgba 1 1 1 1>}
     uVRegion {bIsValid=false, max=<0 0>, min=<0 0>}
    }
    brushColor=<rgba 1 1 1 0.5>
   }
   slot {class=VerticalBoxSlot}
  }
  SectionName {
   widgetClassPath=&TextBlock
   bIsVariable=true
   parent=Border_2
   widget=SectionName
   properties {
    text="           "
    colorAndOpacity {colorUseRule=UseColor_Specified, specifiedColor=<rgba 0 0 0 1>}
    font {
     bForceMonospaced=false
     bMaterialIsStencil=false
     fontMaterial=null
     fontObject=&Roboto
     letterSpacing=0
     monospacedWidth=1
     size=30
     skewAmount=0
     typefaceFontName=Bold
     outlineSettings {
      bApplyOutlineToDropShadows=false
      bMiteredCorners=false
      bSeparateFillAlpha=false
      outlineMaterial=null
      outlineSize=0
      outlineColor=<rgba 0 0 0 1>
     }
    }
   }
   slot {class=BorderSlot}
  }
  Border_3 {
   widgetClassPath=&Border
   parent=VerticalBox_1
   widget=Border_3
   properties {
    background {
     bIsDynamicallyLoaded=false
     drawAs=RoundedBox
     imageType=NoImage
     mirroring=NoMirror
     resourceName=None
     resourceObject=null
     tiling=NoTile
     imageSize=<32 32>
     margin=<ltrb 0 0 0 0>
     outlineSettings {
      bUseBrushTransparency=false
      roundingType=HalfHeightRadius
      width=0
      color {colorUseRule=UseColor_Specified, specifiedColor=<rgba 0 0 0 0>}
      cornerRadii=<0 0 0 0>
     }
     tintColor {colorUseRule=UseColor_Specified, specifiedColor=<rgba 1 1 1 1>}
     uVRegion {bIsValid=false, max=<0 0>, min=<0 0>}
    }
    brushColor=<rgba 1 1 1 0.5>
   }
   slot {class=VerticalBoxSlot}
  }
  Speed {
   widgetClassPath=&TextBlock
   bIsVariable=true
   parent=Border_3
   widget=Speed
   properties {
    text=0
    colorAndOpacity {colorUseRule=UseColor_Specified, specifiedColor=<rgba 0 0 0 1>}
    font {
     bForceMonospaced=false
     bMaterialIsStencil=false
     fontMaterial=null
     fontObject=&Roboto
     letterSpacing=0
     monospacedWidth=1
     size=30
     skewAmount=0
     typefaceFontName=Bold
     outlineSettings {
      bApplyOutlineToDropShadows=false
      bMiteredCorners=false
      bSeparateFillAlpha=false
      outlineMaterial=null
      outlineSize=0
      outlineColor=<rgba 0 0 0 1>
     }
    }
   }
   slot {class=BorderSlot}
  }
  WrapBox_35 {widgetClassPath=&WrapBox, parent=EditorUtilityScrollBox_0, widget=WrapBox_35, slot {class=ScrollBoxSlot}}
  EditorUtilityButton_73 {
   widgetClassPath=&EditorUtilityButton
   parent=WrapBox_35
   widget=EditorUtilityButton_73
   properties {widgetStyle {
     clickedSlateSound {resourceObject=null}
     disabled {
      bIsDynamicallyLoaded=false
      drawAs=RoundedBox
      imageType=NoImage
      mirroring=NoMirror
      resourceName=None
      resourceObject=null
      tiling=NoTile
      imageSize=<0 0>
      margin=<ltrb 0 0 0 0>
      outlineSettings {
       bUseBrushTransparency=false
       roundingType=FixedRadius
       width=1
       color {colorUseRule=UseColor_ColorTable, specifiedColor=<rgba 0 0 0 0>}
       cornerRadii=<4 4 4 4>
      }
      tintColor {colorUseRule=UseColor_Specified, specifiedColor=<rgba 0.0395 0.0395 0.0395 1>}
      uVRegion {bIsValid=false, max=<0 0>, min=<0 0>}
     }
     disabledForeground {colorUseRule=UseColor_ColorTable, specifiedColor=<rgba 0 0 0 0>}
     hovered {
      bIsDynamicallyLoaded=false
      drawAs=RoundedBox
      imageType=NoImage
      mirroring=NoMirror
      resourceName=None
      resourceObject=null
      tiling=NoTile
      imageSize=<32 32>
      margin=<ltrb 0 0 0 0>
      outlineSettings {
       bUseBrushTransparency=false
       roundingType=FixedRadius
       width=1
       color {colorUseRule=UseColor_ColorTable, specifiedColor=<rgba 0 0 0 0>}
       cornerRadii=<4 4 4 4>
      }
      tintColor {colorUseRule=UseColor_Specified, specifiedColor=<rgba 0.0953 0.0953 0.0953 1>}
      uVRegion {bIsValid=false, max=<0 0>, min=<0 0>}
     }
     hoveredForeground {colorUseRule=UseColor_ColorTable, specifiedColor=<rgba 0 0 0 0>}
     hoveredSlateSound {resourceObject=null}
     normal {
      bIsDynamicallyLoaded=false
      drawAs=RoundedBox
      imageType=NoImage
      mirroring=NoMirror
      resourceName=None
      resourceObject=null
      tiling=NoTile
      imageSize=<32 32>
      margin=<ltrb 0 0 0 0>
      outlineSettings {
       bUseBrushTransparency=false
       roundingType=FixedRadius
       width=1
       color {colorUseRule=UseColor_ColorTable, specifiedColor=<rgba 0 0 0 0>}
       cornerRadii=<4 4 4 4>
      }
      tintColor {colorUseRule=UseColor_Specified, specifiedColor=<rgba 0.0395 0.0395 0.0395 1>}
      uVRegion {bIsValid=false, max=<0 0>, min=<0 0>}
     }
     normalForeground {colorUseRule=UseColor_ColorTable, specifiedColor=<rgba 0 0 0 0>}
     normalPadding=<ltrb 3 3 3 3>
     pressed {
      bIsDynamicallyLoaded=false
      drawAs=RoundedBox
      imageType=NoImage
      mirroring=NoMirror
      resourceName=None
      resourceObject=null
      tiling=NoTile
      imageSize=<32 32>
      margin=<ltrb 0 0 0 0>
      outlineSettings {
       bUseBrushTransparency=false
       roundingType=FixedRadius
       width=1
       color {colorUseRule=UseColor_ColorTable, specifiedColor=<rgba 0 0 0 0>}
       cornerRadii=<4 4 4 4>
      }
      tintColor {colorUseRule=UseColor_Specified, specifiedColor=<rgba 0.0284 0.0284 0.0284 1>}
      uVRegion {bIsValid=false, max=<0 0>, min=<0 0>}
     }
     pressedForeground {colorUseRule=UseColor_ColorTable, specifiedColor=<rgba 0 0 0 0>}
     pressedPadding=<ltrb 3 3 3 3>
     pressedSlateSound {resourceObject=null}
    }}
   slot {class=WrapBoxSlot, properties {padding=<ltrb 0 0 5 0>}}
  }
  TextBlock_18 {
   widgetClassPath=&TextBlock
   parent=EditorUtilityButton_73
   widget=TextBlock_18
   properties {
    justification=Center
    text="↻"
    font {
     bForceMonospaced=false
     bMaterialIsStencil=false
     fontMaterial=null
     fontObject=&Roboto
     letterSpacing=0
     monospacedWidth=1
     size=22.5
     skewAmount=0
     typefaceFontName=Bold
     outlineSettings {
      bApplyOutlineToDropShadows=false
      bMiteredCorners=false
      bSeparateFillAlpha=false
      outlineMaterial=null
      outlineSize=1
      outlineColor=<rgba 1 1 1 1>
     }
    }
    renderTransform {angle=72, scale=<1 1>, shear=<0 0>, translation=<0 0>}
    renderTransformPivot=<0.5081 0.4757>
   }
   slot {class=ButtonSlot, properties {padding=<ltrb 0 0 0 -2>}}
  }
  SinglePropertyView_61 {
   widgetClassPath=&SinglePropertyView
   parent=WrapBox_35
   widget=SinglePropertyView_61
   properties {clipping=ClipToBounds, propertyName=Player Index, visibility=Collapsed}
   slot {class=WrapBoxSlot, properties {padding=<ltrb 10 0 -20 0>}}
  }
  TextBlock_438 {
   widgetClassPath=&TextBlock
   parent=WrapBox_35
   widget=TextBlock_438
   properties {
    text=Views Layout :
    font {
     bForceMonospaced=false
     bMaterialIsStencil=false
     fontMaterial=null
     fontObject=&Roboto
     letterSpacing=0
     monospacedWidth=1
     size=9
     skewAmount=0
     typefaceFontName=Bold
     outlineSettings {
      bApplyOutlineToDropShadows=false
      bMiteredCorners=false
      bSeparateFillAlpha=false
      outlineMaterial=null
      outlineSize=0
      outlineColor=<rgba 0 0 0 1>
     }
    }
   }
   slot {class=WrapBoxSlot, properties {verticalAlignment=VAlign_Center}}
  }
  ComboBox_ViewLayouts {
   widgetClassPath=/Script/Blutility.EditorUtilityComboBoxString
   bIsVariable=true
   parent=WrapBox_35
   widget=ComboBox_ViewLayouts
   properties {
    selectedOption=Grid
    defaultOptions [Grid, Full, Halfs, 1L 2R, 2L 1R]
    font {
     bForceMonospaced=false
     bMaterialIsStencil=false
     fontMaterial=null
     fontObject=&Roboto
     letterSpacing=0
     monospacedWidth=1
     size=9
     skewAmount=0
     typefaceFontName=Bold
     outlineSettings {
      bApplyOutlineToDropShadows=false
      bMiteredCorners=false
      bSeparateFillAlpha=false
      outlineMaterial=null
      outlineSize=0
      outlineColor=<rgba 0 0 0 1>
     }
    }
    foregroundColor {colorUseRule=UseColor_Specified, specifiedColor=<rgba 0.5271 0.5271 0.5271 1>}
   }
   slot {class=WrapBoxSlot, properties {verticalAlignment=VAlign_Center}}
  }
  TextBlock_15 {
   widgetClassPath=&TextBlock
   parent=WrapBox_35
   widget=TextBlock_15
   properties {
    text=Face Cam :
    font {
     bForceMonospaced=false
     bMaterialIsStencil=false
     fontMaterial=null
     fontObject=&Roboto
     letterSpacing=0
     monospacedWidth=1
     size=9
     skewAmount=0
     typefaceFontName=Bold
     outlineSettings {
      bApplyOutlineToDropShadows=false
      bMiteredCorners=false
      bSeparateFillAlpha=false
      outlineMaterial=null
      outlineSize=0
      outlineColor=<rgba 0 0 0 1>
     }
    }
   }
   slot {class=WrapBoxSlot, properties {verticalAlignment=VAlign_Center}}
  }
  ComboBox_FaceCam {
   widgetClassPath=&EditorUtilityComboBoxString
   bIsVariable=true
   parent=WrapBox_35
   widget=ComboBox_FaceCam
   properties {
    selectedOption=Small
    defaultOptions [Hidden,Small,Normal,Big]
    font {
     bForceMonospaced=false
     bMaterialIsStencil=false
     fontMaterial=null
     fontObject=&Roboto
     letterSpacing=0
     monospacedWidth=1
     size=9
     skewAmount=0
     typefaceFontName=Bold
     outlineSettings {
      bApplyOutlineToDropShadows=false
      bMiteredCorners=false
      bSeparateFillAlpha=false
      outlineMaterial=null
      outlineSize=0
      outlineColor=<rgba 0 0 0 1>
     }
    }
    foregroundColor {colorUseRule=UseColor_Specified, specifiedColor=<rgba 0.5271 0.5271 0.5271 1>}
   }
   slot {class=WrapBoxSlot, properties {verticalAlignment=VAlign_Center}}
  }
  TextBlock_16 {
   widgetClassPath=&TextBlock
   parent=WrapBox_35
   widget=TextBlock_16
   properties {
    text=Cam Distance :
    font {
     bForceMonospaced=false
     bMaterialIsStencil=false
     fontMaterial=null
     fontObject=&Roboto
     letterSpacing=0
     monospacedWidth=1
     size=9
     skewAmount=0
     typefaceFontName=Bold
     outlineSettings {
      bApplyOutlineToDropShadows=false
      bMiteredCorners=false
      bSeparateFillAlpha=false
      outlineMaterial=null
      outlineSize=0
      outlineColor=<rgba 0 0 0 1>
     }
    }
   }
   slot {class=WrapBoxSlot, properties {verticalAlignment=VAlign_Center}}
  }
  ComboBox_FaceCam_1 {
   widgetClassPath=&EditorUtilityComboBoxString
   bIsVariable=true
   parent=WrapBox_35
   widget=ComboBox_FaceCam_1
   properties {
    selectedOption=Normal
    defaultOptions [Close, Normal, Far, Very Far]
    font {
     bForceMonospaced=false
     bMaterialIsStencil=false
     fontMaterial=null
     fontObject=&Roboto
     letterSpacing=0
     monospacedWidth=1
     size=9
     skewAmount=0
     typefaceFontName=Bold
     outlineSettings {
      bApplyOutlineToDropShadows=false
      bMiteredCorners=false
      bSeparateFillAlpha=false
      outlineMaterial=null
      outlineSize=0
      outlineColor=<rgba 0 0 0 1>
     }
    }
    foregroundColor {colorUseRule=UseColor_Specified, specifiedColor=<rgba 0.5271 0.5271 0.5271 1>}
   }
   slot {class=WrapBoxSlot, properties {verticalAlignment=VAlign_Center}}
  }
  SinglePropertyView {
   widgetClassPath=&SinglePropertyView
   parent=WrapBox_35
   widget=SinglePropertyView
   properties {propertyName=EditViews}
   slot {class=WrapBoxSlot, properties {padding=<ltrb 0 10 0 10>}}
  }
  PropertyView_Montage_Visible {
   widgetClassPath=&SinglePropertyView
   parent=WrapBox_35
   widget=PropertyView_Montage_Visible
   properties {nameOverride=Show Montage Details, propertyName=MontageDetails_Visibility}
   slot {class=WrapBoxSlot}
  }
  HorizontalBox_2 {
   widgetClassPath=&HorizontalBox
   parent=WrapBox_35
   widget=HorizontalBox_2
   properties {visibility=Collapsed}
   slot {class=WrapBoxSlot, properties {padding=<ltrb -5 0 5 0>}}
  }
  TextBlock_25 {
   widgetClassPath=&TextBlock
   parent=HorizontalBox_2
   widget=TextBlock_25
   properties {
    text=Show Views
    font {
     bForceMonospaced=false
     bMaterialIsStencil=false
     fontMaterial=null
     fontObject=&Roboto
     letterSpacing=0
     monospacedWidth=1
     size=10.5
     skewAmount=0
     typefaceFontName=Bold
     outlineSettings {
      bApplyOutlineToDropShadows=false
      bMiteredCorners=false
      bSeparateFillAlpha=false
      outlineMaterial=null
      outlineSize=0
      outlineColor=<rgba 0 0 0 1>
     }
    }
   }
   slot {class=HorizontalBoxSlot, properties {verticalAlignment=VAlign_Center}}
  }
  Checkbox_ShowEIDs_1 {
   widgetClassPath=&EditorUtilityCheckBox
   bIsVariable=true
   parent=HorizontalBox_2
   widget=Checkbox_ShowEIDs_1
   properties {widgetStyle {
     checkBoxType=CheckBox
     backgroundHoveredImage {
      bIsDynamicallyLoaded=false
      drawAs=RoundedBox
      imageType=NoImage
      mirroring=NoMirror
      resourceName=None
      resourceObject=null
      tiling=NoTile
      imageSize=<18 18>
      margin=<ltrb 0 0 0 0>
      outlineSettings {
       bUseBrushTransparency=false
       roundingType=FixedRadius
       width=1
       color {colorUseRule=UseColor_ColorTable, specifiedColor=<rgba 0 0 0 0>}
       cornerRadii=<3 3 3 3>
      }
      tintColor {colorUseRule=UseColor_Specified, specifiedColor=<rgba 0.0048 0.0048 0.0048 1>}
      uVRegion {bIsValid=false, max=<0 0>, min=<0 0>}
     }
     backgroundImage {
      bIsDynamicallyLoaded=false
      drawAs=RoundedBox
      imageType=NoImage
      mirroring=NoMirror
      resourceName=None
      resourceObject=null
      tiling=NoTile
      imageSize=<18 18>
      margin=<ltrb 0 0 0 0>
      outlineSettings {
       bUseBrushTransparency=false
       roundingType=FixedRadius
       width=1
       color {colorUseRule=UseColor_ColorTable, specifiedColor=<rgba 0 0 0 0>}
       cornerRadii=<3 3 3 3>
      }
      tintColor {colorUseRule=UseColor_Specified, specifiedColor=<rgba 0.0048 0.0048 0.0048 1>}
      uVRegion {bIsValid=false, max=<0 0>, min=<0 0>}
     }
     backgroundPressedImage {
      bIsDynamicallyLoaded=false
      drawAs=RoundedBox
      imageType=NoImage
      mirroring=NoMirror
      resourceName=None
      resourceObject=null
      tiling=NoTile
      imageSize=<18 18>
      margin=<ltrb 0 0 0 0>
      outlineSettings {
       bUseBrushTransparency=false
       roundingType=FixedRadius
       width=1
       color {colorUseRule=UseColor_ColorTable, specifiedColor=<rgba 0 0 0 0>}
       cornerRadii=<3 3 3 3>
      }
      tintColor {colorUseRule=UseColor_Specified, specifiedColor=<rgba 0.0048 0.0048 0.0048 1>}
      uVRegion {bIsValid=false, max=<0 0>, min=<0 0>}
     }
     borderBackgroundColor {colorUseRule=UseColor_Specified, specifiedColor=<rgba 1 1 1 1>}
     checkedForeground {colorUseRule=UseColor_Specified, specifiedColor=<rgba 1 1 1 1>}
     checkedHoveredForeground {colorUseRule=UseColor_Specified, specifiedColor=<rgba 1 1 1 1>}
     checkedHoveredImage {
      bIsDynamicallyLoaded=false
      drawAs=Image
      imageType=Vector
      mirroring=NoMirror
      resourceName=../../../Engine/Content/Slate/Starship/CoreWidgets/CheckBox/check.svg
      resourceObject=null
      tiling=NoTile
      imageSize=<16 16>
      margin=<ltrb 0 0 0 0>
      outlineSettings {
       bUseBrushTransparency=false
       roundingType=HalfHeightRadius
       width=0
       color {colorUseRule=UseColor_Specified, specifiedColor=<rgba 0 0 0 0>}
       cornerRadii=<0 0 0 0>
      }
      tintColor {colorUseRule=UseColor_Specified, specifiedColor=<rgba 0.0044 0.2384 1 1>}
      uVRegion {bIsValid=false, max=<0 0>, min=<0 0>}
     }
     checkedImage {
      bIsDynamicallyLoaded=false
      drawAs=Image
      imageType=Vector
      mirroring=NoMirror
      resourceName=../../../Engine/Content/Slate/Starship/CoreWidgets/CheckBox/check.svg
      resourceObject=null
      tiling=NoTile
      imageSize=<16 16>
      margin=<ltrb 0 0 0 0>
      outlineSettings {
       bUseBrushTransparency=false
       roundingType=HalfHeightRadius
       width=0
       color {colorUseRule=UseColor_Specified, specifiedColor=<rgba 0 0 0 0>}
       cornerRadii=<0 0 0 0>
      }
      tintColor {colorUseRule=UseColor_Specified, specifiedColor=<rgba 0 0.162 0.7454 1>}
      uVRegion {bIsValid=false, max=<0 0>, min=<0 0>}
     }
     checkedPressedForeground {colorUseRule=UseColor_Specified, specifiedColor=<rgba 1 1 1 1>}
     checkedPressedImage {
      bIsDynamicallyLoaded=false
      drawAs=Image
      imageType=Vector
      mirroring=NoMirror
      resourceName=../../../Engine/Content/Slate/Starship/CoreWidgets/CheckBox/check.svg
      resourceObject=null
      tiling=NoTile
      imageSize=<16 16>
      margin=<ltrb 0 0 0 0>
      outlineSettings {
       bUseBrushTransparency=false
       roundingType=HalfHeightRadius
       width=0
       color {colorUseRule=UseColor_Specified, specifiedColor=<rgba 0 0 0 0>}
       cornerRadii=<0 0 0 0>
      }
      tintColor {colorUseRule=UseColor_Specified, specifiedColor=<rgba 0 0.0802 0.3515 1>}
      uVRegion {bIsValid=false, max=<0 0>, min=<0 0>}
     }
     checkedSlateSound {resourceObject=null}
     foregroundColor {colorUseRule=UseColor_Specified, specifiedColor=<rgba 1 1 1 1>}
     hoveredForeground {colorUseRule=UseColor_Specified, specifiedColor=<rgba 1 1 1 1>}
     hoveredSlateSound {resourceObject=null}
     padding=<ltrb 2 0 0 0>
     pressedForeground {colorUseRule=UseColor_Specified, specifiedColor=<rgba 1 1 1 1>}
     uncheckedHoveredImage {
      bIsDynamicallyLoaded=false
      drawAs=NoDrawType
      imageType=NoImage
      mirroring=NoMirror
      resourceName=None
      resourceObject=null
      tiling=NoTile
      imageSize=<0 0>
      margin=<ltrb 0 0 0 0>
      outlineSettings {
       bUseBrushTransparency=false
       roundingType=HalfHeightRadius
       width=0
       color {colorUseRule=UseColor_Specified, specifiedColor=<rgba 0 0 0 0>}
       cornerRadii=<0 0 0 0>
      }
      tintColor {colorUseRule=UseColor_Specified, specifiedColor=<rgba 1 1 1 1>}
      uVRegion {bIsValid=false, max=<0 0>, min=<0 0>}
     }
     uncheckedImage {
      bIsDynamicallyLoaded=false
      drawAs=NoDrawType
      imageType=NoImage
      mirroring=NoMirror
      resourceName=None
      resourceObject=null
      tiling=NoTile
      imageSize=<0 0>
      margin=<ltrb 0 0 0 0>
      outlineSettings {
       bUseBrushTransparency=false
       roundingType=HalfHeightRadius
       width=0
       color {colorUseRule=UseColor_Specified, specifiedColor=<rgba 0 0 0 0>}
       cornerRadii=<0 0 0 0>
      }
      tintColor {colorUseRule=UseColor_Specified, specifiedColor=<rgba 1 1 1 1>}
      uVRegion {bIsValid=false, max=<0 0>, min=<0 0>}
     }
     uncheckedPressedImage {
      bIsDynamicallyLoaded=false
      drawAs=NoDrawType
      imageType=NoImage
      mirroring=NoMirror
      resourceName=None
      resourceObject=null
      tiling=NoTile
      imageSize=<0 0>
      margin=<ltrb 0 0 0 0>
      outlineSettings {
       bUseBrushTransparency=false
       roundingType=HalfHeightRadius
       width=0
       color {colorUseRule=UseColor_Specified, specifiedColor=<rgba 0 0 0 0>}
       cornerRadii=<0 0 0 0>
      }
      tintColor {colorUseRule=UseColor_Specified, specifiedColor=<rgba 1 1 1 1>}
      uVRegion {bIsValid=false, max=<0 0>, min=<0 0>}
     }
     uncheckedSlateSound {resourceObject=null}
     undeterminedForeground {colorUseRule=UseColor_Specified, specifiedColor=<rgba 1 1 1 1>}
     undeterminedHoveredImage {
      bIsDynamicallyLoaded=false
      drawAs=Image
      imageType=Vector
      mirroring=NoMirror
      resourceName=../../../Engine/Content/Slate/Starship/CoreWidgets/CheckBox/indeterminate.svg
      resourceObject=null
      tiling=NoTile
      imageSize=<16 16>
      margin=<ltrb 0 0 0 0>
      outlineSettings {
       bUseBrushTransparency=false
       roundingType=HalfHeightRadius
       width=0
       color {colorUseRule=UseColor_Specified, specifiedColor=<rgba 0 0 0 0>}
       cornerRadii=<0 0 0 0>
      }
      tintColor {colorUseRule=UseColor_Specified, specifiedColor=<rgba 0.0044 0.2384 1 1>}
      uVRegion {bIsValid=false, max=<0 0>, min=<0 0>}
     }
     undeterminedImage {
      bIsDynamicallyLoaded=false
      drawAs=Image
      imageType=Vector
      mirroring=NoMirror
      resourceName=../../../Engine/Content/Slate/Starship/CoreWidgets/CheckBox/indeterminate.svg
      resourceObject=null
      tiling=NoTile
      imageSize=<16 16>
      margin=<ltrb 0 0 0 0>
      outlineSettings {
       bUseBrushTransparency=false
       roundingType=HalfHeightRadius
       width=0
       color {colorUseRule=UseColor_Specified, specifiedColor=<rgba 0 0 0 0>}
       cornerRadii=<0 0 0 0>
      }
      tintColor {colorUseRule=UseColor_Specified, specifiedColor=<rgba 0 0.162 0.7454 1>}
      uVRegion {bIsValid=false, max=<0 0>, min=<0 0>}
     }
     undeterminedPressedImage {
      bIsDynamicallyLoaded=false
      drawAs=Image
      imageType=Vector
      mirroring=NoMirror
      resourceName=../../../Engine/Content/Slate/Starship/CoreWidgets/CheckBox/indeterminate.svg
      resourceObject=null
      tiling=NoTile
      imageSize=<16 16>
      margin=<ltrb 0 0 0 0>
      outlineSettings {
       bUseBrushTransparency=false
       roundingType=HalfHeightRadius
       width=0
       color {colorUseRule=UseColor_Specified, specifiedColor=<rgba 0 0 0 0>}
       cornerRadii=<0 0 0 0>
      }
      tintColor {colorUseRule=UseColor_Specified, specifiedColor=<rgba 0 0.0802 0.3515 1>}
      uVRegion {bIsValid=false, max=<0 0>, min=<0 0>}
     }
    }}
   slot {class=HorizontalBoxSlot, properties {verticalAlignment=VAlign_Center, padding=<ltrb 3 0 0 0>}}
  }
  Button_StopEID_1 {widgetClassPath=&EditorUtilityButton, parent=WrapBox_35, widget=Button_StopEID_1, slot {class=WrapBoxSlot}}
  TextBlock_24 {
   widgetClassPath=&TextBlock
   parent=Button_StopEID_1
   widget=TextBlock_24
   properties {
    justification=Center
    text=Detach
    font {
     bForceMonospaced=false
     bMaterialIsStencil=false
     fontMaterial=null
     fontObject=&Roboto
     letterSpacing=0
     monospacedWidth=1
     size=12
     skewAmount=0
     typefaceFontName=Bold
     outlineSettings {
      bApplyOutlineToDropShadows=false
      bMiteredCorners=false
      bSeparateFillAlpha=false
      outlineMaterial=null
      outlineSize=0
      outlineColor=<rgba 0 0 0 1>
     }
    }
   }
   slot {class=ButtonSlot, properties {padding=<ltrb 0 0 0 0>}}
  }
  VerticalBox_2 {
   widgetClassPath=&VerticalBox
   parent=EditorUtilityScrollBox_0
   widget=VerticalBox_2
   properties {bOverride_Cursor=true}
   slot {class=ScrollBoxSlot}
  }
  HorizontalBox_8 {
   widgetClassPath=&HorizontalBox
   parent=VerticalBox_2
   widget=HorizontalBox_8
   slot {class=VerticalBoxSlot, properties {padding=<ltrb 0 5 0 0>}}
  }
  TextBlock_221 {
   widgetClassPath=&TextBlock
   parent=HorizontalBox_8
   widget=TextBlock_221
   properties {
    text="EIDs ▼"
    font {
     bForceMonospaced=false
     bMaterialIsStencil=false
     fontMaterial=null
     fontObject=&Roboto
     letterSpacing=0
     monospacedWidth=1
     size=10.5
     skewAmount=0
     typefaceFontName=Bold
     outlineSettings {
      bApplyOutlineToDropShadows=false
      bMiteredCorners=false
      bSeparateFillAlpha=false
      outlineMaterial=null
      outlineSize=0
      outlineColor=<rgba 0 0 0 1>
     }
    }
   }
   slot {class=HorizontalBoxSlot}
  }
  Checkbox_ShowEIDs {
   widgetClassPath=&EditorUtilityCheckBox
   bIsVariable=true
   parent=HorizontalBox_8
   widget=Checkbox_ShowEIDs
   properties {widgetStyle {
     checkBoxType=CheckBox
     backgroundHoveredImage {
      bIsDynamicallyLoaded=false
      drawAs=RoundedBox
      imageType=NoImage
      mirroring=NoMirror
      resourceName=None
      resourceObject=null
      tiling=NoTile
      imageSize=<18 18>
      margin=<ltrb 0 0 0 0>
      outlineSettings {
       bUseBrushTransparency=false
       roundingType=FixedRadius
       width=1
       color {colorUseRule=UseColor_ColorTable, specifiedColor=<rgba 0 0 0 0>}
       cornerRadii=<3 3 3 3>
      }
      tintColor {colorUseRule=UseColor_Specified, specifiedColor=<rgba 0.0048 0.0048 0.0048 1>}
      uVRegion {bIsValid=false, max=<0 0>, min=<0 0>}
     }
     backgroundImage {
      bIsDynamicallyLoaded=false
      drawAs=RoundedBox
      imageType=NoImage
      mirroring=NoMirror
      resourceName=None
      resourceObject=null
      tiling=NoTile
      imageSize=<18 18>
      margin=<ltrb 0 0 0 0>
      outlineSettings {
       bUseBrushTransparency=false
       roundingType=FixedRadius
       width=1
       color {colorUseRule=UseColor_ColorTable, specifiedColor=<rgba 0 0 0 0>}
       cornerRadii=<3 3 3 3>
      }
      tintColor {colorUseRule=UseColor_Specified, specifiedColor=<rgba 0.0048 0.0048 0.0048 1>}
      uVRegion {bIsValid=false, max=<0 0>, min=<0 0>}
     }
     backgroundPressedImage {
      bIsDynamicallyLoaded=false
      drawAs=RoundedBox
      imageType=NoImage
      mirroring=NoMirror
      resourceName=None
      resourceObject=null
      tiling=NoTile
      imageSize=<18 18>
      margin=<ltrb 0 0 0 0>
      outlineSettings {
       bUseBrushTransparency=false
       roundingType=FixedRadius
       width=1
       color {colorUseRule=UseColor_ColorTable, specifiedColor=<rgba 0 0 0 0>}
       cornerRadii=<3 3 3 3>
      }
      tintColor {colorUseRule=UseColor_Specified, specifiedColor=<rgba 0.0048 0.0048 0.0048 1>}
      uVRegion {bIsValid=false, max=<0 0>, min=<0 0>}
     }
     borderBackgroundColor {colorUseRule=UseColor_Specified, specifiedColor=<rgba 1 1 1 1>}
     checkedForeground {colorUseRule=UseColor_Specified, specifiedColor=<rgba 1 1 1 1>}
     checkedHoveredForeground {colorUseRule=UseColor_Specified, specifiedColor=<rgba 1 1 1 1>}
     checkedHoveredImage {
      bIsDynamicallyLoaded=false
      drawAs=Image
      imageType=Vector
      mirroring=NoMirror
      resourceName=../../../Engine/Content/Slate/Starship/CoreWidgets/CheckBox/check.svg
      resourceObject=null
      tiling=NoTile
      imageSize=<16 16>
      margin=<ltrb 0 0 0 0>
      outlineSettings {
       bUseBrushTransparency=false
       roundingType=HalfHeightRadius
       width=0
       color {colorUseRule=UseColor_Specified, specifiedColor=<rgba 0 0 0 0>}
       cornerRadii=<0 0 0 0>
      }
      tintColor {colorUseRule=UseColor_Specified, specifiedColor=<rgba 0.0044 0.2384 1 1>}
      uVRegion {bIsValid=false, max=<0 0>, min=<0 0>}
     }
     checkedImage {
      bIsDynamicallyLoaded=false
      drawAs=Image
      imageType=Vector
      mirroring=NoMirror
      resourceName=../../../Engine/Content/Slate/Starship/CoreWidgets/CheckBox/check.svg
      resourceObject=null
      tiling=NoTile
      imageSize=<16 16>
      margin=<ltrb 0 0 0 0>
      outlineSettings {
       bUseBrushTransparency=false
       roundingType=HalfHeightRadius
       width=0
       color {colorUseRule=UseColor_Specified, specifiedColor=<rgba 0 0 0 0>}
       cornerRadii=<0 0 0 0>
      }
      tintColor {colorUseRule=UseColor_Specified, specifiedColor=<rgba 0 0.162 0.7454 1>}
      uVRegion {bIsValid=false, max=<0 0>, min=<0 0>}
     }
     checkedPressedForeground {colorUseRule=UseColor_Specified, specifiedColor=<rgba 1 1 1 1>}
     checkedPressedImage {
      bIsDynamicallyLoaded=false
      drawAs=Image
      imageType=Vector
      mirroring=NoMirror
      resourceName=../../../Engine/Content/Slate/Starship/CoreWidgets/CheckBox/check.svg
      resourceObject=null
      tiling=NoTile
      imageSize=<16 16>
      margin=<ltrb 0 0 0 0>
      outlineSettings {
       bUseBrushTransparency=false
       roundingType=HalfHeightRadius
       width=0
       color {colorUseRule=UseColor_Specified, specifiedColor=<rgba 0 0 0 0>}
       cornerRadii=<0 0 0 0>
      }
      tintColor {colorUseRule=UseColor_Specified, specifiedColor=<rgba 0 0.0802 0.3515 1>}
      uVRegion {bIsValid=false, max=<0 0>, min=<0 0>}
     }
     checkedSlateSound {resourceObject=null}
     foregroundColor {colorUseRule=UseColor_Specified, specifiedColor=<rgba 1 1 1 1>}
     hoveredForeground {colorUseRule=UseColor_Specified, specifiedColor=<rgba 1 1 1 1>}
     hoveredSlateSound {resourceObject=null}
     padding=<ltrb 2 0 0 0>
     pressedForeground {colorUseRule=UseColor_Specified, specifiedColor=<rgba 1 1 1 1>}
     uncheckedHoveredImage {
      bIsDynamicallyLoaded=false
      drawAs=NoDrawType
      imageType=NoImage
      mirroring=NoMirror
      resourceName=None
      resourceObject=null
      tiling=NoTile
      imageSize=<0 0>
      margin=<ltrb 0 0 0 0>
      outlineSettings {
       bUseBrushTransparency=false
       roundingType=HalfHeightRadius
       width=0
       color {colorUseRule=UseColor_Specified, specifiedColor=<rgba 0 0 0 0>}
       cornerRadii=<0 0 0 0>
      }
      tintColor {colorUseRule=UseColor_Specified, specifiedColor=<rgba 1 1 1 1>}
      uVRegion {bIsValid=false, max=<0 0>, min=<0 0>}
     }
     uncheckedImage {
      bIsDynamicallyLoaded=false
      drawAs=NoDrawType
      imageType=NoImage
      mirroring=NoMirror
      resourceName=None
      resourceObject=null
      tiling=NoTile
      imageSize=<0 0>
      margin=<ltrb 0 0 0 0>
      outlineSettings {
       bUseBrushTransparency=false
       roundingType=HalfHeightRadius
       width=0
       color {colorUseRule=UseColor_Specified, specifiedColor=<rgba 0 0 0 0>}
       cornerRadii=<0 0 0 0>
      }
      tintColor {colorUseRule=UseColor_Specified, specifiedColor=<rgba 1 1 1 1>}
      uVRegion {bIsValid=false, max=<0 0>, min=<0 0>}
     }
     uncheckedPressedImage {
      bIsDynamicallyLoaded=false
      drawAs=NoDrawType
      imageType=NoImage
      mirroring=NoMirror
      resourceName=None
      resourceObject=null
      tiling=NoTile
      imageSize=<0 0>
      margin=<ltrb 0 0 0 0>
      outlineSettings {
       bUseBrushTransparency=false
       roundingType=HalfHeightRadius
       width=0
       color {colorUseRule=UseColor_Specified, specifiedColor=<rgba 0 0 0 0>}
       cornerRadii=<0 0 0 0>
      }
      tintColor {colorUseRule=UseColor_Specified, specifiedColor=<rgba 1 1 1 1>}
      uVRegion {bIsValid=false, max=<0 0>, min=<0 0>}
     }
     uncheckedSlateSound {resourceObject=null}
     undeterminedForeground {colorUseRule=UseColor_Specified, specifiedColor=<rgba 1 1 1 1>}
     undeterminedHoveredImage {
      bIsDynamicallyLoaded=false
      drawAs=Image
      imageType=Vector
      mirroring=NoMirror
      resourceName=../../../Engine/Content/Slate/Starship/CoreWidgets/CheckBox/indeterminate.svg
      resourceObject=null
      tiling=NoTile
      imageSize=<16 16>
      margin=<ltrb 0 0 0 0>
      outlineSettings {
       bUseBrushTransparency=false
       roundingType=HalfHeightRadius
       width=0
       color {colorUseRule=UseColor_Specified, specifiedColor=<rgba 0 0 0 0>}
       cornerRadii=<0 0 0 0>
      }
      tintColor {colorUseRule=UseColor_Specified, specifiedColor=<rgba 0.0044 0.2384 1 1>}
      uVRegion {bIsValid=false, max=<0 0>, min=<0 0>}
     }
     undeterminedImage {
      bIsDynamicallyLoaded=false
      drawAs=Image
      imageType=Vector
      mirroring=NoMirror
      resourceName=../../../Engine/Content/Slate/Starship/CoreWidgets/CheckBox/indeterminate.svg
      resourceObject=null
      tiling=NoTile
      imageSize=<16 16>
      margin=<ltrb 0 0 0 0>
      outlineSettings {
       bUseBrushTransparency=false
       roundingType=HalfHeightRadius
       width=0
       color {colorUseRule=UseColor_Specified, specifiedColor=<rgba 0 0 0 0>}
       cornerRadii=<0 0 0 0>
      }
      tintColor {colorUseRule=UseColor_Specified, specifiedColor=<rgba 0 0.162 0.7454 1>}
      uVRegion {bIsValid=false, max=<0 0>, min=<0 0>}
     }
     undeterminedPressedImage {
      bIsDynamicallyLoaded=false
      drawAs=Image
      imageType=Vector
      mirroring=NoMirror
      resourceName=../../../Engine/Content/Slate/Starship/CoreWidgets/CheckBox/indeterminate.svg
      resourceObject=null
      tiling=NoTile
      imageSize=<16 16>
      margin=<ltrb 0 0 0 0>
      outlineSettings {
       bUseBrushTransparency=false
       roundingType=HalfHeightRadius
       width=0
       color {colorUseRule=UseColor_Specified, specifiedColor=<rgba 0 0 0 0>}
       cornerRadii=<0 0 0 0>
      }
      tintColor {colorUseRule=UseColor_Specified, specifiedColor=<rgba 0 0.0802 0.3515 1>}
      uVRegion {bIsValid=false, max=<0 0>, min=<0 0>}
     }
    }}
   slot {class=HorizontalBoxSlot, properties {padding=<ltrb 3 0 0 0>}}
  }
  WrapBoxEID {
   widgetClassPath=&WrapBox
   bIsVariable=true
   parent=VerticalBox_2
   widget=WrapBoxEID
   slot {class=VerticalBoxSlot, properties {verticalAlignment=VAlign_Top}}
  }
  HorizontalBox_11 {
   widgetClassPath=&HorizontalBox
   parent=WrapBoxEID
   widget=HorizontalBox_11
   slot {class=WrapBoxSlot, properties {padding=<ltrb 0 0 10 0>}}
  }
  Button_StopEID {widgetClassPath=&EditorUtilityButton, parent=HorizontalBox_11, widget=Button_StopEID, slot {class=HorizontalBoxSlot}}
  TextBlock {
   widgetClassPath=&TextBlock
   parent=Button_StopEID
   widget=TextBlock
   properties {
    justification=Center
    text="Stop\r\nEmote"
    font {
     bForceMonospaced=false
     bMaterialIsStencil=false
     fontMaterial=null
     fontObject=&Roboto
     letterSpacing=0
     monospacedWidth=1
     size=12
     skewAmount=0
     typefaceFontName=Bold
     outlineSettings {
      bApplyOutlineToDropShadows=false
      bMiteredCorners=false
      bSeparateFillAlpha=false
      outlineMaterial=null
      outlineSize=0
      outlineColor=<rgba 0 0 0 1>
     }
    }
   }
   slot {class=ButtonSlot, properties {padding=<ltrb 0 0 0 0>}}
  }
  HorizontalBox_9 {widgetClassPath=&HorizontalBox, parent=HorizontalBox_11, widget=HorizontalBox_9, slot {class=HorizontalBoxSlot}}
  Border_4 {
   widgetClassPath=&Border
   parent=HorizontalBox_9
   widget=Border_4
   properties {
    background {
     bIsDynamicallyLoaded=false
     drawAs=RoundedBox
     imageType=NoImage
     mirroring=NoMirror
     resourceName=None
     resourceObject=null
     tiling=NoTile
     imageSize=<32 32>
     margin=<ltrb 0 0 0 0>
     outlineSettings {
      bUseBrushTransparency=false
      roundingType=FixedRadius
      width=0
      color {colorUseRule=UseColor_Specified, specifiedColor=<rgba 0 0 0 0>}
      cornerRadii=<4 4 4 4>
     }
     tintColor {colorUseRule=UseColor_Specified, specifiedColor=<rgba 0.0469 0.0469 0.0469 1>}
     uVRegion {bIsValid=false, max=<0 0>, min=<0 0>}
    }
    padding=<ltrb 2 0 2 0>
   }
   slot {class=HorizontalBoxSlot, properties {horizontalAlignment=HAlign_Center, padding=<ltrb 2 1.5 2 1.5>}}
  }
  VerticalBox_5 {widgetClassPath=&VerticalBox, parent=Border_4, widget=VerticalBox_5, slot {class=BorderSlot, properties {padding=<ltrb 2 0 2 0>}}}
  TextBlock_19 {
   widgetClassPath=&TextBlock
   parent=VerticalBox_5
   widget=TextBlock_19
   properties {
    justification=Center
    text=Turn
    font {
     bForceMonospaced=false
     bMaterialIsStencil=false
     fontMaterial=null
     fontObject=&Roboto
     letterSpacing=0
     monospacedWidth=1
     size=9
     skewAmount=0
     typefaceFontName=Bold
     outlineSettings {
      bApplyOutlineToDropShadows=false
      bMiteredCorners=false
      bSeparateFillAlpha=false
      outlineMaterial=null
      outlineSize=0
      outlineColor=<rgba 0 0 0 1>
     }
    }
   }
   slot {class=VerticalBoxSlot, properties {padding=<ltrb 0 5 0 0>}}
  }
  EditorUtilityButton_3 {
   widgetClassPath=&EditorUtilityButton
   parent=VerticalBox_5
   widget=EditorUtilityButton_3
   properties {widgetStyle {
     clickedSlateSound {resourceObject=null}
     disabled {
      bIsDynamicallyLoaded=false
      drawAs=RoundedBox
      imageType=NoImage
      mirroring=NoMirror
      resourceName=None
      resourceObject=null
      tiling=NoTile
      imageSize=<0 0>
      margin=<ltrb 0 0 0 0>
      outlineSettings {
       bUseBrushTransparency=false
       roundingType=FixedRadius
       width=1
       color {colorUseRule=UseColor_ColorTable, specifiedColor=<rgba 0 0 0 0>}
       cornerRadii=<4 4 4 4>
      }
      tintColor {colorUseRule=UseColor_Specified, specifiedColor=<rgba 0.0395 0.0395 0.0395 1>}
      uVRegion {bIsValid=false, max=<0 0>, min=<0 0>}
     }
     disabledForeground {colorUseRule=UseColor_ColorTable, specifiedColor=<rgba 0 0 0 0>}
     hovered {
      bIsDynamicallyLoaded=false
      drawAs=RoundedBox
      imageType=NoImage
      mirroring=NoMirror
      resourceName=None
      resourceObject=null
      tiling=NoTile
      imageSize=<32 32>
      margin=<ltrb 0 0 0 0>
      outlineSettings {
       bUseBrushTransparency=false
       roundingType=FixedRadius
       width=1
       color {colorUseRule=UseColor_ColorTable, specifiedColor=<rgba 0 0 0 0>}
       cornerRadii=<4 4 4 4>
      }
      tintColor {colorUseRule=UseColor_Specified, specifiedColor=<rgba 0.0953 0.0953 0.0953 1>}
      uVRegion {bIsValid=false, max=<0 0>, min=<0 0>}
     }
     hoveredForeground {colorUseRule=UseColor_ColorTable, specifiedColor=<rgba 0 0 0 0>}
     hoveredSlateSound {resourceObject=null}
     normal {
      bIsDynamicallyLoaded=false
      drawAs=RoundedBox
      imageType=NoImage
      mirroring=NoMirror
      resourceName=None
      resourceObject=null
      tiling=NoTile
      imageSize=<32 32>
      margin=<ltrb 0 0 0 0>
      outlineSettings {
       bUseBrushTransparency=false
       roundingType=FixedRadius
       width=1
       color {colorUseRule=UseColor_ColorTable, specifiedColor=<rgba 0 0 0 0>}
       cornerRadii=<4 4 4 4>
      }
      tintColor {colorUseRule=UseColor_Specified, specifiedColor=<rgba 0.0395 0.0395 0.0395 1>}
      uVRegion {bIsValid=false, max=<0 0>, min=<0 0>}
     }
     normalForeground {colorUseRule=UseColor_ColorTable, specifiedColor=<rgba 0 0 0 0>}
     normalPadding=<ltrb 0 0 0 0>
     pressed {
      bIsDynamicallyLoaded=false
      drawAs=RoundedBox
      imageType=NoImage
      mirroring=NoMirror
      resourceName=None
      resourceObject=null
      tiling=NoTile
      imageSize=<32 32>
      margin=<ltrb 0 0 0 0>
      outlineSettings {
       bUseBrushTransparency=false
       roundingType=FixedRadius
       width=1
       color {colorUseRule=UseColor_ColorTable, specifiedColor=<rgba 0 0 0 0>}
       cornerRadii=<4 4 4 4>
      }
      tintColor {colorUseRule=UseColor_Specified, specifiedColor=<rgba 0.0284 0.0284 0.0284 1>}
      uVRegion {bIsValid=false, max=<0 0>, min=<0 0>}
     }
     pressedForeground {colorUseRule=UseColor_ColorTable, specifiedColor=<rgba 0 0 0 0>}
     pressedPadding=<ltrb 0 0 0 0>
     pressedSlateSound {resourceObject=null}
    }}
   slot {class=VerticalBoxSlot}
  }
  VerticalBox_3 {
   widgetClassPath=&VerticalBox
   parent=EditorUtilityButton_3
   widget=VerticalBox_3
   slot {class=ButtonSlot, properties {padding=<ltrb 7 7 7 7>}}
  }
  Image_4 {
   widgetClassPath=&Image
   bIsVariable=true
   parent=VerticalBox_3
   widget=Image_4
   properties {
    brush {
     bIsDynamicallyLoaded=false
     drawAs=Image
     imageType=FullColor
     mirroring=NoMirror
     resourceName=None
     resourceObject=/Game/Creative/Devices/DirectionalDecals/Textures/Directional_UTurnArrow_Decal
     tiling=NoTile
     imageSize=<20 20>
     margin=<ltrb 0 0 0 0>
     outlineSettings {
      bUseBrushTransparency=false
      roundingType=HalfHeightRadius
      width=0
      color {colorUseRule=UseColor_Specified, specifiedColor=<rgba 0 0 0 0>}
      cornerRadii=<0 0 0 0>
     }
     tintColor {colorUseRule=UseColor_Specified, specifiedColor=<rgba 1 1 1 1>}
     uVRegion {bIsValid=false, max=<0 0>, min=<0 0>}
    }
    renderTransform {angle=-151, scale=<1 1>, shear=<0 0>, translation=<0 0>}
   }
   slot {class=VerticalBoxSlot, properties {horizontalAlignment=HAlign_Center, verticalAlignment=VAlign_Center}}
  }
  EditorUtilityButton_5 {
   widgetClassPath=&EditorUtilityButton
   parent=VerticalBox_5
   widget=EditorUtilityButton_5
   properties {widgetStyle {
     clickedSlateSound {resourceObject=null}
     disabled {
      bIsDynamicallyLoaded=false
      drawAs=RoundedBox
      imageType=NoImage
      mirroring=NoMirror
      resourceName=None
      resourceObject=null
      tiling=NoTile
      imageSize=<0 0>
      margin=<ltrb 0 0 0 0>
      outlineSettings {
       bUseBrushTransparency=false
       roundingType=FixedRadius
       width=1
       color {colorUseRule=UseColor_ColorTable, specifiedColor=<rgba 0 0 0 0>}
       cornerRadii=<4 4 4 4>
      }
      tintColor {colorUseRule=UseColor_Specified, specifiedColor=<rgba 0.0395 0.0395 0.0395 1>}
      uVRegion {bIsValid=false, max=<0 0>, min=<0 0>}
     }
     disabledForeground {colorUseRule=UseColor_ColorTable, specifiedColor=<rgba 0 0 0 0>}
     hovered {
      bIsDynamicallyLoaded=false
      drawAs=RoundedBox
      imageType=NoImage
      mirroring=NoMirror
      resourceName=None
      resourceObject=null
      tiling=NoTile
      imageSize=<32 32>
      margin=<ltrb 0 0 0 0>
      outlineSettings {
       bUseBrushTransparency=false
       roundingType=FixedRadius
       width=1
       color {colorUseRule=UseColor_ColorTable, specifiedColor=<rgba 0 0 0 0>}
       cornerRadii=<4 4 4 4>
      }
      tintColor {colorUseRule=UseColor_Specified, specifiedColor=<rgba 0.0953 0.0953 0.0953 1>}
      uVRegion {bIsValid=false, max=<0 0>, min=<0 0>}
     }
     hoveredForeground {colorUseRule=UseColor_ColorTable, specifiedColor=<rgba 0 0 0 0>}
     hoveredSlateSound {resourceObject=null}
     normal {
      bIsDynamicallyLoaded=false
      drawAs=RoundedBox
      imageType=NoImage
      mirroring=NoMirror
      resourceName=None
      resourceObject=null
      tiling=NoTile
      imageSize=<32 32>
      margin=<ltrb 0 0 0 0>
      outlineSettings {
       bUseBrushTransparency=false
       roundingType=FixedRadius
       width=1
       color {colorUseRule=UseColor_ColorTable, specifiedColor=<rgba 0 0 0 0>}
       cornerRadii=<4 4 4 4>
      }
      tintColor {colorUseRule=UseColor_Specified, specifiedColor=<rgba 0.0395 0.0395 0.0395 1>}
      uVRegion {bIsValid=false, max=<0 0>, min=<0 0>}
     }
     normalForeground {colorUseRule=UseColor_ColorTable, specifiedColor=<rgba 0 0 0 0>}
     normalPadding=<ltrb 0 0 0 0>
     pressed {
      bIsDynamicallyLoaded=false
      drawAs=RoundedBox
      imageType=NoImage
      mirroring=NoMirror
      resourceName=None
      resourceObject=null
      tiling=NoTile
      imageSize=<32 32>
      margin=<ltrb 0 0 0 0>
      outlineSettings {
       bUseBrushTransparency=false
       roundingType=FixedRadius
       width=1
       color {colorUseRule=UseColor_ColorTable, specifiedColor=<rgba 0 0 0 0>}
       cornerRadii=<4 4 4 4>
      }
      tintColor {colorUseRule=UseColor_Specified, specifiedColor=<rgba 0.0284 0.0284 0.0284 1>}
      uVRegion {bIsValid=false, max=<0 0>, min=<0 0>}
     }
     pressedForeground {colorUseRule=UseColor_ColorTable, specifiedColor=<rgba 0 0 0 0>}
     pressedPadding=<ltrb 0 0 0 0>
     pressedSlateSound {resourceObject=null}
    }}
   slot {class=VerticalBoxSlot}
  }
  VerticalBox_8 {
   widgetClassPath=&VerticalBox
   parent=EditorUtilityButton_5
   widget=VerticalBox_8
   slot {class=ButtonSlot, properties {padding=<ltrb 7 7 7 7>}}
  }
  Image_5 {
   widgetClassPath=&Image
   bIsVariable=true
   parent=VerticalBox_8
   widget=Image_5
   properties {
    brush {
     bIsDynamicallyLoaded=false
     drawAs=Image
     imageType=FullColor
     mirroring=NoMirror
     resourceName=None
     resourceObject=/Game/Creative/Devices/DirectionalDecals/Textures/Directional_UTurnArrow_Flipped_Decal
     tiling=NoTile
     imageSize=<20 20>
     margin=<ltrb 0 0 0 0>
     outlineSettings {
      bUseBrushTransparency=false
      roundingType=HalfHeightRadius
      width=0
      color {colorUseRule=UseColor_Specified, specifiedColor=<rgba 0 0 0 0>}
      cornerRadii=<0 0 0 0>
     }
     tintColor {colorUseRule=UseColor_Specified, specifiedColor=<rgba 1 1 1 1>}
     uVRegion {bIsValid=false, max=<0 0>, min=<0 0>}
    }
    renderTransform {angle=151, scale=<1 1>, shear=<0 0>, translation=<0 0>}
   }
   slot {class=VerticalBoxSlot, properties {horizontalAlignment=HAlign_Center, verticalAlignment=VAlign_Center}}
  }
  Button_AutoWalk {
   widgetClassPath=&EditorUtilityButton
   parent=HorizontalBox_9
   widget=Button_AutoWalk
   properties {widgetStyle {
     clickedSlateSound {resourceObject=null}
     disabled {
      bIsDynamicallyLoaded=false
      drawAs=RoundedBox
      imageType=NoImage
      mirroring=NoMirror
      resourceName=None
      resourceObject=null
      tiling=NoTile
      imageSize=<0 0>
      margin=<ltrb 0 0 0 0>
      outlineSettings {
       bUseBrushTransparency=false
       roundingType=FixedRadius
       width=1
       color {colorUseRule=UseColor_ColorTable, specifiedColor=<rgba 0 0 0 0>}
       cornerRadii=<4 4 4 4>
      }
      tintColor {colorUseRule=UseColor_Specified, specifiedColor=<rgba 0.0395 0.0395 0.0395 1>}
      uVRegion {bIsValid=false, max=<0 0>, min=<0 0>}
     }
     disabledForeground {colorUseRule=UseColor_ColorTable, specifiedColor=<rgba 0 0 0 0>}
     hovered {
      bIsDynamicallyLoaded=false
      drawAs=RoundedBox
      imageType=NoImage
      mirroring=NoMirror
      resourceName=None
      resourceObject=null
      tiling=NoTile
      imageSize=<32 32>
      margin=<ltrb 0 0 0 0>
      outlineSettings {
       bUseBrushTransparency=false
       roundingType=FixedRadius
       width=1
       color {colorUseRule=UseColor_ColorTable, specifiedColor=<rgba 0 0 0 0>}
       cornerRadii=<4 4 4 4>
      }
      tintColor {colorUseRule=UseColor_Specified, specifiedColor=<rgba 0.0953 0.0953 0.0953 1>}
      uVRegion {bIsValid=false, max=<0 0>, min=<0 0>}
     }
     hoveredForeground {colorUseRule=UseColor_ColorTable, specifiedColor=<rgba 0 0 0 0>}
     hoveredSlateSound {resourceObject=null}
     normal {
      bIsDynamicallyLoaded=false
      drawAs=RoundedBox
      imageType=NoImage
      mirroring=NoMirror
      resourceName=None
      resourceObject=null
      tiling=NoTile
      imageSize=<32 32>
      margin=<ltrb 0 0 0 0>
      outlineSettings {
       bUseBrushTransparency=false
       roundingType=FixedRadius
       width=1
       color {colorUseRule=UseColor_ColorTable, specifiedColor=<rgba 0 0 0 0>}
       cornerRadii=<4 4 4 4>
      }
      tintColor {colorUseRule=UseColor_Specified, specifiedColor=<rgba 0.0395 0.0395 0.0395 1>}
      uVRegion {bIsValid=false, max=<0 0>, min=<0 0>}
     }
     normalForeground {colorUseRule=UseColor_ColorTable, specifiedColor=<rgba 0 0 0 0>}
     normalPadding=<ltrb 4 1.5 4 1.5>
     pressed {
      bIsDynamicallyLoaded=false
      drawAs=RoundedBox
      imageType=NoImage
      mirroring=NoMirror
      resourceName=None
      resourceObject=null
      tiling=NoTile
      imageSize=<32 32>
      margin=<ltrb 0 0 0 0>
      outlineSettings {
       bUseBrushTransparency=false
       roundingType=FixedRadius
       width=1
       color {colorUseRule=UseColor_ColorTable, specifiedColor=<rgba 0 0 0 0>}
       cornerRadii=<4 4 4 4>
      }
      tintColor {colorUseRule=UseColor_Specified, specifiedColor=<rgba 0.0284 0.0284 0.0284 1>}
      uVRegion {bIsValid=false, max=<0 0>, min=<0 0>}
     }
     pressedForeground {colorUseRule=UseColor_ColorTable, specifiedColor=<rgba 0 0 0 0>}
     pressedPadding=<ltrb 4 2.5 4 0.5>
     pressedSlateSound {resourceObject=null}
    }}
   slot {class=HorizontalBoxSlot}
  }
  CanvasPanel_0 {
   widgetClassPath=&CanvasPanel
   parent=Button_AutoWalk
   widget=CanvasPanel_0
   slot {class=ButtonSlot, properties {horizontalAlignment=HAlign_Fill, verticalAlignment=VAlign_Fill, padding=<ltrb 0 0 0 0>}}
  }
  TextBlock_1 {
   widgetClassPath=&TextBlock
   parent=CanvasPanel_0
   widget=TextBlock_1
   properties {
    text=Auto Walk
    font {
     bForceMonospaced=false
     bMaterialIsStencil=false
     fontMaterial=null
     fontObject=&Roboto
     letterSpacing=0
     monospacedWidth=1
     size=9
     skewAmount=0
     typefaceFontName=Bold
     outlineSettings {
      bApplyOutlineToDropShadows=false
      bMiteredCorners=false
      bSeparateFillAlpha=false
      outlineMaterial=null
      outlineSize=0
      outlineColor=<rgba 0 0 0 1>
     }
    }
   }
   slot {
    class=CanvasPanelSlot
    properties {bAutoSize=true, layoutData {alignment=<0.5 1>, anchors {maximum=<0.5 0.5>, minimum=<0.5 0.5>}, offsets=<ltrb 0 23.0439 0 0>}}
   }
  }
  EditorUtilityCircularThrobber_26 {
   widgetClassPath=/Script/Blutility.EditorUtilityCircularThrobber
   bIsVariable=true
   parent=CanvasPanel_0
   widget=EditorUtilityCircularThrobber_26
   properties {
    numberOfPieces=11
    period=3
    radius=21.4045
    visibility=Collapsed
    image {
     bIsDynamicallyLoaded=false
     drawAs=Image
     imageType=FullColor
     mirroring=NoMirror
     resourceName=../../../Engine/Content/Slate/Common/Throbber_Piece.png
     resourceObject=null
     tiling=NoTile
     imageSize=<5 5>
     margin=<ltrb 0 0 0 0>
     outlineSettings {
      bUseBrushTransparency=false
      roundingType=HalfHeightRadius
      width=0
      color {colorUseRule=UseColor_Specified, specifiedColor=<rgba 0 0 0 0>}
      cornerRadii=<0 0 0 0>
     }
     tintColor {colorUseRule=UseColor_Specified, specifiedColor=<rgba 1 0.8495 0.0192 1>}
     uVRegion {bIsValid=false, max=<0 0>, min=<0 0>}
    }
   }
   slot {
    class=CanvasPanelSlot
    properties {
     bAutoSize=true
     layoutData {alignment=<0.5 0.5>, anchors {maximum=<0.5 0.5>, minimum=<0.5 0.5>}, offsets=<ltrb 1.2513 -14.6775 102.5025 7.978>}
    }
   }
  }
  Image_52 {
   widgetClassPath=&Image
   bIsVariable=true
   parent=CanvasPanel_0
   widget=Image_52
   properties {
    brush {
     bIsDynamicallyLoaded=false
     drawAs=Image
     imageType=FullColor
     mirroring=NoMirror
     resourceName=None
     resourceObject=/Game/Athena/UI/InGame/Creative/PropertyEditor/Icon_Picker/Textures_V2/128x128/T_UI_IconLibrary_Boot_128
     tiling=NoTile
     imageSize=<30 30>
     margin=<ltrb 0 0 0 0>
     outlineSettings {
      bUseBrushTransparency=false
      roundingType=HalfHeightRadius
      width=0
      color {colorUseRule=UseColor_Specified, specifiedColor=<rgba 0 0 0 0>}
      cornerRadii=<0 0 0 0>
     }
     tintColor {colorUseRule=UseColor_Specified, specifiedColor=<rgba 1 1 1 1>}
     uVRegion {bIsValid=false, max=<0 0>, min=<0 0>}
    }
    renderTransform {angle=29, scale=<-1 1>, shear=<0 0>, translation=<0 0>}
   }
   slot {
    class=CanvasPanelSlot
    properties {bAutoSize=true, layoutData {alignment=<0.5 0.5>, anchors {maximum=<0.5 0.5>, minimum=<0.5 0.5>}, offsets=<ltrb 4.6384 -16.904 100 30>}}
   }
  }
  ScaleBox_2 {
   widgetClassPath=&ScaleBox
   parent=CanvasPanel_0
   widget=ScaleBox_2
   properties {clipping=ClipToBounds, stretch=UserSpecified, userSpecifiedScale=0.72}
   slot {
    class=CanvasPanelSlot
    properties {layoutData {alignment=<0.5 0>, anchors {maximum=<0.5 0.5>, minimum=<0.5 0.5>}, offsets=<ltrb 0 21.3373 78.176 20.9095>}}
   }
  }
  SinglePropertyView_3 {
   widgetClassPath=&SinglePropertyView
   parent=ScaleBox_2
   widget=SinglePropertyView_3
   properties {nameOverride=" ", propertyName=WalkStyle}
   slot {class=ScaleBoxSlot}
  }
  HorizontalBox_3 {widgetClassPath=&HorizontalBox, parent=HorizontalBox_11, widget=HorizontalBox_3, slot {class=HorizontalBoxSlot}}
  Button_AddEID {
   widgetClassPath=&EditorUtilityButton
   parent=HorizontalBox_3
   widget=Button_AddEID
   properties {widgetStyle {
     clickedSlateSound {resourceObject=null}
     disabled {
      bIsDynamicallyLoaded=false
      drawAs=RoundedBox
      imageType=NoImage
      mirroring=NoMirror
      resourceName=None
      resourceObject=null
      tiling=NoTile
      imageSize=<0 0>
      margin=<ltrb 0 0 0 0>
      outlineSettings {
       bUseBrushTransparency=false
       roundingType=FixedRadius
       width=1
       color {colorUseRule=UseColor_ColorTable, specifiedColor=<rgba 0 0 0 0>}
       cornerRadii=<4 4 4 4>
      }
      tintColor {colorUseRule=UseColor_Specified, specifiedColor=<rgba 0.0395 0.0395 0.0395 1>}
      uVRegion {bIsValid=false, max=<0 0>, min=<0 0>}
     }
     disabledForeground {colorUseRule=UseColor_ColorTable, specifiedColor=<rgba 0 0 0 0>}
     hovered {
      bIsDynamicallyLoaded=false
      drawAs=RoundedBox
      imageType=NoImage
      mirroring=NoMirror
      resourceName=None
      resourceObject=null
      tiling=NoTile
      imageSize=<32 32>
      margin=<ltrb 0 0 0 0>
      outlineSettings {
       bUseBrushTransparency=false
       roundingType=FixedRadius
       width=1
       color {colorUseRule=UseColor_ColorTable, specifiedColor=<rgba 0 0 0 0>}
       cornerRadii=<4 4 4 4>
      }
      tintColor {colorUseRule=UseColor_Specified, specifiedColor=<rgba 0.0953 0.0953 0.0953 1>}
      uVRegion {bIsValid=false, max=<0 0>, min=<0 0>}
     }
     hoveredForeground {colorUseRule=UseColor_ColorTable, specifiedColor=<rgba 0 0 0 0>}
     hoveredSlateSound {resourceObject=null}
     normal {
      bIsDynamicallyLoaded=false
      drawAs=RoundedBox
      imageType=NoImage
      mirroring=NoMirror
      resourceName=None
      resourceObject=null
      tiling=NoTile
      imageSize=<32 32>
      margin=<ltrb 0 0 0 0>
      outlineSettings {
       bUseBrushTransparency=false
       roundingType=FixedRadius
       width=1
       color {colorUseRule=UseColor_ColorTable, specifiedColor=<rgba 0 0 0 0>}
       cornerRadii=<4 4 4 4>
      }
      tintColor {colorUseRule=UseColor_Specified, specifiedColor=<rgba 0.0395 0.0395 0.0395 1>}
      uVRegion {bIsValid=false, max=<0 0>, min=<0 0>}
     }
     normalForeground {colorUseRule=UseColor_ColorTable, specifiedColor=<rgba 0 0 0 0>}
     normalPadding=<ltrb 5 1.5 5 1.5>
     pressed {
      bIsDynamicallyLoaded=false
      drawAs=RoundedBox
      imageType=NoImage
      mirroring=NoMirror
      resourceName=None
      resourceObject=null
      tiling=NoTile
      imageSize=<32 32>
      margin=<ltrb 0 0 0 0>
      outlineSettings {
       bUseBrushTransparency=false
       roundingType=FixedRadius
       width=1
       color {colorUseRule=UseColor_ColorTable, specifiedColor=<rgba 0 0 0 0>}
       cornerRadii=<4 4 4 4>
      }
      tintColor {colorUseRule=UseColor_Specified, specifiedColor=<rgba 0.0284 0.0284 0.0284 1>}
      uVRegion {bIsValid=false, max=<0 0>, min=<0 0>}
     }
     pressedForeground {colorUseRule=UseColor_ColorTable, specifiedColor=<rgba 0 0 0 0>}
     pressedPadding=<ltrb 5 2.5 5 0.5>
     pressedSlateSound {resourceObject=null}
    }}
   slot {class=HorizontalBoxSlot}
  }
  VerticalBox_4 {
   widgetClassPath=&VerticalBox
   parent=Button_AddEID
   widget=VerticalBox_4
   slot {class=ButtonSlot, properties {padding=<ltrb 0 0 0 0>}}
  }
  TextBlock_3 {
   widgetClassPath=&TextBlock
   parent=VerticalBox_4
   widget=TextBlock_3
   properties {
    justification=Center
    text=+
    font {
     bForceMonospaced=false
     bMaterialIsStencil=false
     fontMaterial=null
     fontObject=&Roboto
     letterSpacing=0
     monospacedWidth=1
     size=22.5
     skewAmount=0
     typefaceFontName=Bold
     outlineSettings {
      bApplyOutlineToDropShadows=false
      bMiteredCorners=false
      bSeparateFillAlpha=false
      outlineMaterial=null
      outlineSize=0
      outlineColor=<rgba 0 0 0 1>
     }
    }
   }
   slot {class=VerticalBoxSlot}
  }
  TextBlock_5 {
   widgetClassPath=&TextBlock
   parent=VerticalBox_4
   widget=TextBlock_5
   properties {
    justification=Center
    text=Add EID
    font {
     bForceMonospaced=false
     bMaterialIsStencil=false
     fontMaterial=null
     fontObject=&Roboto
     letterSpacing=0
     monospacedWidth=1
     size=12
     skewAmount=0
     typefaceFontName=Bold
     outlineSettings {
      bApplyOutlineToDropShadows=false
      bMiteredCorners=false
      bSeparateFillAlpha=false
      outlineMaterial=null
      outlineSize=0
      outlineColor=<rgba 0 0 0 1>
     }
    }
   }
   slot {class=VerticalBoxSlot}
  }
  VerticalBox_6 {
   widgetClassPath=&VerticalBox
   parent=HorizontalBox_3
   widget=VerticalBox_6
   properties {bOverride_Cursor=true}
   slot {class=HorizontalBoxSlot}
  }
  TextBlock_11 {
   widgetClassPath=&TextBlock
   parent=VerticalBox_6
   widget=TextBlock_11
   properties {
    text=Load Release EIDs
    font {
     bForceMonospaced=false
     bMaterialIsStencil=false
     fontMaterial=null
     fontObject=&Roboto
     letterSpacing=0
     monospacedWidth=1
     size=9
     skewAmount=0
     typefaceFontName=Bold
     outlineSettings {
      bApplyOutlineToDropShadows=false
      bMiteredCorners=false
      bSeparateFillAlpha=false
      outlineMaterial=null
      outlineSize=0
      outlineColor=<rgba 0 0 0 1>
     }
    }
   }
   slot {class=VerticalBoxSlot, properties {horizontalAlignment=HAlign_Center, verticalAlignment=VAlign_Center, padding=<ltrb 0 5 0 2>}}
  }
  HorizontalBox_95 {widgetClassPath=&HorizontalBox, parent=VerticalBox_6, widget=HorizontalBox_95, slot {class=VerticalBoxSlot}}
  SinglePropertyView_48 {
   widgetClassPath=&SinglePropertyView
   parent=HorizontalBox_95
   widget=SinglePropertyView_48
   properties {nameOverride="      ", propertyName=EID Release}
   slot {class=HorizontalBoxSlot, properties {padding=<ltrb -15 0 -15 0>, size {sizeRule=Automatic, value=2}}}
  }
  EditorUtilityButton_1 {
   widgetClassPath=&EditorUtilityButton
   parent=HorizontalBox_95
   widget=EditorUtilityButton_1
   properties {widgetStyle {
     clickedSlateSound {resourceObject=null}
     disabled {
      bIsDynamicallyLoaded=false
      drawAs=RoundedBox
      imageType=NoImage
      mirroring=NoMirror
      resourceName=None
      resourceObject=null
      tiling=NoTile
      imageSize=<0 0>
      margin=<ltrb 0 0 0 0>
      outlineSettings {
       bUseBrushTransparency=false
       roundingType=FixedRadius
       width=1
       color {colorUseRule=UseColor_ColorTable, specifiedColor=<rgba 0 0 0 0>}
       cornerRadii=<4 4 4 4>
      }
      tintColor {colorUseRule=UseColor_Specified, specifiedColor=<rgba 0.0395 0.0395 0.0395 1>}
      uVRegion {bIsValid=false, max=<0 0>, min=<0 0>}
     }
     disabledForeground {colorUseRule=UseColor_ColorTable, specifiedColor=<rgba 0 0 0 0>}
     hovered {
      bIsDynamicallyLoaded=false
      drawAs=RoundedBox
      imageType=NoImage
      mirroring=NoMirror
      resourceName=None
      resourceObject=null
      tiling=NoTile
      imageSize=<32 32>
      margin=<ltrb 0 0 0 0>
      outlineSettings {
       bUseBrushTransparency=false
       roundingType=FixedRadius
       width=1
       color {colorUseRule=UseColor_ColorTable, specifiedColor=<rgba 0 0 0 0>}
       cornerRadii=<4 4 4 4>
      }
      tintColor {colorUseRule=UseColor_Specified, specifiedColor=<rgba 0.0953 0.0953 0.0953 1>}
      uVRegion {bIsValid=false, max=<0 0>, min=<0 0>}
     }
     hoveredForeground {colorUseRule=UseColor_ColorTable, specifiedColor=<rgba 0 0 0 0>}
     hoveredSlateSound {resourceObject=null}
     normal {
      bIsDynamicallyLoaded=false
      drawAs=RoundedBox
      imageType=NoImage
      mirroring=NoMirror
      resourceName=None
      resourceObject=null
      tiling=NoTile
      imageSize=<32 32>
      margin=<ltrb 0 0 0 0>
      outlineSettings {
       bUseBrushTransparency=false
       roundingType=FixedRadius
       width=1
       color {colorUseRule=UseColor_ColorTable, specifiedColor=<rgba 0 0 0 0>}
       cornerRadii=<4 4 4 4>
      }
      tintColor {colorUseRule=UseColor_Specified, specifiedColor=<rgba 0.0395 0.0395 0.0395 1>}
      uVRegion {bIsValid=false, max=<0 0>, min=<0 0>}
     }
     normalForeground {colorUseRule=UseColor_ColorTable, specifiedColor=<rgba 0 0 0 0>}
     normalPadding=<ltrb 12 0 12 0>
     pressed {
      bIsDynamicallyLoaded=false
      drawAs=RoundedBox
      imageType=NoImage
      mirroring=NoMirror
      resourceName=None
      resourceObject=null
      tiling=NoTile
      imageSize=<32 32>
      margin=<ltrb 0 0 0 0>
      outlineSettings {
       bUseBrushTransparency=false
       roundingType=FixedRadius
       width=1
       color {colorUseRule=UseColor_ColorTable, specifiedColor=<rgba 0 0 0 0>}
       cornerRadii=<4 4 4 4>
      }
      tintColor {colorUseRule=UseColor_Specified, specifiedColor=<rgba 0.0284 0.0284 0.0284 1>}
      uVRegion {bIsValid=false, max=<0 0>, min=<0 0>}
     }
     pressedForeground {colorUseRule=UseColor_ColorTable, specifiedColor=<rgba 0 0 0 0>}
     pressedPadding=<ltrb 12 0 12 0>
     pressedSlateSound {resourceObject=null}
    }}
   slot {class=HorizontalBoxSlot, properties {horizontalAlignment=HAlign_Right, verticalAlignment=VAlign_Center, size {sizeRule=Fill, value=1}}}
  }
  Image_1 {
   widgetClassPath=&Image
   bIsVariable=true
   parent=EditorUtilityButton_1
   widget=Image_1
   properties {brush {
     bIsDynamicallyLoaded=false
     drawAs=Image
     imageType=FullColor
     mirroring=NoMirror
     resourceName=None
     resourceObject=null
     tiling=NoTile
     imageSize=<15 15>
     margin=<ltrb 0 0 0 0>
     outlineSettings {
      bUseBrushTransparency=false
      roundingType=HalfHeightRadius
      width=0
      color {colorUseRule=UseColor_Specified, specifiedColor=<rgba 0 0 0 0>}
      cornerRadii=<0 0 0 0>
     }
     tintColor {colorUseRule=UseColor_Specified, specifiedColor=<rgba 1 1 1 1>}
     uVRegion {bIsValid=false, max=<0 0>, min=<0 0>}
    }}
   slot {class=ButtonSlot, properties {padding=<ltrb 2 2 2 2>}}
  }
  TextBlock_17 {
   widgetClassPath=&TextBlock
   parent=VerticalBox_6
   widget=TextBlock_17
   properties {
    justification=Center
    text=Clear All EIDs
    font {
     bForceMonospaced=false
     bMaterialIsStencil=false
     fontMaterial=null
     fontObject=&Roboto
     letterSpacing=0
     monospacedWidth=1
     size=9
     skewAmount=0
     typefaceFontName=Bold
     outlineSettings {
      bApplyOutlineToDropShadows=false
      bMiteredCorners=false
      bSeparateFillAlpha=false
      outlineMaterial=null
      outlineSize=0
      outlineColor=<rgba 0 0 0 1>
     }
    }
   }
   slot {class=VerticalBoxSlot, properties {padding=<ltrb 0 7 0 0>}}
  }
  EditorUtilityButton_4 {
   widgetClassPath=&EditorUtilityButton
   parent=VerticalBox_6
   widget=EditorUtilityButton_4
   properties {widgetStyle {
     clickedSlateSound {resourceObject=null}
     disabled {
      bIsDynamicallyLoaded=false
      drawAs=RoundedBox
      imageType=NoImage
      mirroring=NoMirror
      resourceName=None
      resourceObject=null
      tiling=NoTile
      imageSize=<0 0>
      margin=<ltrb 0 0 0 0>
      outlineSettings {
       bUseBrushTransparency=false
       roundingType=FixedRadius
       width=1
       color {colorUseRule=UseColor_ColorTable, specifiedColor=<rgba 0 0 0 0>}
       cornerRadii=<4 4 4 4>
      }
      tintColor {colorUseRule=UseColor_Specified, specifiedColor=<rgba 0.0395 0.0395 0.0395 1>}
      uVRegion {bIsValid=false, max=<0 0>, min=<0 0>}
     }
     disabledForeground {colorUseRule=UseColor_ColorTable, specifiedColor=<rgba 0 0 0 0>}
     hovered {
      bIsDynamicallyLoaded=false
      drawAs=RoundedBox
      imageType=NoImage
      mirroring=NoMirror
      resourceName=None
      resourceObject=null
      tiling=NoTile
      imageSize=<32 32>
      margin=<ltrb 0 0 0 0>
      outlineSettings {
       bUseBrushTransparency=false
       roundingType=FixedRadius
       width=1
       color {colorUseRule=UseColor_ColorTable, specifiedColor=<rgba 0 0 0 0>}
       cornerRadii=<4 4 4 4>
      }
      tintColor {colorUseRule=UseColor_Specified, specifiedColor=<rgba 0.0953 0.0953 0.0953 1>}
      uVRegion {bIsValid=false, max=<0 0>, min=<0 0>}
     }
     hoveredForeground {colorUseRule=UseColor_ColorTable, specifiedColor=<rgba 0 0 0 0>}
     hoveredSlateSound {resourceObject=null}
     normal {
      bIsDynamicallyLoaded=false
      drawAs=RoundedBox
      imageType=NoImage
      mirroring=NoMirror
      resourceName=None
      resourceObject=null
      tiling=NoTile
      imageSize=<32 32>
      margin=<ltrb 0 0 0 0>
      outlineSettings {
       bUseBrushTransparency=false
       roundingType=FixedRadius
       width=1
       color {colorUseRule=UseColor_ColorTable, specifiedColor=<rgba 0 0 0 0>}
       cornerRadii=<4 4 4 4>
      }
      tintColor {colorUseRule=UseColor_Specified, specifiedColor=<rgba 0.0395 0.0395 0.0395 1>}
      uVRegion {bIsValid=false, max=<0 0>, min=<0 0>}
     }
     normalForeground {colorUseRule=UseColor_ColorTable, specifiedColor=<rgba 0 0 0 0>}
     normalPadding=<ltrb 12 0 12 0>
     pressed {
      bIsDynamicallyLoaded=false
      drawAs=RoundedBox
      imageType=NoImage
      mirroring=NoMirror
      resourceName=None
      resourceObject=null
      tiling=NoTile
      imageSize=<32 32>
      margin=<ltrb 0 0 0 0>
      outlineSettings {
       bUseBrushTransparency=false
       roundingType=FixedRadius
       width=1
       color {colorUseRule=UseColor_ColorTable, specifiedColor=<rgba 0 0 0 0>}
       cornerRadii=<4 4 4 4>
      }
      tintColor {colorUseRule=UseColor_Specified, specifiedColor=<rgba 0.0284 0.0284 0.0284 1>}
      uVRegion {bIsValid=false, max=<0 0>, min=<0 0>}
     }
     pressedForeground {colorUseRule=UseColor_ColorTable, specifiedColor=<rgba 0 0 0 0>}
     pressedPadding=<ltrb 12 0 12 0>
     pressedSlateSound {resourceObject=null}
    }}
   slot {class=VerticalBoxSlot, properties {horizontalAlignment=HAlign_Center}}
  }
  Image {
   widgetClassPath=&Image
   bIsVariable=true
   parent=EditorUtilityButton_4
   widget=Image
   properties {brush {
     bIsDynamicallyLoaded=false
     drawAs=Image
     imageType=FullColor
     mirroring=NoMirror
     resourceName=None
     resourceObject=null
     tiling=NoTile
     imageSize=<20 20>
     margin=<ltrb 0 0 0 0>
     outlineSettings {
      bUseBrushTransparency=false
      roundingType=HalfHeightRadius
      width=0
      color {colorUseRule=UseColor_Specified, specifiedColor=<rgba 0 0 0 0>}
      cornerRadii=<0 0 0 0>
     }
     tintColor {colorUseRule=UseColor_Specified, specifiedColor=<rgba 1 1 1 1>}
     uVRegion {bIsValid=false, max=<0 0>, min=<0 0>}
    }}
   slot {class=ButtonSlot, properties {padding=<ltrb 2 2 2 2>}}
  }
  HorizontalBox_7 {
   widgetClassPath=&HorizontalBox
   parent=VerticalBox_2
   widget=HorizontalBox_7
   slot {class=VerticalBoxSlot, properties {padding=<ltrb 0 5 0 0>}}
  }
  TextBlock_13 {
   widgetClassPath=&TextBlock
   parent=HorizontalBox_7
   widget=TextBlock_13
   properties {
    text="CIDs ▼"
    font {
     bForceMonospaced=false
     bMaterialIsStencil=false
     fontMaterial=null
     fontObject=&Roboto
     letterSpacing=0
     monospacedWidth=1
     size=10.5
     skewAmount=0
     typefaceFontName=Bold
     outlineSettings {
      bApplyOutlineToDropShadows=false
      bMiteredCorners=false
      bSeparateFillAlpha=false
      outlineMaterial=null
      outlineSize=0
      outlineColor=<rgba 0 0 0 1>
     }
    }
   }
   slot {class=HorizontalBoxSlot}
  }
  Checkbox_ShowCIDs {
   widgetClassPath=&EditorUtilityCheckBox
   bIsVariable=true
   parent=HorizontalBox_7
   widget=Checkbox_ShowCIDs
   properties {widgetStyle {
     checkBoxType=CheckBox
     backgroundHoveredImage {
      bIsDynamicallyLoaded=false
      drawAs=RoundedBox
      imageType=NoImage
      mirroring=NoMirror
      resourceName=None
      resourceObject=null
      tiling=NoTile
      imageSize=<18 18>
      margin=<ltrb 0 0 0 0>
      outlineSettings {
       bUseBrushTransparency=false
       roundingType=FixedRadius
       width=1
       color {colorUseRule=UseColor_ColorTable, specifiedColor=<rgba 0 0 0 0>}
       cornerRadii=<3 3 3 3>
      }
      tintColor {colorUseRule=UseColor_Specified, specifiedColor=<rgba 0.0048 0.0048 0.0048 1>}
      uVRegion {bIsValid=false, max=<0 0>, min=<0 0>}
     }
     backgroundImage {
      bIsDynamicallyLoaded=false
      drawAs=RoundedBox
      imageType=NoImage
      mirroring=NoMirror
      resourceName=None
      resourceObject=null
      tiling=NoTile
      imageSize=<18 18>
      margin=<ltrb 0 0 0 0>
      outlineSettings {
       bUseBrushTransparency=false
       roundingType=FixedRadius
       width=1
       color {colorUseRule=UseColor_ColorTable, specifiedColor=<rgba 0 0 0 0>}
       cornerRadii=<3 3 3 3>
      }
      tintColor {colorUseRule=UseColor_Specified, specifiedColor=<rgba 0.0048 0.0048 0.0048 1>}
      uVRegion {bIsValid=false, max=<0 0>, min=<0 0>}
     }
     backgroundPressedImage {
      bIsDynamicallyLoaded=false
      drawAs=RoundedBox
      imageType=NoImage
      mirroring=NoMirror
      resourceName=None
      resourceObject=null
      tiling=NoTile
      imageSize=<18 18>
      margin=<ltrb 0 0 0 0>
      outlineSettings {
       bUseBrushTransparency=false
       roundingType=FixedRadius
       width=1
       color {colorUseRule=UseColor_ColorTable, specifiedColor=<rgba 0 0 0 0>}
       cornerRadii=<3 3 3 3>
      }
      tintColor {colorUseRule=UseColor_Specified, specifiedColor=<rgba 0.0048 0.0048 0.0048 1>}
      uVRegion {bIsValid=false, max=<0 0>, min=<0 0>}
     }
     borderBackgroundColor {colorUseRule=UseColor_Specified, specifiedColor=<rgba 1 1 1 1>}
     checkedForeground {colorUseRule=UseColor_Specified, specifiedColor=<rgba 1 1 1 1>}
     checkedHoveredForeground {colorUseRule=UseColor_Specified, specifiedColor=<rgba 1 1 1 1>}
     checkedHoveredImage {
      bIsDynamicallyLoaded=false
      drawAs=Image
      imageType=Vector
      mirroring=NoMirror
      resourceName=../../../Engine/Content/Slate/Starship/CoreWidgets/CheckBox/check.svg
      resourceObject=null
      tiling=NoTile
      imageSize=<16 16>
      margin=<ltrb 0 0 0 0>
      outlineSettings {
       bUseBrushTransparency=false
       roundingType=HalfHeightRadius
       width=0
       color {colorUseRule=UseColor_Specified, specifiedColor=<rgba 0 0 0 0>}
       cornerRadii=<0 0 0 0>
      }
      tintColor {colorUseRule=UseColor_Specified, specifiedColor=<rgba 0.0044 0.2384 1 1>}
      uVRegion {bIsValid=false, max=<0 0>, min=<0 0>}
     }
     checkedImage {
      bIsDynamicallyLoaded=false
      drawAs=Image
      imageType=Vector
      mirroring=NoMirror
      resourceName=../../../Engine/Content/Slate/Starship/CoreWidgets/CheckBox/check.svg
      resourceObject=null
      tiling=NoTile
      imageSize=<16 16>
      margin=<ltrb 0 0 0 0>
      outlineSettings {
       bUseBrushTransparency=false
       roundingType=HalfHeightRadius
       width=0
       color {colorUseRule=UseColor_Specified, specifiedColor=<rgba 0 0 0 0>}
       cornerRadii=<0 0 0 0>
      }
      tintColor {colorUseRule=UseColor_Specified, specifiedColor=<rgba 0 0.162 0.7454 1>}
      uVRegion {bIsValid=false, max=<0 0>, min=<0 0>}
     }
     checkedPressedForeground {colorUseRule=UseColor_Specified, specifiedColor=<rgba 1 1 1 1>}
     checkedPressedImage {
      bIsDynamicallyLoaded=false
      drawAs=Image
      imageType=Vector
      mirroring=NoMirror
      resourceName=../../../Engine/Content/Slate/Starship/CoreWidgets/CheckBox/check.svg
      resourceObject=null
      tiling=NoTile
      imageSize=<16 16>
      margin=<ltrb 0 0 0 0>
      outlineSettings {
       bUseBrushTransparency=false
       roundingType=HalfHeightRadius
       width=0
       color {colorUseRule=UseColor_Specified, specifiedColor=<rgba 0 0 0 0>}
       cornerRadii=<0 0 0 0>
      }
      tintColor {colorUseRule=UseColor_Specified, specifiedColor=<rgba 0 0.0802 0.3515 1>}
      uVRegion {bIsValid=false, max=<0 0>, min=<0 0>}
     }
     checkedSlateSound {resourceObject=null}
     foregroundColor {colorUseRule=UseColor_Specified, specifiedColor=<rgba 1 1 1 1>}
     hoveredForeground {colorUseRule=UseColor_Specified, specifiedColor=<rgba 1 1 1 1>}
     hoveredSlateSound {resourceObject=null}
     padding=<ltrb 2 0 0 0>
     pressedForeground {colorUseRule=UseColor_Specified, specifiedColor=<rgba 1 1 1 1>}
     uncheckedHoveredImage {
      bIsDynamicallyLoaded=false
      drawAs=NoDrawType
      imageType=NoImage
      mirroring=NoMirror
      resourceName=None
      resourceObject=null
      tiling=NoTile
      imageSize=<0 0>
      margin=<ltrb 0 0 0 0>
      outlineSettings {
       bUseBrushTransparency=false
       roundingType=HalfHeightRadius
       width=0
       color {colorUseRule=UseColor_Specified, specifiedColor=<rgba 0 0 0 0>}
       cornerRadii=<0 0 0 0>
      }
      tintColor {colorUseRule=UseColor_Specified, specifiedColor=<rgba 1 1 1 1>}
      uVRegion {bIsValid=false, max=<0 0>, min=<0 0>}
     }
     uncheckedImage {
      bIsDynamicallyLoaded=false
      drawAs=NoDrawType
      imageType=NoImage
      mirroring=NoMirror
      resourceName=None
      resourceObject=null
      tiling=NoTile
      imageSize=<0 0>
      margin=<ltrb 0 0 0 0>
      outlineSettings {
       bUseBrushTransparency=false
       roundingType=HalfHeightRadius
       width=0
       color {colorUseRule=UseColor_Specified, specifiedColor=<rgba 0 0 0 0>}
       cornerRadii=<0 0 0 0>
      }
      tintColor {colorUseRule=UseColor_Specified, specifiedColor=<rgba 1 1 1 1>}
      uVRegion {bIsValid=false, max=<0 0>, min=<0 0>}
     }
     uncheckedPressedImage {
      bIsDynamicallyLoaded=false
      drawAs=NoDrawType
      imageType=NoImage
      mirroring=NoMirror
      resourceName=None
      resourceObject=null
      tiling=NoTile
      imageSize=<0 0>
      margin=<ltrb 0 0 0 0>
      outlineSettings {
       bUseBrushTransparency=false
       roundingType=HalfHeightRadius
       width=0
       color {colorUseRule=UseColor_Specified, specifiedColor=<rgba 0 0 0 0>}
       cornerRadii=<0 0 0 0>
      }
      tintColor {colorUseRule=UseColor_Specified, specifiedColor=<rgba 1 1 1 1>}
      uVRegion {bIsValid=false, max=<0 0>, min=<0 0>}
     }
     uncheckedSlateSound {resourceObject=null}
     undeterminedForeground {colorUseRule=UseColor_Specified, specifiedColor=<rgba 1 1 1 1>}
     undeterminedHoveredImage {
      bIsDynamicallyLoaded=false
      drawAs=Image
      imageType=Vector
      mirroring=NoMirror
      resourceName=../../../Engine/Content/Slate/Starship/CoreWidgets/CheckBox/indeterminate.svg
      resourceObject=null
      tiling=NoTile
      imageSize=<16 16>
      margin=<ltrb 0 0 0 0>
      outlineSettings {
       bUseBrushTransparency=false
       roundingType=HalfHeightRadius
       width=0
       color {colorUseRule=UseColor_Specified, specifiedColor=<rgba 0 0 0 0>}
       cornerRadii=<0 0 0 0>
      }
      tintColor {colorUseRule=UseColor_Specified, specifiedColor=<rgba 0.0044 0.2384 1 1>}
      uVRegion {bIsValid=false, max=<0 0>, min=<0 0>}
     }
     undeterminedImage {
      bIsDynamicallyLoaded=false
      drawAs=Image
      imageType=Vector
      mirroring=NoMirror
      resourceName=../../../Engine/Content/Slate/Starship/CoreWidgets/CheckBox/indeterminate.svg
      resourceObject=null
      tiling=NoTile
      imageSize=<16 16>
      margin=<ltrb 0 0 0 0>
      outlineSettings {
       bUseBrushTransparency=false
       roundingType=HalfHeightRadius
       width=0
       color {colorUseRule=UseColor_Specified, specifiedColor=<rgba 0 0 0 0>}
       cornerRadii=<0 0 0 0>
      }
      tintColor {colorUseRule=UseColor_Specified, specifiedColor=<rgba 0 0.162 0.7454 1>}
      uVRegion {bIsValid=false, max=<0 0>, min=<0 0>}
     }
     undeterminedPressedImage {
      bIsDynamicallyLoaded=false
      drawAs=Image
      imageType=Vector
      mirroring=NoMirror
      resourceName=../../../Engine/Content/Slate/Starship/CoreWidgets/CheckBox/indeterminate.svg
      resourceObject=null
      tiling=NoTile
      imageSize=<16 16>
      margin=<ltrb 0 0 0 0>
      outlineSettings {
       bUseBrushTransparency=false
       roundingType=HalfHeightRadius
       width=0
       color {colorUseRule=UseColor_Specified, specifiedColor=<rgba 0 0 0 0>}
       cornerRadii=<0 0 0 0>
      }
      tintColor {colorUseRule=UseColor_Specified, specifiedColor=<rgba 0 0.0802 0.3515 1>}
      uVRegion {bIsValid=false, max=<0 0>, min=<0 0>}
     }
    }}
   slot {class=HorizontalBoxSlot, properties {padding=<ltrb 3 0 0 0>}}
  }
  WrapBox_CIDs {
   widgetClassPath=&WrapBox
   bIsVariable=true
   parent=VerticalBox_2
   widget=WrapBox_CIDs
   slot {class=VerticalBoxSlot, properties {verticalAlignment=VAlign_Top}}
  }
  EUW_MultiCamViewer_IdButton {
   widgetClassPath=/Game/NeverCook/Animation/MultiCamViewer/Components/EUW_MultiCamViewer_IdButton
   bIsVariable=true
   parent=WrapBox_CIDs
   widget=EUW_MultiCamViewer_IdButton
   properties {
    cosmeticType=1
    isEditable=false
    itemDef=/Game/Athena/Items/Cosmetics/Characters/CID_560_Athena_Commando_M_RebirthDefaultA
    replaceName=true
   }
   slot {class=WrapBoxSlot}
  }
  EUW_MultiCamViewer_IdButton_3 {
   widgetClassPath=&EUW_MultiCamViewer_IdButton
   bIsVariable=true
   parent=WrapBox_CIDs
   widget=EUW_MultiCamViewer_IdButton_3
   properties {
    cosmeticType=1
    isEditable=false
    itemDef=/Game/Athena/Items/Cosmetics/Characters/CID_556_Athena_Commando_F_RebirthDefaultA
    newName=CMF
    replaceName=true
   }
   slot {class=WrapBoxSlot}
  }
  EUW_MultiCamViewer_IdButton_4 {
   widgetClassPath=&EUW_MultiCamViewer_IdButton
   bIsVariable=true
   parent=WrapBox_CIDs
   widget=EUW_MultiCamViewer_IdButton_4
   properties {
    cosmeticType=1
    isEditable=false
    itemDef=/BRCosmetics/Athena/Items/Cosmetics/Characters/CID_719_Athena_Commando_F_Blonde
    newName=CLF
    replaceName=true
   }
   slot {class=WrapBoxSlot}
  }
  EUW_MultiCamViewer_IdButton_2 {
   widgetClassPath=&EUW_MultiCamViewer_IdButton
   bIsVariable=true
   parent=WrapBox_CIDs
   widget=EUW_MultiCamViewer_IdButton_2
   properties {
    cosmeticType=1
    isEditable=false
    itemDef=/BRCosmetics/Athena/Items/Cosmetics/Characters/CID_NPC_Athena_Commando_M_Kyle
    newName=CLM
    replaceName=true
   }
   slot {class=WrapBoxSlot}
  }
  EUW_MultiCamViewer_IdButton_1 {
   widgetClassPath=&EUW_MultiCamViewer_IdButton
   bIsVariable=true
   parent=WrapBox_CIDs
   widget=EUW_MultiCamViewer_IdButton_1
   properties {cosmeticType=1, isEditable=false}
   slot {class=WrapBoxSlot}
  }
  EUW_MultiCamViewer_IdButton_5 {
   widgetClassPath=&EUW_MultiCamViewer_IdButton
   bIsVariable=true
   parent=WrapBox_CIDs
   widget=EUW_MultiCamViewer_IdButton_5
   properties {cosmeticType=1, isEditable=false, itemDef=/BRCosmetics/Athena/Items/Cosmetics/Characters/Character_SirWolf}
   slot {class=WrapBoxSlot}
  }
  EUW_MultiCamViewer_IdButton_7 {
   widgetClassPath=&EUW_MultiCamViewer_IdButton
   bIsVariable=true
   parent=WrapBox_CIDs
   widget=EUW_MultiCamViewer_IdButton_7
   properties {cosmeticType=1, isEditable=false, itemDef=/BRCosmetics/Athena/Items/Cosmetics/Characters/Character_ZebraScramble_Bacon}
   slot {class=WrapBoxSlot}
  }
  EUW_MultiCamViewer_IdButton_8 {
   widgetClassPath=&EUW_MultiCamViewer_IdButton
   bIsVariable=true
   parent=WrapBox_CIDs
   widget=EUW_MultiCamViewer_IdButton_8
   properties {cosmeticType=1, isEditable=false, itemDef=/BRCosmetics/Athena/Items/Cosmetics/Characters/CID_A_158_Athena_Commando_F_Buffet_YC20H}
   slot {class=WrapBoxSlot}
  }
  Button_AddEID_1 {
   widgetClassPath=&EditorUtilityButton
   parent=WrapBox_CIDs
   widget=Button_AddEID_1
   properties {widgetStyle {
     clickedSlateSound {resourceObject=null}
     disabled {
      bIsDynamicallyLoaded=false
      drawAs=RoundedBox
      imageType=NoImage
      mirroring=NoMirror
      resourceName=None
      resourceObject=null
      tiling=NoTile
      imageSize=<0 0>
      margin=<ltrb 0 0 0 0>
      outlineSettings {
       bUseBrushTransparency=false
       roundingType=FixedRadius
       width=1
       color {colorUseRule=UseColor_ColorTable, specifiedColor=<rgba 0 0 0 0>}
       cornerRadii=<4 4 4 4>
      }
      tintColor {colorUseRule=UseColor_Specified, specifiedColor=<rgba 0.0395 0.0395 0.0395 1>}
      uVRegion {bIsValid=false, max=<0 0>, min=<0 0>}
     }
     disabledForeground {colorUseRule=UseColor_ColorTable, specifiedColor=<rgba 0 0 0 0>}
     hovered {
      bIsDynamicallyLoaded=false
      drawAs=RoundedBox
      imageType=NoImage
      mirroring=NoMirror
      resourceName=None
      resourceObject=null
      tiling=NoTile
      imageSize=<32 32>
      margin=<ltrb 0 0 0 0>
      outlineSettings {
       bUseBrushTransparency=false
       roundingType=FixedRadius
       width=1
       color {colorUseRule=UseColor_ColorTable, specifiedColor=<rgba 0 0 0 0>}
       cornerRadii=<4 4 4 4>
      }
      tintColor {colorUseRule=UseColor_Specified, specifiedColor=<rgba 0.0953 0.0953 0.0953 1>}
      uVRegion {bIsValid=false, max=<0 0>, min=<0 0>}
     }
     hoveredForeground {colorUseRule=UseColor_ColorTable, specifiedColor=<rgba 0 0 0 0>}
     hoveredSlateSound {resourceObject=null}
     normal {
      bIsDynamicallyLoaded=false
      drawAs=RoundedBox
      imageType=NoImage
      mirroring=NoMirror
      resourceName=None
      resourceObject=null
      tiling=NoTile
      imageSize=<32 32>
      margin=<ltrb 0 0 0 0>
      outlineSettings {
       bUseBrushTransparency=false
       roundingType=FixedRadius
       width=1
       color {colorUseRule=UseColor_ColorTable, specifiedColor=<rgba 0 0 0 0>}
       cornerRadii=<4 4 4 4>
      }
      tintColor {colorUseRule=UseColor_Specified, specifiedColor=<rgba 0.0395 0.0395 0.0395 1>}
      uVRegion {bIsValid=false, max=<0 0>, min=<0 0>}
     }
     normalForeground {colorUseRule=UseColor_ColorTable, specifiedColor=<rgba 0 0 0 0>}
     normalPadding=<ltrb 5 1.5 5 1.5>
     pressed {
      bIsDynamicallyLoaded=false
      drawAs=RoundedBox
      imageType=NoImage
      mirroring=NoMirror
      resourceName=None
      resourceObject=null
      tiling=NoTile
      imageSize=<32 32>
      margin=<ltrb 0 0 0 0>
      outlineSettings {
       bUseBrushTransparency=false
       roundingType=FixedRadius
       width=1
       color {colorUseRule=UseColor_ColorTable, specifiedColor=<rgba 0 0 0 0>}
       cornerRadii=<4 4 4 4>
      }
      tintColor {colorUseRule=UseColor_Specified, specifiedColor=<rgba 0.0284 0.0284 0.0284 1>}
      uVRegion {bIsValid=false, max=<0 0>, min=<0 0>}
     }
     pressedForeground {colorUseRule=UseColor_ColorTable, specifiedColor=<rgba 0 0 0 0>}
     pressedPadding=<ltrb 5 2.5 5 0.5>
     pressedSlateSound {resourceObject=null}
    }}
   slot {class=WrapBoxSlot}
  }
  VerticalBox {
   widgetClassPath=&VerticalBox
   parent=Button_AddEID_1
   widget=VerticalBox
   slot {class=ButtonSlot, properties {padding=<ltrb 0 0 0 0>}}
  }
  TextBlock_4 {
   widgetClassPath=&TextBlock
   parent=VerticalBox
   widget=TextBlock_4
   properties {
    justification=Center
    text=+
    font {
     bForceMonospaced=false
     bMaterialIsStencil=false
     fontMaterial=null
     fontObject=&Roboto
     letterSpacing=0
     monospacedWidth=1
     size=15
     skewAmount=0
     typefaceFontName=Bold
     outlineSettings {
      bApplyOutlineToDropShadows=false
      bMiteredCorners=false
      bSeparateFillAlpha=false
      outlineMaterial=null
      outlineSize=0
      outlineColor=<rgba 0 0 0 1>
     }
    }
   }
   slot {class=VerticalBoxSlot}
  }
  TextBlock_12 {
   widgetClassPath=&TextBlock
   parent=VerticalBox
   widget=TextBlock_12
   properties {
    justification=Center
    text=Add CID
    font {
     bForceMonospaced=false
     bMaterialIsStencil=false
     fontMaterial=null
     fontObject=&Roboto
     letterSpacing=0
     monospacedWidth=1
     size=9
     skewAmount=0
     typefaceFontName=Bold
     outlineSettings {
      bApplyOutlineToDropShadows=false
      bMiteredCorners=false
      bSeparateFillAlpha=false
      outlineMaterial=null
      outlineSize=0
      outlineColor=<rgba 0 0 0 1>
     }
    }
   }
   slot {class=VerticalBoxSlot}
  }
  HorizontalBox_4 {widgetClassPath=&HorizontalBox, parent=WrapBox_CIDs, widget=HorizontalBox_4, slot {class=WrapBoxSlot}}
  EditorUtilityButton_2 {
   widgetClassPath=&EditorUtilityButton
   parent=HorizontalBox_4
   widget=EditorUtilityButton_2
   properties {widgetStyle {
     clickedSlateSound {resourceObject=null}
     disabled {
      bIsDynamicallyLoaded=false
      drawAs=RoundedBox
      imageType=NoImage
      mirroring=NoMirror
      resourceName=None
      resourceObject=null
      tiling=NoTile
      imageSize=<0 0>
      margin=<ltrb 0 0 0 0>
      outlineSettings {
       bUseBrushTransparency=false
       roundingType=FixedRadius
       width=1
       color {colorUseRule=UseColor_ColorTable, specifiedColor=<rgba 0 0 0 0>}
       cornerRadii=<4 4 4 4>
      }
      tintColor {colorUseRule=UseColor_Specified, specifiedColor=<rgba 0.0395 0.0395 0.0395 1>}
      uVRegion {bIsValid=false, max=<0 0>, min=<0 0>}
     }
     disabledForeground {colorUseRule=UseColor_ColorTable, specifiedColor=<rgba 0 0 0 0>}
     hovered {
      bIsDynamicallyLoaded=false
      drawAs=RoundedBox
      imageType=NoImage
      mirroring=NoMirror
      resourceName=None
      resourceObject=null
      tiling=NoTile
      imageSize=<32 32>
      margin=<ltrb 0 0 0 0>
      outlineSettings {
       bUseBrushTransparency=false
       roundingType=FixedRadius
       width=1
       color {colorUseRule=UseColor_ColorTable, specifiedColor=<rgba 0 0 0 0>}
       cornerRadii=<4 4 4 4>
      }
      tintColor {colorUseRule=UseColor_Specified, specifiedColor=<rgba 0.0953 0.0953 0.0953 1>}
      uVRegion {bIsValid=false, max=<0 0>, min=<0 0>}
     }
     hoveredForeground {colorUseRule=UseColor_ColorTable, specifiedColor=<rgba 0 0 0 0>}
     hoveredSlateSound {resourceObject=null}
     normal {
      bIsDynamicallyLoaded=false
      drawAs=RoundedBox
      imageType=NoImage
      mirroring=NoMirror
      resourceName=None
      resourceObject=null
      tiling=NoTile
      imageSize=<32 32>
      margin=<ltrb 0 0 0 0>
      outlineSettings {
       bUseBrushTransparency=false
       roundingType=FixedRadius
       width=1
       color {colorUseRule=UseColor_ColorTable, specifiedColor=<rgba 0 0 0 0>}
       cornerRadii=<4 4 4 4>
      }
      tintColor {colorUseRule=UseColor_Specified, specifiedColor=<rgba 0.0395 0.0395 0.0395 1>}
      uVRegion {bIsValid=false, max=<0 0>, min=<0 0>}
     }
     normalForeground {colorUseRule=UseColor_ColorTable, specifiedColor=<rgba 0 0 0 0>}
     normalPadding=<ltrb 0 0 0 0>
     pressed {
      bIsDynamicallyLoaded=false
      drawAs=RoundedBox
      imageType=NoImage
      mirroring=NoMirror
      resourceName=None
      resourceObject=null
      tiling=NoTile
      imageSize=<32 32>
      margin=<ltrb 0 0 0 0>
      outlineSettings {
       bUseBrushTransparency=false
       roundingType=FixedRadius
       width=1
       color {colorUseRule=UseColor_ColorTable, specifiedColor=<rgba 0 0 0 0>}
       cornerRadii=<4 4 4 4>
      }
      tintColor {colorUseRule=UseColor_Specified, specifiedColor=<rgba 0.0284 0.0284 0.0284 1>}
      uVRegion {bIsValid=false, max=<0 0>, min=<0 0>}
     }
     pressedForeground {colorUseRule=UseColor_ColorTable, specifiedColor=<rgba 0 0 0 0>}
     pressedPadding=<ltrb 0 0 0 0>
     pressedSlateSound {resourceObject=null}
    }}
   slot {class=HorizontalBoxSlot, properties {horizontalAlignment=HAlign_Center, padding=<ltrb 10 0 10 0>}}
  }
  VerticalBox_7 {
   widgetClassPath=&VerticalBox
   parent=EditorUtilityButton_2
   widget=VerticalBox_7
   slot {class=ButtonSlot, properties {padding=<ltrb 7 7 7 7>}}
  }
  Image_3 {
   widgetClassPath=&Image
   bIsVariable=true
   parent=VerticalBox_7
   widget=Image_3
   properties {brush {
     bIsDynamicallyLoaded=false
     drawAs=Image
     imageType=FullColor
     mirroring=NoMirror
     resourceName=None
     resourceObject=null
     tiling=NoTile
     imageSize=<20 20>
     margin=<ltrb 0 0 0 0>
     outlineSettings {
      bUseBrushTransparency=false
      roundingType=HalfHeightRadius
      width=0
      color {colorUseRule=UseColor_Specified, specifiedColor=<rgba 0 0 0 0>}
      cornerRadii=<0 0 0 0>
     }
     tintColor {colorUseRule=UseColor_Specified, specifiedColor=<rgba 1 1 1 1>}
     uVRegion {bIsValid=false, max=<0 0>, min=<0 0>}
    }}
   slot {class=VerticalBoxSlot, properties {horizontalAlignment=HAlign_Center, verticalAlignment=VAlign_Center}}
  }
  TextBlock_117 {
   widgetClassPath=&TextBlock
   parent=VerticalBox_7
   widget=TextBlock_117
   properties {
    justification=Center
    text=All CIDs
    font {
     bForceMonospaced=false
     bMaterialIsStencil=false
     fontMaterial=null
     fontObject=&Roboto
     letterSpacing=0
     monospacedWidth=1
     size=7.5
     skewAmount=0
     typefaceFontName=Bold
     outlineSettings {
      bApplyOutlineToDropShadows=false
      bMiteredCorners=false
      bSeparateFillAlpha=false
      outlineMaterial=null
      outlineSize=0
      outlineColor=<rgba 0 0 0 1>
     }
    }
   }
   slot {class=VerticalBoxSlot}
  }
  SizeBox_0 {
   widgetClassPath=&SizeBox
   parent=HorizontalBox_4
   widget=SizeBox_0
   properties {bOverride_MaxDesiredHeight=true, maxDesiredHeight=90}
   slot {class=HorizontalBoxSlot, properties {verticalAlignment=VAlign_Top}}
  }
  EditorUtilityScrollBox_1 {widgetClassPath=&EditorUtilityScrollBox, parent=SizeBox_0, widget=EditorUtilityScrollBox_1, slot {class=SizeBoxSlot}}
  EUW_CosmeticPickerv2_Variants {
   widgetClassPath=/Game/NeverCook/ArtTools/Tech_Art/CosmeticPicker/EUW_CosmeticPickerv2_Variants
   bIsVariable=true
   parent=EditorUtilityScrollBox_1
   widget=EUW_CosmeticPickerv2_Variants
   slot {class=ScrollBoxSlot, properties {verticalAlignment=VAlign_Top}}
  }
  HorizontalBox_10 {
   widgetClassPath=&HorizontalBox
   parent=VerticalBox_2
   widget=HorizontalBox_10
   slot {class=VerticalBoxSlot, properties {padding=<ltrb 0 5 0 0>}}
  }
  TextBlock_20 {
   widgetClassPath=&TextBlock
   parent=HorizontalBox_10
   widget=TextBlock_20
   properties {
    text="Sparks ▼"
    font {
     bForceMonospaced=false
     bMaterialIsStencil=false
     fontMaterial=null
     fontObject=&Roboto
     letterSpacing=0
     monospacedWidth=1
     size=10.5
     skewAmount=0
     typefaceFontName=Bold
     outlineSettings {
      bApplyOutlineToDropShadows=false
      bMiteredCorners=false
      bSeparateFillAlpha=false
      outlineMaterial=null
      outlineSize=0
      outlineColor=<rgba 0 0 0 1>
     }
    }
   }
   slot {class=HorizontalBoxSlot}
  }
  Checkbox_ShowSparks {
   widgetClassPath=&EditorUtilityCheckBox
   bIsVariable=true
   parent=HorizontalBox_10
   widget=Checkbox_ShowSparks
   properties {widgetStyle {
     checkBoxType=CheckBox
     backgroundHoveredImage {
      bIsDynamicallyLoaded=false
      drawAs=RoundedBox
      imageType=NoImage
      mirroring=NoMirror
      resourceName=None
      resourceObject=null
      tiling=NoTile
      imageSize=<18 18>
      margin=<ltrb 0 0 0 0>
      outlineSettings {
       bUseBrushTransparency=false
       roundingType=FixedRadius
       width=1
       color {colorUseRule=UseColor_ColorTable, specifiedColor=<rgba 0 0 0 0>}
       cornerRadii=<3 3 3 3>
      }
      tintColor {colorUseRule=UseColor_Specified, specifiedColor=<rgba 0.0048 0.0048 0.0048 1>}
      uVRegion {bIsValid=false, max=<0 0>, min=<0 0>}
     }
     backgroundImage {
      bIsDynamicallyLoaded=false
      drawAs=RoundedBox
      imageType=NoImage
      mirroring=NoMirror
      resourceName=None
      resourceObject=null
      tiling=NoTile
      imageSize=<18 18>
      margin=<ltrb 0 0 0 0>
      outlineSettings {
       bUseBrushTransparency=false
       roundingType=FixedRadius
       width=1
       color {colorUseRule=UseColor_ColorTable, specifiedColor=<rgba 0 0 0 0>}
       cornerRadii=<3 3 3 3>
      }
      tintColor {colorUseRule=UseColor_Specified, specifiedColor=<rgba 0.0048 0.0048 0.0048 1>}
      uVRegion {bIsValid=false, max=<0 0>, min=<0 0>}
     }
     backgroundPressedImage {
      bIsDynamicallyLoaded=false
      drawAs=RoundedBox
      imageType=NoImage
      mirroring=NoMirror
      resourceName=None
      resourceObject=null
      tiling=NoTile
      imageSize=<18 18>
      margin=<ltrb 0 0 0 0>
      outlineSettings {
       bUseBrushTransparency=false
       roundingType=FixedRadius
       width=1
       color {colorUseRule=UseColor_ColorTable, specifiedColor=<rgba 0 0 0 0>}
       cornerRadii=<3 3 3 3>
      }
      tintColor {colorUseRule=UseColor_Specified, specifiedColor=<rgba 0.0048 0.0048 0.0048 1>}
      uVRegion {bIsValid=false, max=<0 0>, min=<0 0>}
     }
     borderBackgroundColor {colorUseRule=UseColor_Specified, specifiedColor=<rgba 1 1 1 1>}
     checkedForeground {colorUseRule=UseColor_Specified, specifiedColor=<rgba 1 1 1 1>}
     checkedHoveredForeground {colorUseRule=UseColor_Specified, specifiedColor=<rgba 1 1 1 1>}
     checkedHoveredImage {
      bIsDynamicallyLoaded=false
      drawAs=Image
      imageType=Vector
      mirroring=NoMirror
      resourceName=../../../Engine/Content/Slate/Starship/CoreWidgets/CheckBox/check.svg
      resourceObject=null
      tiling=NoTile
      imageSize=<16 16>
      margin=<ltrb 0 0 0 0>
      outlineSettings {
       bUseBrushTransparency=false
       roundingType=HalfHeightRadius
       width=0
       color {colorUseRule=UseColor_Specified, specifiedColor=<rgba 0 0 0 0>}
       cornerRadii=<0 0 0 0>
      }
      tintColor {colorUseRule=UseColor_Specified, specifiedColor=<rgba 0.0044 0.2384 1 1>}
      uVRegion {bIsValid=false, max=<0 0>, min=<0 0>}
     }
     checkedImage {
      bIsDynamicallyLoaded=false
      drawAs=Image
      imageType=Vector
      mirroring=NoMirror
      resourceName=../../../Engine/Content/Slate/Starship/CoreWidgets/CheckBox/check.svg
      resourceObject=null
      tiling=NoTile
      imageSize=<16 16>
      margin=<ltrb 0 0 0 0>
      outlineSettings {
       bUseBrushTransparency=false
       roundingType=HalfHeightRadius
       width=0
       color {colorUseRule=UseColor_Specified, specifiedColor=<rgba 0 0 0 0>}
       cornerRadii=<0 0 0 0>
      }
      tintColor {colorUseRule=UseColor_Specified, specifiedColor=<rgba 0 0.162 0.7454 1>}
      uVRegion {bIsValid=false, max=<0 0>, min=<0 0>}
     }
     checkedPressedForeground {colorUseRule=UseColor_Specified, specifiedColor=<rgba 1 1 1 1>}
     checkedPressedImage {
      bIsDynamicallyLoaded=false
      drawAs=Image
      imageType=Vector
      mirroring=NoMirror
      resourceName=../../../Engine/Content/Slate/Starship/CoreWidgets/CheckBox/check.svg
      resourceObject=null
      tiling=NoTile
      imageSize=<16 16>
      margin=<ltrb 0 0 0 0>
      outlineSettings {
       bUseBrushTransparency=false
       roundingType=HalfHeightRadius
       width=0
       color {colorUseRule=UseColor_Specified, specifiedColor=<rgba 0 0 0 0>}
       cornerRadii=<0 0 0 0>
      }
      tintColor {colorUseRule=UseColor_Specified, specifiedColor=<rgba 0 0.0802 0.3515 1>}
      uVRegion {bIsValid=false, max=<0 0>, min=<0 0>}
     }
     checkedSlateSound {resourceObject=null}
     foregroundColor {colorUseRule=UseColor_Specified, specifiedColor=<rgba 1 1 1 1>}
     hoveredForeground {colorUseRule=UseColor_Specified, specifiedColor=<rgba 1 1 1 1>}
     hoveredSlateSound {resourceObject=null}
     padding=<ltrb 2 0 0 0>
     pressedForeground {colorUseRule=UseColor_Specified, specifiedColor=<rgba 1 1 1 1>}
     uncheckedHoveredImage {
      bIsDynamicallyLoaded=false
      drawAs=NoDrawType
      imageType=NoImage
      mirroring=NoMirror
      resourceName=None
      resourceObject=null
      tiling=NoTile
      imageSize=<0 0>
      margin=<ltrb 0 0 0 0>
      outlineSettings {
       bUseBrushTransparency=false
       roundingType=HalfHeightRadius
       width=0
       color {colorUseRule=UseColor_Specified, specifiedColor=<rgba 0 0 0 0>}
       cornerRadii=<0 0 0 0>
      }
      tintColor {colorUseRule=UseColor_Specified, specifiedColor=<rgba 1 1 1 1>}
      uVRegion {bIsValid=false, max=<0 0>, min=<0 0>}
     }
     uncheckedImage {
      bIsDynamicallyLoaded=false
      drawAs=NoDrawType
      imageType=NoImage
      mirroring=NoMirror
      resourceName=None
      resourceObject=null
      tiling=NoTile
      imageSize=<0 0>
      margin=<ltrb 0 0 0 0>
      outlineSettings {
       bUseBrushTransparency=false
       roundingType=HalfHeightRadius
       width=0
       color {colorUseRule=UseColor_Specified, specifiedColor=<rgba 0 0 0 0>}
       cornerRadii=<0 0 0 0>
      }
      tintColor {colorUseRule=UseColor_Specified, specifiedColor=<rgba 1 1 1 1>}
      uVRegion {bIsValid=false, max=<0 0>, min=<0 0>}
     }
     uncheckedPressedImage {
      bIsDynamicallyLoaded=false
      drawAs=NoDrawType
      imageType=NoImage
      mirroring=NoMirror
      resourceName=None
      resourceObject=null
      tiling=NoTile
      imageSize=<0 0>
      margin=<ltrb 0 0 0 0>
      outlineSettings {
       bUseBrushTransparency=false
       roundingType=HalfHeightRadius
       width=0
       color {colorUseRule=UseColor_Specified, specifiedColor=<rgba 0 0 0 0>}
       cornerRadii=<0 0 0 0>
      }
      tintColor {colorUseRule=UseColor_Specified, specifiedColor=<rgba 1 1 1 1>}
      uVRegion {bIsValid=false, max=<0 0>, min=<0 0>}
     }
     uncheckedSlateSound {resourceObject=null}
     undeterminedForeground {colorUseRule=UseColor_Specified, specifiedColor=<rgba 1 1 1 1>}
     undeterminedHoveredImage {
      bIsDynamicallyLoaded=false
      drawAs=Image
      imageType=Vector
      mirroring=NoMirror
      resourceName=../../../Engine/Content/Slate/Starship/CoreWidgets/CheckBox/indeterminate.svg
      resourceObject=null
      tiling=NoTile
      imageSize=<16 16>
      margin=<ltrb 0 0 0 0>
      outlineSettings {
       bUseBrushTransparency=false
       roundingType=HalfHeightRadius
       width=0
       color {colorUseRule=UseColor_Specified, specifiedColor=<rgba 0 0 0 0>}
       cornerRadii=<0 0 0 0>
      }
      tintColor {colorUseRule=UseColor_Specified, specifiedColor=<rgba 0.0044 0.2384 1 1>}
      uVRegion {bIsValid=false, max=<0 0>, min=<0 0>}
     }
     undeterminedImage {
      bIsDynamicallyLoaded=false
      drawAs=Image
      imageType=Vector
      mirroring=NoMirror
      resourceName=../../../Engine/Content/Slate/Starship/CoreWidgets/CheckBox/indeterminate.svg
      resourceObject=null
      tiling=NoTile
      imageSize=<16 16>
      margin=<ltrb 0 0 0 0>
      outlineSettings {
       bUseBrushTransparency=false
       roundingType=HalfHeightRadius
       width=0
       color {colorUseRule=UseColor_Specified, specifiedColor=<rgba 0 0 0 0>}
       cornerRadii=<0 0 0 0>
      }
      tintColor {colorUseRule=UseColor_Specified, specifiedColor=<rgba 0 0.162 0.7454 1>}
      uVRegion {bIsValid=false, max=<0 0>, min=<0 0>}
     }
     undeterminedPressedImage {
      bIsDynamicallyLoaded=false
      drawAs=Image
      imageType=Vector
      mirroring=NoMirror
      resourceName=../../../Engine/Content/Slate/Starship/CoreWidgets/CheckBox/indeterminate.svg
      resourceObject=null
      tiling=NoTile
      imageSize=<16 16>
      margin=<ltrb 0 0 0 0>
      outlineSettings {
       bUseBrushTransparency=false
       roundingType=HalfHeightRadius
       width=0
       color {colorUseRule=UseColor_Specified, specifiedColor=<rgba 0 0 0 0>}
       cornerRadii=<0 0 0 0>
      }
      tintColor {colorUseRule=UseColor_Specified, specifiedColor=<rgba 0 0.0802 0.3515 1>}
      uVRegion {bIsValid=false, max=<0 0>, min=<0 0>}
     }
    }}
   slot {class=HorizontalBoxSlot, properties {padding=<ltrb 3 0 0 0>}}
  }
  WrapBox_SparksID {
   widgetClassPath=&WrapBox
   bIsVariable=true
   parent=VerticalBox_2
   widget=WrapBox_SparksID
   slot {class=VerticalBoxSlot, properties {verticalAlignment=VAlign_Top}}
  }
  HorizontalBox_15 {
   widgetClassPath=&HorizontalBox
   parent=WrapBox_SparksID
   widget=HorizontalBox_15
   slot {class=WrapBoxSlot, properties {padding=<ltrb 10 0 0 0>}}
  }
  Button_AddSparks_1 {
   widgetClassPath=&EditorUtilityButton
   parent=HorizontalBox_15
   widget=Button_AddSparks_1
   properties {widgetStyle {
     clickedSlateSound {resourceObject=null}
     disabled {
      bIsDynamicallyLoaded=false
      drawAs=RoundedBox
      imageType=NoImage
      mirroring=NoMirror
      resourceName=None
      resourceObject=null
      tiling=NoTile
      imageSize=<0 0>
      margin=<ltrb 0 0 0 0>
      outlineSettings {
       bUseBrushTransparency=false
       roundingType=FixedRadius
       width=1
       color {colorUseRule=UseColor_ColorTable, specifiedColor=<rgba 0 0 0 0>}
       cornerRadii=<4 4 4 4>
      }
      tintColor {colorUseRule=UseColor_Specified, specifiedColor=<rgba 0.0395 0.0395 0.0395 1>}
      uVRegion {bIsValid=false, max=<0 0>, min=<0 0>}
     }
     disabledForeground {colorUseRule=UseColor_ColorTable, specifiedColor=<rgba 0 0 0 0>}
     hovered {
      bIsDynamicallyLoaded=false
      drawAs=RoundedBox
      imageType=NoImage
      mirroring=NoMirror
      resourceName=None
      resourceObject=null
      tiling=NoTile
      imageSize=<32 32>
      margin=<ltrb 0 0 0 0>
      outlineSettings {
       bUseBrushTransparency=false
       roundingType=FixedRadius
       width=1
       color {colorUseRule=UseColor_ColorTable, specifiedColor=<rgba 0 0 0 0>}
       cornerRadii=<4 4 4 4>
      }
      tintColor {colorUseRule=UseColor_Specified, specifiedColor=<rgba 0.0953 0.0953 0.0953 1>}
      uVRegion {bIsValid=false, max=<0 0>, min=<0 0>}
     }
     hoveredForeground {colorUseRule=UseColor_ColorTable, specifiedColor=<rgba 0 0 0 0>}
     hoveredSlateSound {resourceObject=null}
     normal {
      bIsDynamicallyLoaded=false
      drawAs=RoundedBox
      imageType=NoImage
      mirroring=NoMirror
      resourceName=None
      resourceObject=null
      tiling=NoTile
      imageSize=<32 32>
      margin=<ltrb 0 0 0 0>
      outlineSettings {
       bUseBrushTransparency=false
       roundingType=FixedRadius
       width=1
       color {colorUseRule=UseColor_ColorTable, specifiedColor=<rgba 0 0 0 0>}
       cornerRadii=<4 4 4 4>
      }
      tintColor {colorUseRule=UseColor_Specified, specifiedColor=<rgba 0.0395 0.0395 0.0395 1>}
      uVRegion {bIsValid=false, max=<0 0>, min=<0 0>}
     }
     normalForeground {colorUseRule=UseColor_ColorTable, specifiedColor=<rgba 0 0 0 0>}
     normalPadding=<ltrb 5 1.5 5 1.5>
     pressed {
      bIsDynamicallyLoaded=false
      drawAs=RoundedBox
      imageType=NoImage
      mirroring=NoMirror
      resourceName=None
      resourceObject=null
      tiling=NoTile
      imageSize=<32 32>
      margin=<ltrb 0 0 0 0>
      outlineSettings {
       bUseBrushTransparency=false
       roundingType=FixedRadius
       width=1
       color {colorUseRule=UseColor_ColorTable, specifiedColor=<rgba 0 0 0 0>}
       cornerRadii=<4 4 4 4>
      }
      tintColor {colorUseRule=UseColor_Specified, specifiedColor=<rgba 0.0284 0.0284 0.0284 1>}
      uVRegion {bIsValid=false, max=<0 0>, min=<0 0>}
     }
     pressedForeground {colorUseRule=UseColor_ColorTable, specifiedColor=<rgba 0 0 0 0>}
     pressedPadding=<ltrb 5 2.5 5 0.5>
     pressedSlateSound {resourceObject=null}
    }}
   slot {class=HorizontalBoxSlot, properties {padding=<ltrb 10 0 3 0>}}
  }
  TextBlock_32 {
   widgetClassPath=&TextBlock
   parent=Button_AddSparks_1
   widget=TextBlock_32
   properties {
    justification=Center
    text="Equip as\r\nWeapons"
    font {
     bForceMonospaced=false
     bMaterialIsStencil=false
     fontMaterial=null
     fontObject=&Roboto
     letterSpacing=0
     monospacedWidth=1
     size=9
     skewAmount=0
     typefaceFontName=Bold
     outlineSettings {
      bApplyOutlineToDropShadows=false
      bMiteredCorners=false
      bSeparateFillAlpha=false
      outlineMaterial=null
      outlineSize=0
      outlineColor=<rgba 0 0 0 1>
     }
    }
   }
   slot {class=ButtonSlot, properties {padding=<ltrb 0 0 0 0>}}
  }
  Button_RemoveSparks {
   widgetClassPath=&EditorUtilityButton
   parent=HorizontalBox_15
   widget=Button_RemoveSparks
   properties {widgetStyle {
     clickedSlateSound {resourceObject=null}
     disabled {
      bIsDynamicallyLoaded=false
      drawAs=RoundedBox
      imageType=NoImage
      mirroring=NoMirror
      resourceName=None
      resourceObject=null
      tiling=NoTile
      imageSize=<0 0>
      margin=<ltrb 0 0 0 0>
      outlineSettings {
       bUseBrushTransparency=false
       roundingType=FixedRadius
       width=1
       color {colorUseRule=UseColor_ColorTable, specifiedColor=<rgba 0 0 0 0>}
       cornerRadii=<4 4 4 4>
      }
      tintColor {colorUseRule=UseColor_Specified, specifiedColor=<rgba 0.0395 0.0395 0.0395 1>}
      uVRegion {bIsValid=false, max=<0 0>, min=<0 0>}
     }
     disabledForeground {colorUseRule=UseColor_ColorTable, specifiedColor=<rgba 0 0 0 0>}
     hovered {
      bIsDynamicallyLoaded=false
      drawAs=RoundedBox
      imageType=NoImage
      mirroring=NoMirror
      resourceName=None
      resourceObject=null
      tiling=NoTile
      imageSize=<32 32>
      margin=<ltrb 0 0 0 0>
      outlineSettings {
       bUseBrushTransparency=false
       roundingType=FixedRadius
       width=1
       color {colorUseRule=UseColor_ColorTable, specifiedColor=<rgba 0 0 0 0>}
       cornerRadii=<4 4 4 4>
      }
      tintColor {colorUseRule=UseColor_Specified, specifiedColor=<rgba 0.0953 0.0953 0.0953 1>}
      uVRegion {bIsValid=false, max=<0 0>, min=<0 0>}
     }
     hoveredForeground {colorUseRule=UseColor_ColorTable, specifiedColor=<rgba 0 0 0 0>}
     hoveredSlateSound {resourceObject=null}
     normal {
      bIsDynamicallyLoaded=false
      drawAs=RoundedBox
      imageType=NoImage
      mirroring=NoMirror
      resourceName=None
      resourceObject=null
      tiling=NoTile
      imageSize=<32 32>
      margin=<ltrb 0 0 0 0>
      outlineSettings {
       bUseBrushTransparency=false
       roundingType=FixedRadius
       width=1
       color {colorUseRule=UseColor_ColorTable, specifiedColor=<rgba 0 0 0 0>}
       cornerRadii=<4 4 4 4>
      }
      tintColor {colorUseRule=UseColor_Specified, specifiedColor=<rgba 0.0395 0.0395 0.0395 1>}
      uVRegion {bIsValid=false, max=<0 0>, min=<0 0>}
     }
     normalForeground {colorUseRule=UseColor_ColorTable, specifiedColor=<rgba 0 0 0 0>}
     normalPadding=<ltrb 0 0 0 0>
     pressed {
      bIsDynamicallyLoaded=false
      drawAs=RoundedBox
      imageType=NoImage
      mirroring=NoMirror
      resourceName=None
      resourceObject=null
      tiling=NoTile
      imageSize=<32 32>
      margin=<ltrb 0 0 0 0>
      outlineSettings {
       bUseBrushTransparency=false
       roundingType=FixedRadius
       width=1
       color {colorUseRule=UseColor_ColorTable, specifiedColor=<rgba 0 0 0 0>}
       cornerRadii=<4 4 4 4>
      }
      tintColor {colorUseRule=UseColor_Specified, specifiedColor=<rgba 0.0284 0.0284 0.0284 1>}
      uVRegion {bIsValid=false, max=<0 0>, min=<0 0>}
     }
     pressedForeground {colorUseRule=UseColor_ColorTable, specifiedColor=<rgba 0 0 0 0>}
     pressedPadding=<ltrb 0 0 0 0>
     pressedSlateSound {resourceObject=null}
    }}
   slot {class=HorizontalBoxSlot, properties {horizontalAlignment=HAlign_Center, padding=<ltrb 10 0 10 0>}}
  }
  VerticalBox_11 {
   widgetClassPath=&VerticalBox
   parent=Button_RemoveSparks
   widget=VerticalBox_11
   slot {class=ButtonSlot, properties {padding=<ltrb 7 7 7 7>}}
  }
  Image_7 {
   widgetClassPath=&Image
   bIsVariable=true
   parent=VerticalBox_11
   widget=Image_7
   properties {brush {
     bIsDynamicallyLoaded=false
     drawAs=Image
     imageType=FullColor
     mirroring=NoMirror
     resourceName=None
     resourceObject=null
     tiling=NoTile
     imageSize=<20 20>
     margin=<ltrb 0 0 0 0>
     outlineSettings {
      bUseBrushTransparency=false
      roundingType=HalfHeightRadius
      width=0
      color {colorUseRule=UseColor_Specified, specifiedColor=<rgba 0 0 0 0>}
      cornerRadii=<0 0 0 0>
     }
     tintColor {colorUseRule=UseColor_Specified, specifiedColor=<rgba 1 1 1 1>}
     uVRegion {bIsValid=false, max=<0 0>, min=<0 0>}
    }}
   slot {class=VerticalBoxSlot, properties {horizontalAlignment=HAlign_Center, verticalAlignment=VAlign_Center}}
  }
  TextBlock_27 {
   widgetClassPath=&TextBlock
   parent=VerticalBox_11
   widget=TextBlock_27
   properties {
    justification=Center
    text=All IDs
    font {
     bForceMonospaced=false
     bMaterialIsStencil=false
     fontMaterial=null
     fontObject=&Roboto
     letterSpacing=0
     monospacedWidth=1
     size=7.5
     skewAmount=0
     typefaceFontName=Bold
     outlineSettings {
      bApplyOutlineToDropShadows=false
      bMiteredCorners=false
      bSeparateFillAlpha=false
      outlineMaterial=null
      outlineSize=0
      outlineColor=<rgba 0 0 0 1>
     }
    }
   }
   slot {class=VerticalBoxSlot}
  }
  SizeBox_1 {
   widgetClassPath=&SizeBox
   parent=HorizontalBox_15
   widget=SizeBox_1
   properties {bOverride_MaxDesiredHeight=true, maxDesiredHeight=90}
   slot {class=HorizontalBoxSlot, properties {verticalAlignment=VAlign_Top}}
  }
  Button_AddSparks {
   widgetClassPath=&EditorUtilityButton
   parent=HorizontalBox_15
   widget=Button_AddSparks
   properties {widgetStyle {
     clickedSlateSound {resourceObject=null}
     disabled {
      bIsDynamicallyLoaded=false
      drawAs=RoundedBox
      imageType=NoImage
      mirroring=NoMirror
      resourceName=None
      resourceObject=null
      tiling=NoTile
      imageSize=<0 0>
      margin=<ltrb 0 0 0 0>
      outlineSettings {
       bUseBrushTransparency=false
       roundingType=FixedRadius
       width=1
       color {colorUseRule=UseColor_ColorTable, specifiedColor=<rgba 0 0 0 0>}
       cornerRadii=<4 4 4 4>
      }
      tintColor {colorUseRule=UseColor_Specified, specifiedColor=<rgba 0.0395 0.0395 0.0395 1>}
      uVRegion {bIsValid=false, max=<0 0>, min=<0 0>}
     }
     disabledForeground {colorUseRule=UseColor_ColorTable, specifiedColor=<rgba 0 0 0 0>}
     hovered {
      bIsDynamicallyLoaded=false
      drawAs=RoundedBox
      imageType=NoImage
      mirroring=NoMirror
      resourceName=None
      resourceObject=null
      tiling=NoTile
      imageSize=<32 32>
      margin=<ltrb 0 0 0 0>
      outlineSettings {
       bUseBrushTransparency=false
       roundingType=FixedRadius
       width=1
       color {colorUseRule=UseColor_ColorTable, specifiedColor=<rgba 0 0 0 0>}
       cornerRadii=<4 4 4 4>
      }
      tintColor {colorUseRule=UseColor_Specified, specifiedColor=<rgba 0.0953 0.0953 0.0953 1>}
      uVRegion {bIsValid=false, max=<0 0>, min=<0 0>}
     }
     hoveredForeground {colorUseRule=UseColor_ColorTable, specifiedColor=<rgba 0 0 0 0>}
     hoveredSlateSound {resourceObject=null}
     normal {
      bIsDynamicallyLoaded=false
      drawAs=RoundedBox
      imageType=NoImage
      mirroring=NoMirror
      resourceName=None
      resourceObject=null
      tiling=NoTile
      imageSize=<32 32>
      margin=<ltrb 0 0 0 0>
      outlineSettings {
       bUseBrushTransparency=false
       roundingType=FixedRadius
       width=1
       color {colorUseRule=UseColor_ColorTable, specifiedColor=<rgba 0 0 0 0>}
       cornerRadii=<4 4 4 4>
      }
      tintColor {colorUseRule=UseColor_Specified, specifiedColor=<rgba 0.0395 0.0395 0.0395 1>}
      uVRegion {bIsValid=false, max=<0 0>, min=<0 0>}
     }
     normalForeground {colorUseRule=UseColor_ColorTable, specifiedColor=<rgba 0 0 0 0>}
     normalPadding=<ltrb 5 1.5 5 1.5>
     pressed {
      bIsDynamicallyLoaded=false
      drawAs=RoundedBox
      imageType=NoImage
      mirroring=NoMirror
      resourceName=None
      resourceObject=null
      tiling=NoTile
      imageSize=<32 32>
      margin=<ltrb 0 0 0 0>
      outlineSettings {
       bUseBrushTransparency=false
       roundingType=FixedRadius
       width=1
       color {colorUseRule=UseColor_ColorTable, specifiedColor=<rgba 0 0 0 0>}
       cornerRadii=<4 4 4 4>
      }
      tintColor {colorUseRule=UseColor_Specified, specifiedColor=<rgba 0.0284 0.0284 0.0284 1>}
      uVRegion {bIsValid=false, max=<0 0>, min=<0 0>}
     }
     pressedForeground {colorUseRule=UseColor_ColorTable, specifiedColor=<rgba 0 0 0 0>}
     pressedPadding=<ltrb 5 2.5 5 0.5>
     pressedSlateSound {resourceObject=null}
    }}
   slot {class=HorizontalBoxSlot, properties {padding=<ltrb 10 0 3 0>}}
  }
  VerticalBox_12 {
   widgetClassPath=&VerticalBox
   parent=Button_AddSparks
   widget=VerticalBox_12
   slot {class=ButtonSlot, properties {padding=<ltrb 0 0 0 0>}}
  }
  TextBlock_28 {
   widgetClassPath=&TextBlock
   parent=VerticalBox_12
   widget=TextBlock_28
   properties {
    justification=Center
    text=+
    font {
     bForceMonospaced=false
     bMaterialIsStencil=false
     fontMaterial=null
     fontObject=&Roboto
     letterSpacing=0
     monospacedWidth=1
     size=15
     skewAmount=0
     typefaceFontName=Bold
     outlineSettings {
      bApplyOutlineToDropShadows=false
      bMiteredCorners=false
      bSeparateFillAlpha=false
      outlineMaterial=null
      outlineSize=0
      outlineColor=<rgba 0 0 0 1>
     }
    }
   }
   slot {class=VerticalBoxSlot}
  }
  TextBlock_29 {
   widgetClassPath=&TextBlock
   parent=VerticalBox_12
   widget=TextBlock_29
   properties {
    justification=Center
    text=Add ID
    font {
     bForceMonospaced=false
     bMaterialIsStencil=false
     fontMaterial=null
     fontObject=&Roboto
     letterSpacing=0
     monospacedWidth=1
     size=9
     skewAmount=0
     typefaceFontName=Bold
     outlineSettings {
      bApplyOutlineToDropShadows=false
      bMiteredCorners=false
      bSeparateFillAlpha=false
      outlineMaterial=null
      outlineSize=0
      outlineColor=<rgba 0 0 0 1>
     }
    }
   }
   slot {class=VerticalBoxSlot}
  }
  HorizontalBox_13 {
   widgetClassPath=&HorizontalBox
   parent=VerticalBox_2
   widget=HorizontalBox_13
   slot {class=VerticalBoxSlot, properties {padding=<ltrb 0 5 0 0>}}
  }
  TextBlock_26 {
   widgetClassPath=&TextBlock
   parent=HorizontalBox_13
   widget=TextBlock_26
   properties {
    text="Pickaxe ▼"
    font {
     bForceMonospaced=false
     bMaterialIsStencil=false
     fontMaterial=null
     fontObject=&Roboto
     letterSpacing=0
     monospacedWidth=1
     size=10.5
     skewAmount=0
     typefaceFontName=Bold
     outlineSettings {
      bApplyOutlineToDropShadows=false
      bMiteredCorners=false
      bSeparateFillAlpha=false
      outlineMaterial=null
      outlineSize=0
      outlineColor=<rgba 0 0 0 1>
     }
    }
   }
   slot {class=HorizontalBoxSlot}
  }
  Checkbox_ShowPickaxes {
   widgetClassPath=&EditorUtilityCheckBox
   bIsVariable=true
   parent=HorizontalBox_13
   widget=Checkbox_ShowPickaxes
   properties {widgetStyle {
     checkBoxType=CheckBox
     backgroundHoveredImage {
      bIsDynamicallyLoaded=false
      drawAs=RoundedBox
      imageType=NoImage
      mirroring=NoMirror
      resourceName=None
      resourceObject=null
      tiling=NoTile
      imageSize=<18 18>
      margin=<ltrb 0 0 0 0>
      outlineSettings {
       bUseBrushTransparency=false
       roundingType=FixedRadius
       width=1
       color {colorUseRule=UseColor_ColorTable, specifiedColor=<rgba 0 0 0 0>}
       cornerRadii=<3 3 3 3>
      }
      tintColor {colorUseRule=UseColor_Specified, specifiedColor=<rgba 0.0048 0.0048 0.0048 1>}
      uVRegion {bIsValid=false, max=<0 0>, min=<0 0>}
     }
     backgroundImage {
      bIsDynamicallyLoaded=false
      drawAs=RoundedBox
      imageType=NoImage
      mirroring=NoMirror
      resourceName=None
      resourceObject=null
      tiling=NoTile
      imageSize=<18 18>
      margin=<ltrb 0 0 0 0>
      outlineSettings {
       bUseBrushTransparency=false
       roundingType=FixedRadius
       width=1
       color {colorUseRule=UseColor_ColorTable, specifiedColor=<rgba 0 0 0 0>}
       cornerRadii=<3 3 3 3>
      }
      tintColor {colorUseRule=UseColor_Specified, specifiedColor=<rgba 0.0048 0.0048 0.0048 1>}
      uVRegion {bIsValid=false, max=<0 0>, min=<0 0>}
     }
     backgroundPressedImage {
      bIsDynamicallyLoaded=false
      drawAs=RoundedBox
      imageType=NoImage
      mirroring=NoMirror
      resourceName=None
      resourceObject=null
      tiling=NoTile
      imageSize=<18 18>
      margin=<ltrb 0 0 0 0>
      outlineSettings {
       bUseBrushTransparency=false
       roundingType=FixedRadius
       width=1
       color {colorUseRule=UseColor_ColorTable, specifiedColor=<rgba 0 0 0 0>}
       cornerRadii=<3 3 3 3>
      }
      tintColor {colorUseRule=UseColor_Specified, specifiedColor=<rgba 0.0048 0.0048 0.0048 1>}
      uVRegion {bIsValid=false, max=<0 0>, min=<0 0>}
     }
     borderBackgroundColor {colorUseRule=UseColor_Specified, specifiedColor=<rgba 1 1 1 1>}
     checkedForeground {colorUseRule=UseColor_Specified, specifiedColor=<rgba 1 1 1 1>}
     checkedHoveredForeground {colorUseRule=UseColor_Specified, specifiedColor=<rgba 1 1 1 1>}
     checkedHoveredImage {
      bIsDynamicallyLoaded=false
      drawAs=Image
      imageType=Vector
      mirroring=NoMirror
      resourceName=../../../Engine/Content/Slate/Starship/CoreWidgets/CheckBox/check.svg
      resourceObject=null
      tiling=NoTile
      imageSize=<16 16>
      margin=<ltrb 0 0 0 0>
      outlineSettings {
       bUseBrushTransparency=false
       roundingType=HalfHeightRadius
       width=0
       color {colorUseRule=UseColor_Specified, specifiedColor=<rgba 0 0 0 0>}
       cornerRadii=<0 0 0 0>
      }
      tintColor {colorUseRule=UseColor_Specified, specifiedColor=<rgba 0.0044 0.2384 1 1>}
      uVRegion {bIsValid=false, max=<0 0>, min=<0 0>}
     }
     checkedImage {
      bIsDynamicallyLoaded=false
      drawAs=Image
      imageType=Vector
      mirroring=NoMirror
      resourceName=../../../Engine/Content/Slate/Starship/CoreWidgets/CheckBox/check.svg
      resourceObject=null
      tiling=NoTile
      imageSize=<16 16>
      margin=<ltrb 0 0 0 0>
      outlineSettings {
       bUseBrushTransparency=false
       roundingType=HalfHeightRadius
       width=0
       color {colorUseRule=UseColor_Specified, specifiedColor=<rgba 0 0 0 0>}
       cornerRadii=<0 0 0 0>
      }
      tintColor {colorUseRule=UseColor_Specified, specifiedColor=<rgba 0 0.162 0.7454 1>}
      uVRegion {bIsValid=false, max=<0 0>, min=<0 0>}
     }
     checkedPressedForeground {colorUseRule=UseColor_Specified, specifiedColor=<rgba 1 1 1 1>}
     checkedPressedImage {
      bIsDynamicallyLoaded=false
      drawAs=Image
      imageType=Vector
      mirroring=NoMirror
      resourceName=../../../Engine/Content/Slate/Starship/CoreWidgets/CheckBox/check.svg
      resourceObject=null
      tiling=NoTile
      imageSize=<16 16>
      margin=<ltrb 0 0 0 0>
      outlineSettings {
       bUseBrushTransparency=false
       roundingType=HalfHeightRadius
       width=0
       color {colorUseRule=UseColor_Specified, specifiedColor=<rgba 0 0 0 0>}
       cornerRadii=<0 0 0 0>
      }
      tintColor {colorUseRule=UseColor_Specified, specifiedColor=<rgba 0 0.0802 0.3515 1>}
      uVRegion {bIsValid=false, max=<0 0>, min=<0 0>}
     }
     checkedSlateSound {resourceObject=null}
     foregroundColor {colorUseRule=UseColor_Specified, specifiedColor=<rgba 1 1 1 1>}
     hoveredForeground {colorUseRule=UseColor_Specified, specifiedColor=<rgba 1 1 1 1>}
     hoveredSlateSound {resourceObject=null}
     padding=<ltrb 2 0 0 0>
     pressedForeground {colorUseRule=UseColor_Specified, specifiedColor=<rgba 1 1 1 1>}
     uncheckedHoveredImage {
      bIsDynamicallyLoaded=false
      drawAs=NoDrawType
      imageType=NoImage
      mirroring=NoMirror
      resourceName=None
      resourceObject=null
      tiling=NoTile
      imageSize=<0 0>
      margin=<ltrb 0 0 0 0>
      outlineSettings {
       bUseBrushTransparency=false
       roundingType=HalfHeightRadius
       width=0
       color {colorUseRule=UseColor_Specified, specifiedColor=<rgba 0 0 0 0>}
       cornerRadii=<0 0 0 0>
      }
      tintColor {colorUseRule=UseColor_Specified, specifiedColor=<rgba 1 1 1 1>}
      uVRegion {bIsValid=false, max=<0 0>, min=<0 0>}
     }
     uncheckedImage {
      bIsDynamicallyLoaded=false
      drawAs=NoDrawType
      imageType=NoImage
      mirroring=NoMirror
      resourceName=None
      resourceObject=null
      tiling=NoTile
      imageSize=<0 0>
      margin=<ltrb 0 0 0 0>
      outlineSettings {
       bUseBrushTransparency=false
       roundingType=HalfHeightRadius
       width=0
       color {colorUseRule=UseColor_Specified, specifiedColor=<rgba 0 0 0 0>}
       cornerRadii=<0 0 0 0>
      }
      tintColor {colorUseRule=UseColor_Specified, specifiedColor=<rgba 1 1 1 1>}
      uVRegion {bIsValid=false, max=<0 0>, min=<0 0>}
     }
     uncheckedPressedImage {
      bIsDynamicallyLoaded=false
      drawAs=NoDrawType
      imageType=NoImage
      mirroring=NoMirror
      resourceName=None
      resourceObject=null
      tiling=NoTile
      imageSize=<0 0>
      margin=<ltrb 0 0 0 0>
      outlineSettings {
       bUseBrushTransparency=false
       roundingType=HalfHeightRadius
       width=0
       color {colorUseRule=UseColor_Specified, specifiedColor=<rgba 0 0 0 0>}
       cornerRadii=<0 0 0 0>
      }
      tintColor {colorUseRule=UseColor_Specified, specifiedColor=<rgba 1 1 1 1>}
      uVRegion {bIsValid=false, max=<0 0>, min=<0 0>}
     }
     uncheckedSlateSound {resourceObject=null}
     undeterminedForeground {colorUseRule=UseColor_Specified, specifiedColor=<rgba 1 1 1 1>}
     undeterminedHoveredImage {
      bIsDynamicallyLoaded=false
      drawAs=Image
      imageType=Vector
      mirroring=NoMirror
      resourceName=../../../Engine/Content/Slate/Starship/CoreWidgets/CheckBox/indeterminate.svg
      resourceObject=null
      tiling=NoTile
      imageSize=<16 16>
      margin=<ltrb 0 0 0 0>
      outlineSettings {
       bUseBrushTransparency=false
       roundingType=HalfHeightRadius
       width=0
       color {colorUseRule=UseColor_Specified, specifiedColor=<rgba 0 0 0 0>}
       cornerRadii=<0 0 0 0>
      }
      tintColor {colorUseRule=UseColor_Specified, specifiedColor=<rgba 0.0044 0.2384 1 1>}
      uVRegion {bIsValid=false, max=<0 0>, min=<0 0>}
     }
     undeterminedImage {
      bIsDynamicallyLoaded=false
      drawAs=Image
      imageType=Vector
      mirroring=NoMirror
      resourceName=../../../Engine/Content/Slate/Starship/CoreWidgets/CheckBox/indeterminate.svg
      resourceObject=null
      tiling=NoTile
      imageSize=<16 16>
      margin=<ltrb 0 0 0 0>
      outlineSettings {
       bUseBrushTransparency=false
       roundingType=HalfHeightRadius
       width=0
       color {colorUseRule=UseColor_Specified, specifiedColor=<rgba 0 0 0 0>}
       cornerRadii=<0 0 0 0>
      }
      tintColor {colorUseRule=UseColor_Specified, specifiedColor=<rgba 0 0.162 0.7454 1>}
      uVRegion {bIsValid=false, max=<0 0>, min=<0 0>}
     }
     undeterminedPressedImage {
      bIsDynamicallyLoaded=false
      drawAs=Image
      imageType=Vector
      mirroring=NoMirror
      resourceName=../../../Engine/Content/Slate/Starship/CoreWidgets/CheckBox/indeterminate.svg
      resourceObject=null
      tiling=NoTile
      imageSize=<16 16>
      margin=<ltrb 0 0 0 0>
      outlineSettings {
       bUseBrushTransparency=false
       roundingType=HalfHeightRadius
       width=0
       color {colorUseRule=UseColor_Specified, specifiedColor=<rgba 0 0 0 0>}
       cornerRadii=<0 0 0 0>
      }
      tintColor {colorUseRule=UseColor_Specified, specifiedColor=<rgba 0 0.0802 0.3515 1>}
      uVRegion {bIsValid=false, max=<0 0>, min=<0 0>}
     }
    }}
   slot {class=HorizontalBoxSlot, properties {padding=<ltrb 3 0 0 0>}}
  }
  WrapBox_PickaxeID {
   widgetClassPath=&WrapBox
   bIsVariable=true
   parent=VerticalBox_2
   widget=WrapBox_PickaxeID
   slot {class=VerticalBoxSlot, properties {verticalAlignment=VAlign_Top}}
  }
  HorizontalBox_12 {
   widgetClassPath=&HorizontalBox
   parent=WrapBox_PickaxeID
   widget=HorizontalBox_12
   slot {class=WrapBoxSlot, properties {padding=<ltrb 10 0 0 0>}}
  }
  Button_AutoWalk_1 {
   widgetClassPath=&EditorUtilityButton
   parent=HorizontalBox_12
   widget=Button_AutoWalk_1
   properties {
    visibility=Collapsed
    widgetStyle {
     clickedSlateSound {resourceObject=null}
     disabled {
      bIsDynamicallyLoaded=false
      drawAs=RoundedBox
      imageType=NoImage
      mirroring=NoMirror
      resourceName=None
      resourceObject=null
      tiling=NoTile
      imageSize=<0 0>
      margin=<ltrb 0 0 0 0>
      outlineSettings {
       bUseBrushTransparency=false
       roundingType=FixedRadius
       width=1
       color {colorUseRule=UseColor_ColorTable, specifiedColor=<rgba 0 0 0 0>}
       cornerRadii=<4 4 4 4>
      }
      tintColor {colorUseRule=UseColor_Specified, specifiedColor=<rgba 0.0395 0.0395 0.0395 1>}
      uVRegion {bIsValid=false, max=<0 0>, min=<0 0>}
     }
     disabledForeground {colorUseRule=UseColor_ColorTable, specifiedColor=<rgba 0 0 0 0>}
     hovered {
      bIsDynamicallyLoaded=false
      drawAs=RoundedBox
      imageType=NoImage
      mirroring=NoMirror
      resourceName=None
      resourceObject=null
      tiling=NoTile
      imageSize=<32 32>
      margin=<ltrb 0 0 0 0>
      outlineSettings {
       bUseBrushTransparency=false
       roundingType=FixedRadius
       width=1
       color {colorUseRule=UseColor_ColorTable, specifiedColor=<rgba 0 0 0 0>}
       cornerRadii=<4 4 4 4>
      }
      tintColor {colorUseRule=UseColor_Specified, specifiedColor=<rgba 0.0953 0.0953 0.0953 1>}
      uVRegion {bIsValid=false, max=<0 0>, min=<0 0>}
     }
     hoveredForeground {colorUseRule=UseColor_ColorTable, specifiedColor=<rgba 0 0 0 0>}
     hoveredSlateSound {resourceObject=null}
     normal {
      bIsDynamicallyLoaded=false
      drawAs=RoundedBox
      imageType=NoImage
      mirroring=NoMirror
      resourceName=None
      resourceObject=null
      tiling=NoTile
      imageSize=<32 32>
      margin=<ltrb 0 0 0 0>
      outlineSettings {
       bUseBrushTransparency=false
       roundingType=FixedRadius
       width=1
       color {colorUseRule=UseColor_ColorTable, specifiedColor=<rgba 0 0 0 0>}
       cornerRadii=<4 4 4 4>
      }
      tintColor {colorUseRule=UseColor_Specified, specifiedColor=<rgba 0.0395 0.0395 0.0395 1>}
      uVRegion {bIsValid=false, max=<0 0>, min=<0 0>}
     }
     normalForeground {colorUseRule=UseColor_ColorTable, specifiedColor=<rgba 0 0 0 0>}
     normalPadding=<ltrb 4 1.5 4 1.5>
     pressed {
      bIsDynamicallyLoaded=false
      drawAs=RoundedBox
      imageType=NoImage
      mirroring=NoMirror
      resourceName=None
      resourceObject=null
      tiling=NoTile
      imageSize=<32 32>
      margin=<ltrb 0 0 0 0>
      outlineSettings {
       bUseBrushTransparency=false
       roundingType=FixedRadius
       width=1
       color {colorUseRule=UseColor_ColorTable, specifiedColor=<rgba 0 0 0 0>}
       cornerRadii=<4 4 4 4>
      }
      tintColor {colorUseRule=UseColor_Specified, specifiedColor=<rgba 0.0284 0.0284 0.0284 1>}
      uVRegion {bIsValid=false, max=<0 0>, min=<0 0>}
     }
     pressedForeground {colorUseRule=UseColor_ColorTable, specifiedColor=<rgba 0 0 0 0>}
     pressedPadding=<ltrb 4 2.5 4 0.5>
     pressedSlateSound {resourceObject=null}
    }
   }
   slot {class=HorizontalBoxSlot}
  }
  CanvasPanel {
   widgetClassPath=&CanvasPanel
   parent=Button_AutoWalk_1
   widget=CanvasPanel
   slot {class=ButtonSlot, properties {horizontalAlignment=HAlign_Fill, verticalAlignment=VAlign_Fill, padding=<ltrb 0 0 0 0>}}
  }
  TextBlock_30 {
   widgetClassPath=&TextBlock
   parent=CanvasPanel
   widget=TextBlock_30
   properties {
    text=Auto Swing
    font {
     bForceMonospaced=false
     bMaterialIsStencil=false
     fontMaterial=null
     fontObject=&Roboto
     letterSpacing=0
     monospacedWidth=1
     size=9
     skewAmount=0
     typefaceFontName=Bold
     outlineSettings {
      bApplyOutlineToDropShadows=false
      bMiteredCorners=false
      bSeparateFillAlpha=false
      outlineMaterial=null
      outlineSize=0
      outlineColor=<rgba 0 0 0 1>
     }
    }
   }
   slot {
    class=CanvasPanelSlot
    properties {bAutoSize=true, layoutData {alignment=<0.5 1>, anchors {maximum=<0.5 0.5>, minimum=<0.5 0.5>}, offsets=<ltrb 0 32.3002 0 0>}}
   }
  }
  EditorUtilityCircularThrobber {
   widgetClassPath=&EditorUtilityCircularThrobber
   bIsVariable=true
   parent=CanvasPanel
   widget=EditorUtilityCircularThrobber
   properties {
    numberOfPieces=11
    period=3
    radius=21.4045
    image {
     bIsDynamicallyLoaded=false
     drawAs=Image
     imageType=FullColor
     mirroring=NoMirror
     resourceName=../../../Engine/Content/Slate/Common/Throbber_Piece.png
     resourceObject=null
     tiling=NoTile
     imageSize=<5 5>
     margin=<ltrb 0 0 0 0>
     outlineSettings {
      bUseBrushTransparency=false
      roundingType=HalfHeightRadius
      width=0
      color {colorUseRule=UseColor_Specified, specifiedColor=<rgba 0 0 0 0>}
      cornerRadii=<0 0 0 0>
     }
     tintColor {colorUseRule=UseColor_Specified, specifiedColor=<rgba 1 0.8495 0.0192 1>}
     uVRegion {bIsValid=false, max=<0 0>, min=<0 0>}
    }
   }
   slot {
    class=CanvasPanelSlot
    properties {
     bAutoSize=true
     layoutData {alignment=<0.5 0.5>, anchors {maximum=<0.5 0.5>, minimum=<0.5 0.5>}, offsets=<ltrb -1.2263 -8.2296 102.5025 7.978>}
    }
   }
  }
  Image_8 {
   widgetClassPath=&Image
   bIsVariable=true
   parent=CanvasPanel
   widget=Image_8
   properties {
    brush {
     bIsDynamicallyLoaded=false
     drawAs=Image
     imageType=FullColor
     mirroring=NoMirror
     resourceName=None
     resourceObject=/Game/UI/Foundation/Textures/Icons/Locker/EmptyCategoryIcons/T_UI_Empty_Pickaxe
     tiling=NoTile
     imageSize=<30 30>
     margin=<ltrb 0 0 0 0>
     outlineSettings {
      bUseBrushTransparency=false
      roundingType=HalfHeightRadius
      width=0
      color {colorUseRule=UseColor_Specified, specifiedColor=<rgba 0 0 0 0>}
      cornerRadii=<0 0 0 0>
     }
     tintColor {colorUseRule=UseColor_Specified, specifiedColor=<rgba 1 1 1 1>}
     uVRegion {bIsValid=false, max=<0 0>, min=<0 0>}
    }
    renderTransform {angle=0, scale=<1.5 1.5>, shear=<0 0>, translation=<0 0>}
   }
   slot {
    class=CanvasPanelSlot
    properties {bAutoSize=true, layoutData {alignment=<0.5 0.5>, anchors {maximum=<0.5 0.5>, minimum=<0.5 0.5>}, offsets=<ltrb -0.3642 -9.5 100 30>}}
   }
  }
  Button_RemovePickaxes {
   widgetClassPath=&EditorUtilityButton
   parent=HorizontalBox_12
   widget=Button_RemovePickaxes
   properties {widgetStyle {
     clickedSlateSound {resourceObject=null}
     disabled {
      bIsDynamicallyLoaded=false
      drawAs=RoundedBox
      imageType=NoImage
      mirroring=NoMirror
      resourceName=None
      resourceObject=null
      tiling=NoTile
      imageSize=<0 0>
      margin=<ltrb 0 0 0 0>
      outlineSettings {
       bUseBrushTransparency=false
       roundingType=FixedRadius
       width=1
       color {colorUseRule=UseColor_ColorTable, specifiedColor=<rgba 0 0 0 0>}
       cornerRadii=<4 4 4 4>
      }
      tintColor {colorUseRule=UseColor_Specified, specifiedColor=<rgba 0.0395 0.0395 0.0395 1>}
      uVRegion {bIsValid=false, max=<0 0>, min=<0 0>}
     }
     disabledForeground {colorUseRule=UseColor_ColorTable, specifiedColor=<rgba 0 0 0 0>}
     hovered {
      bIsDynamicallyLoaded=false
      drawAs=RoundedBox
      imageType=NoImage
      mirroring=NoMirror
      resourceName=None
      resourceObject=null
      tiling=NoTile
      imageSize=<32 32>
      margin=<ltrb 0 0 0 0>
      outlineSettings {
       bUseBrushTransparency=false
       roundingType=FixedRadius
       width=1
       color {colorUseRule=UseColor_ColorTable, specifiedColor=<rgba 0 0 0 0>}
       cornerRadii=<4 4 4 4>
      }
      tintColor {colorUseRule=UseColor_Specified, specifiedColor=<rgba 0.0953 0.0953 0.0953 1>}
      uVRegion {bIsValid=false, max=<0 0>, min=<0 0>}
     }
     hoveredForeground {colorUseRule=UseColor_ColorTable, specifiedColor=<rgba 0 0 0 0>}
     hoveredSlateSound {resourceObject=null}
     normal {
      bIsDynamicallyLoaded=false
      drawAs=RoundedBox
      imageType=NoImage
      mirroring=NoMirror
      resourceName=None
      resourceObject=null
      tiling=NoTile
      imageSize=<32 32>
      margin=<ltrb 0 0 0 0>
      outlineSettings {
       bUseBrushTransparency=false
       roundingType=FixedRadius
       width=1
       color {colorUseRule=UseColor_ColorTable, specifiedColor=<rgba 0 0 0 0>}
       cornerRadii=<4 4 4 4>
      }
      tintColor {colorUseRule=UseColor_Specified, specifiedColor=<rgba 0.0395 0.0395 0.0395 1>}
      uVRegion {bIsValid=false, max=<0 0>, min=<0 0>}
     }
     normalForeground {colorUseRule=UseColor_ColorTable, specifiedColor=<rgba 0 0 0 0>}
     normalPadding=<ltrb 0 0 0 0>
     pressed {
      bIsDynamicallyLoaded=false
      drawAs=RoundedBox
      imageType=NoImage
      mirroring=NoMirror
      resourceName=None
      resourceObject=null
      tiling=NoTile
      imageSize=<32 32>
      margin=<ltrb 0 0 0 0>
      outlineSettings {
       bUseBrushTransparency=false
       roundingType=FixedRadius
       width=1
       color {colorUseRule=UseColor_ColorTable, specifiedColor=<rgba 0 0 0 0>}
       cornerRadii=<4 4 4 4>
      }
      tintColor {colorUseRule=UseColor_Specified, specifiedColor=<rgba 0.0284 0.0284 0.0284 1>}
      uVRegion {bIsValid=false, max=<0 0>, min=<0 0>}
     }
     pressedForeground {colorUseRule=UseColor_ColorTable, specifiedColor=<rgba 0 0 0 0>}
     pressedPadding=<ltrb 0 0 0 0>
     pressedSlateSound {resourceObject=null}
    }}
   slot {class=HorizontalBoxSlot, properties {horizontalAlignment=HAlign_Center, padding=<ltrb 10 0 10 0>}}
  }
  VerticalBox_9 {
   widgetClassPath=&VerticalBox
   parent=Button_RemovePickaxes
   widget=VerticalBox_9
   slot {class=ButtonSlot, properties {padding=<ltrb 7 7 7 7>}}
  }
  Image_6 {
   widgetClassPath=&Image
   bIsVariable=true
   parent=VerticalBox_9
   widget=Image_6
   properties {brush {
     bIsDynamicallyLoaded=false
     drawAs=Image
     imageType=FullColor
     mirroring=NoMirror
     resourceName=None
     resourceObject=null
     tiling=NoTile
     imageSize=<20 20>
     margin=<ltrb 0 0 0 0>
     outlineSettings {
      bUseBrushTransparency=false
      roundingType=HalfHeightRadius
      width=0
      color {colorUseRule=UseColor_Specified, specifiedColor=<rgba 0 0 0 0>}
      cornerRadii=<0 0 0 0>
     }
     tintColor {colorUseRule=UseColor_Specified, specifiedColor=<rgba 1 1 1 1>}
     uVRegion {bIsValid=false, max=<0 0>, min=<0 0>}
    }}
   slot {class=VerticalBoxSlot, properties {horizontalAlignment=HAlign_Center, verticalAlignment=VAlign_Center}}
  }
  TextBlock_21 {
   widgetClassPath=&TextBlock
   parent=VerticalBox_9
   widget=TextBlock_21
   properties {
    justification=Center
    text=All IDs
    font {
     bForceMonospaced=false
     bMaterialIsStencil=false
     fontMaterial=null
     fontObject=&Roboto
     letterSpacing=0
     monospacedWidth=1
     size=7.5
     skewAmount=0
     typefaceFontName=Bold
     outlineSettings {
      bApplyOutlineToDropShadows=false
      bMiteredCorners=false
      bSeparateFillAlpha=false
      outlineMaterial=null
      outlineSize=0
      outlineColor=<rgba 0 0 0 1>
     }
    }
   }
   slot {class=VerticalBoxSlot}
  }
  SizeBox {
   widgetClassPath=&SizeBox
   parent=HorizontalBox_12
   widget=SizeBox
   properties {bOverride_MaxDesiredHeight=true, maxDesiredHeight=90}
   slot {class=HorizontalBoxSlot, properties {verticalAlignment=VAlign_Top}}
  }
  Button_AddPickaxes {
   widgetClassPath=&EditorUtilityButton
   parent=HorizontalBox_12
   widget=Button_AddPickaxes
   properties {widgetStyle {
     clickedSlateSound {resourceObject=null}
     disabled {
      bIsDynamicallyLoaded=false
      drawAs=RoundedBox
      imageType=NoImage
      mirroring=NoMirror
      resourceName=None
      resourceObject=null
      tiling=NoTile
      imageSize=<0 0>
      margin=<ltrb 0 0 0 0>
      outlineSettings {
       bUseBrushTransparency=false
       roundingType=FixedRadius
       width=1
       color {colorUseRule=UseColor_ColorTable, specifiedColor=<rgba 0 0 0 0>}
       cornerRadii=<4 4 4 4>
      }
      tintColor {colorUseRule=UseColor_Specified, specifiedColor=<rgba 0.0395 0.0395 0.0395 1>}
      uVRegion {bIsValid=false, max=<0 0>, min=<0 0>}
     }
     disabledForeground {colorUseRule=UseColor_ColorTable, specifiedColor=<rgba 0 0 0 0>}
     hovered {
      bIsDynamicallyLoaded=false
      drawAs=RoundedBox
      imageType=NoImage
      mirroring=NoMirror
      resourceName=None
      resourceObject=null
      tiling=NoTile
      imageSize=<32 32>
      margin=<ltrb 0 0 0 0>
      outlineSettings {
       bUseBrushTransparency=false
       roundingType=FixedRadius
       width=1
       color {colorUseRule=UseColor_ColorTable, specifiedColor=<rgba 0 0 0 0>}
       cornerRadii=<4 4 4 4>
      }
      tintColor {colorUseRule=UseColor_Specified, specifiedColor=<rgba 0.0953 0.0953 0.0953 1>}
      uVRegion {bIsValid=false, max=<0 0>, min=<0 0>}
     }
     hoveredForeground {colorUseRule=UseColor_ColorTable, specifiedColor=<rgba 0 0 0 0>}
     hoveredSlateSound {resourceObject=null}
     normal {
      bIsDynamicallyLoaded=false
      drawAs=RoundedBox
      imageType=NoImage
      mirroring=NoMirror
      resourceName=None
      resourceObject=null
      tiling=NoTile
      imageSize=<32 32>
      margin=<ltrb 0 0 0 0>
      outlineSettings {
       bUseBrushTransparency=false
       roundingType=FixedRadius
       width=1
       color {colorUseRule=UseColor_ColorTable, specifiedColor=<rgba 0 0 0 0>}
       cornerRadii=<4 4 4 4>
      }
      tintColor {colorUseRule=UseColor_Specified, specifiedColor=<rgba 0.0395 0.0395 0.0395 1>}
      uVRegion {bIsValid=false, max=<0 0>, min=<0 0>}
     }
     normalForeground {colorUseRule=UseColor_ColorTable, specifiedColor=<rgba 0 0 0 0>}
     normalPadding=<ltrb 5 1.5 5 1.5>
     pressed {
      bIsDynamicallyLoaded=false
      drawAs=RoundedBox
      imageType=NoImage
      mirroring=NoMirror
      resourceName=None
      resourceObject=null
      tiling=NoTile
      imageSize=<32 32>
      margin=<ltrb 0 0 0 0>
      outlineSettings {
       bUseBrushTransparency=false
       roundingType=FixedRadius
       width=1
       color {colorUseRule=UseColor_ColorTable, specifiedColor=<rgba 0 0 0 0>}
       cornerRadii=<4 4 4 4>
      }
      tintColor {colorUseRule=UseColor_Specified, specifiedColor=<rgba 0.0284 0.0284 0.0284 1>}
      uVRegion {bIsValid=false, max=<0 0>, min=<0 0>}
     }
     pressedForeground {colorUseRule=UseColor_ColorTable, specifiedColor=<rgba 0 0 0 0>}
     pressedPadding=<ltrb 5 2.5 5 0.5>
     pressedSlateSound {resourceObject=null}
    }}
   slot {class=HorizontalBoxSlot, properties {padding=<ltrb 10 0 3 0>}}
  }
  VerticalBox_10 {
   widgetClassPath=&VerticalBox
   parent=Button_AddPickaxes
   widget=VerticalBox_10
   slot {class=ButtonSlot, properties {padding=<ltrb 0 0 0 0>}}
  }
  TextBlock_22 {
   widgetClassPath=&TextBlock
   parent=VerticalBox_10
   widget=TextBlock_22
   properties {
    justification=Center
    text=+
    font {
     bForceMonospaced=false
     bMaterialIsStencil=false
     fontMaterial=null
     fontObject=&Roboto
     letterSpacing=0
     monospacedWidth=1
     size=15
     skewAmount=0
     typefaceFontName=Bold
     outlineSettings {
      bApplyOutlineToDropShadows=false
      bMiteredCorners=false
      bSeparateFillAlpha=false
      outlineMaterial=null
      outlineSize=0
      outlineColor=<rgba 0 0 0 1>
     }
    }
   }
   slot {class=VerticalBoxSlot}
  }
  TextBlock_23 {
   widgetClassPath=&TextBlock
   parent=VerticalBox_10
   widget=TextBlock_23
   properties {
    justification=Center
    text=Add ID
    font {
     bForceMonospaced=false
     bMaterialIsStencil=false
     fontMaterial=null
     fontObject=&Roboto
     letterSpacing=0
     monospacedWidth=1
     size=9
     skewAmount=0
     typefaceFontName=Bold
     outlineSettings {
      bApplyOutlineToDropShadows=false
      bMiteredCorners=false
      bSeparateFillAlpha=false
      outlineMaterial=null
      outlineSize=0
      outlineColor=<rgba 0 0 0 1>
     }
    }
   }
   slot {class=VerticalBoxSlot}
  }
  HorizontalBox_6 {
   widgetClassPath=&HorizontalBox
   parent=VerticalBox_2
   widget=HorizontalBox_6
   slot {class=VerticalBoxSlot, properties {padding=<ltrb 0 5 0 0>}}
  }
  TextBlock_14 {
   widgetClassPath=&TextBlock
   parent=HorizontalBox_6
   widget=TextBlock_14
   properties {
    text="Text Overlays ▼"
    font {
     bForceMonospaced=false
     bMaterialIsStencil=false
     fontMaterial=null
     fontObject=&Roboto
     letterSpacing=0
     monospacedWidth=1
     size=10.5
     skewAmount=0
     typefaceFontName=Bold
     outlineSettings {
      bApplyOutlineToDropShadows=false
      bMiteredCorners=false
      bSeparateFillAlpha=false
      outlineMaterial=null
      outlineSize=0
      outlineColor=<rgba 0 0 0 1>
     }
    }
   }
   slot {class=HorizontalBoxSlot}
  }
  Checkbox_EditOverlays {
   widgetClassPath=&EditorUtilityCheckBox
   bIsVariable=true
   parent=HorizontalBox_6
   widget=Checkbox_EditOverlays
   properties {widgetStyle {
     checkBoxType=CheckBox
     backgroundHoveredImage {
      bIsDynamicallyLoaded=false
      drawAs=RoundedBox
      imageType=NoImage
      mirroring=NoMirror
      resourceName=None
      resourceObject=null
      tiling=NoTile
      imageSize=<18 18>
      margin=<ltrb 0 0 0 0>
      outlineSettings {
       bUseBrushTransparency=false
       roundingType=FixedRadius
       width=1
       color {colorUseRule=UseColor_ColorTable, specifiedColor=<rgba 0 0 0 0>}
       cornerRadii=<3 3 3 3>
      }
      tintColor {colorUseRule=UseColor_Specified, specifiedColor=<rgba 0.0048 0.0048 0.0048 1>}
      uVRegion {bIsValid=false, max=<0 0>, min=<0 0>}
     }
     backgroundImage {
      bIsDynamicallyLoaded=false
      drawAs=RoundedBox
      imageType=NoImage
      mirroring=NoMirror
      resourceName=None
      resourceObject=null
      tiling=NoTile
      imageSize=<18 18>
      margin=<ltrb 0 0 0 0>
      outlineSettings {
       bUseBrushTransparency=false
       roundingType=FixedRadius
       width=1
       color {colorUseRule=UseColor_ColorTable, specifiedColor=<rgba 0 0 0 0>}
       cornerRadii=<3 3 3 3>
      }
      tintColor {colorUseRule=UseColor_Specified, specifiedColor=<rgba 0.0048 0.0048 0.0048 1>}
      uVRegion {bIsValid=false, max=<0 0>, min=<0 0>}
     }
     backgroundPressedImage {
      bIsDynamicallyLoaded=false
      drawAs=RoundedBox
      imageType=NoImage
      mirroring=NoMirror
      resourceName=None
      resourceObject=null
      tiling=NoTile
      imageSize=<18 18>
      margin=<ltrb 0 0 0 0>
      outlineSettings {
       bUseBrushTransparency=false
       roundingType=FixedRadius
       width=1
       color {colorUseRule=UseColor_ColorTable, specifiedColor=<rgba 0 0 0 0>}
       cornerRadii=<3 3 3 3>
      }
      tintColor {colorUseRule=UseColor_Specified, specifiedColor=<rgba 0.0048 0.0048 0.0048 1>}
      uVRegion {bIsValid=false, max=<0 0>, min=<0 0>}
     }
     borderBackgroundColor {colorUseRule=UseColor_Specified, specifiedColor=<rgba 1 1 1 1>}
     checkedForeground {colorUseRule=UseColor_Specified, specifiedColor=<rgba 1 1 1 1>}
     checkedHoveredForeground {colorUseRule=UseColor_Specified, specifiedColor=<rgba 1 1 1 1>}
     checkedHoveredImage {
      bIsDynamicallyLoaded=false
      drawAs=Image
      imageType=Vector
      mirroring=NoMirror
      resourceName=../../../Engine/Content/Slate/Starship/CoreWidgets/CheckBox/check.svg
      resourceObject=null
      tiling=NoTile
      imageSize=<16 16>
      margin=<ltrb 0 0 0 0>
      outlineSettings {
       bUseBrushTransparency=false
       roundingType=HalfHeightRadius
       width=0
       color {colorUseRule=UseColor_Specified, specifiedColor=<rgba 0 0 0 0>}
       cornerRadii=<0 0 0 0>
      }
      tintColor {colorUseRule=UseColor_Specified, specifiedColor=<rgba 0.0044 0.2384 1 1>}
      uVRegion {bIsValid=false, max=<0 0>, min=<0 0>}
     }
     checkedImage {
      bIsDynamicallyLoaded=false
      drawAs=Image
      imageType=Vector
      mirroring=NoMirror
      resourceName=../../../Engine/Content/Slate/Starship/CoreWidgets/CheckBox/check.svg
      resourceObject=null
      tiling=NoTile
      imageSize=<16 16>
      margin=<ltrb 0 0 0 0>
      outlineSettings {
       bUseBrushTransparency=false
       roundingType=HalfHeightRadius
       width=0
       color {colorUseRule=UseColor_Specified, specifiedColor=<rgba 0 0 0 0>}
       cornerRadii=<0 0 0 0>
      }
      tintColor {colorUseRule=UseColor_Specified, specifiedColor=<rgba 0 0.162 0.7454 1>}
      uVRegion {bIsValid=false, max=<0 0>, min=<0 0>}
     }
     checkedPressedForeground {colorUseRule=UseColor_Specified, specifiedColor=<rgba 1 1 1 1>}
     checkedPressedImage {
      bIsDynamicallyLoaded=false
      drawAs=Image
      imageType=Vector
      mirroring=NoMirror
      resourceName=../../../Engine/Content/Slate/Starship/CoreWidgets/CheckBox/check.svg
      resourceObject=null
      tiling=NoTile
      imageSize=<16 16>
      margin=<ltrb 0 0 0 0>
      outlineSettings {
       bUseBrushTransparency=false
       roundingType=HalfHeightRadius
       width=0
       color {colorUseRule=UseColor_Specified, specifiedColor=<rgba 0 0 0 0>}
       cornerRadii=<0 0 0 0>
      }
      tintColor {colorUseRule=UseColor_Specified, specifiedColor=<rgba 0 0.0802 0.3515 1>}
      uVRegion {bIsValid=false, max=<0 0>, min=<0 0>}
     }
     checkedSlateSound {resourceObject=null}
     foregroundColor {colorUseRule=UseColor_Specified, specifiedColor=<rgba 1 1 1 1>}
     hoveredForeground {colorUseRule=UseColor_Specified, specifiedColor=<rgba 1 1 1 1>}
     hoveredSlateSound {resourceObject=null}
     padding=<ltrb 2 0 0 0>
     pressedForeground {colorUseRule=UseColor_Specified, specifiedColor=<rgba 1 1 1 1>}
     uncheckedHoveredImage {
      bIsDynamicallyLoaded=false
      drawAs=NoDrawType
      imageType=NoImage
      mirroring=NoMirror
      resourceName=None
      resourceObject=null
      tiling=NoTile
      imageSize=<0 0>
      margin=<ltrb 0 0 0 0>
      outlineSettings {
       bUseBrushTransparency=false
       roundingType=HalfHeightRadius
       width=0
       color {colorUseRule=UseColor_Specified, specifiedColor=<rgba 0 0 0 0>}
       cornerRadii=<0 0 0 0>
      }
      tintColor {colorUseRule=UseColor_Specified, specifiedColor=<rgba 1 1 1 1>}
      uVRegion {bIsValid=false, max=<0 0>, min=<0 0>}
     }
     uncheckedImage {
      bIsDynamicallyLoaded=false
      drawAs=NoDrawType
      imageType=NoImage
      mirroring=NoMirror
      resourceName=None
      resourceObject=null
      tiling=NoTile
      imageSize=<0 0>
      margin=<ltrb 0 0 0 0>
      outlineSettings {
       bUseBrushTransparency=false
       roundingType=HalfHeightRadius
       width=0
       color {colorUseRule=UseColor_Specified, specifiedColor=<rgba 0 0 0 0>}
       cornerRadii=<0 0 0 0>
      }
      tintColor {colorUseRule=UseColor_Specified, specifiedColor=<rgba 1 1 1 1>}
      uVRegion {bIsValid=false, max=<0 0>, min=<0 0>}
     }
     uncheckedPressedImage {
      bIsDynamicallyLoaded=false
      drawAs=NoDrawType
      imageType=NoImage
      mirroring=NoMirror
      resourceName=None
      resourceObject=null
      tiling=NoTile
      imageSize=<0 0>
      margin=<ltrb 0 0 0 0>
      outlineSettings {
       bUseBrushTransparency=false
       roundingType=HalfHeightRadius
       width=0
       color {colorUseRule=UseColor_Specified, specifiedColor=<rgba 0 0 0 0>}
       cornerRadii=<0 0 0 0>
      }
      tintColor {colorUseRule=UseColor_Specified, specifiedColor=<rgba 1 1 1 1>}
      uVRegion {bIsValid=false, max=<0 0>, min=<0 0>}
     }
     uncheckedSlateSound {resourceObject=null}
     undeterminedForeground {colorUseRule=UseColor_Specified, specifiedColor=<rgba 1 1 1 1>}
     undeterminedHoveredImage {
      bIsDynamicallyLoaded=false
      drawAs=Image
      imageType=Vector
      mirroring=NoMirror
      resourceName=../../../Engine/Content/Slate/Starship/CoreWidgets/CheckBox/indeterminate.svg
      resourceObject=null
      tiling=NoTile
      imageSize=<16 16>
      margin=<ltrb 0 0 0 0>
      outlineSettings {
       bUseBrushTransparency=false
       roundingType=HalfHeightRadius
       width=0
       color {colorUseRule=UseColor_Specified, specifiedColor=<rgba 0 0 0 0>}
       cornerRadii=<0 0 0 0>
      }
      tintColor {colorUseRule=UseColor_Specified, specifiedColor=<rgba 0.0044 0.2384 1 1>}
      uVRegion {bIsValid=false, max=<0 0>, min=<0 0>}
     }
     undeterminedImage {
      bIsDynamicallyLoaded=false
      drawAs=Image
      imageType=Vector
      mirroring=NoMirror
      resourceName=../../../Engine/Content/Slate/Starship/CoreWidgets/CheckBox/indeterminate.svg
      resourceObject=null
      tiling=NoTile
      imageSize=<16 16>
      margin=<ltrb 0 0 0 0>
      outlineSettings {
       bUseBrushTransparency=false
       roundingType=HalfHeightRadius
       width=0
       color {colorUseRule=UseColor_Specified, specifiedColor=<rgba 0 0 0 0>}
       cornerRadii=<0 0 0 0>
      }
      tintColor {colorUseRule=UseColor_Specified, specifiedColor=<rgba 0 0.162 0.7454 1>}
      uVRegion {bIsValid=false, max=<0 0>, min=<0 0>}
     }
     undeterminedPressedImage {
      bIsDynamicallyLoaded=false
      drawAs=Image
      imageType=Vector
      mirroring=NoMirror
      resourceName=../../../Engine/Content/Slate/Starship/CoreWidgets/CheckBox/indeterminate.svg
      resourceObject=null
      tiling=NoTile
      imageSize=<16 16>
      margin=<ltrb 0 0 0 0>
      outlineSettings {
       bUseBrushTransparency=false
       roundingType=HalfHeightRadius
       width=0
       color {colorUseRule=UseColor_Specified, specifiedColor=<rgba 0 0 0 0>}
       cornerRadii=<0 0 0 0>
      }
      tintColor {colorUseRule=UseColor_Specified, specifiedColor=<rgba 0 0.0802 0.3515 1>}
      uVRegion {bIsValid=false, max=<0 0>, min=<0 0>}
     }
    }}
   slot {class=HorizontalBoxSlot, properties {padding=<ltrb 3 0 0 0>}}
  }
  Box_TextOverlays {
   widgetClassPath=&VerticalBox
   bIsVariable=true
   parent=VerticalBox_2
   widget=Box_TextOverlays
   properties {bOverride_Cursor=true, visibility=Collapsed}
   slot {class=VerticalBoxSlot, properties {padding=<ltrb 10 0 0 0>}}
  }
  HorizontalBox_0 {widgetClassPath=&HorizontalBox, parent=Box_TextOverlays, widget=HorizontalBox_0, slot {class=VerticalBoxSlot}}
  HorizontalBox_14 {
   widgetClassPath=&HorizontalBox
   parent=Box_TextOverlays
   widget=HorizontalBox_14
   properties {visibility=Collapsed}
   slot {class=VerticalBoxSlot}
  }
  TextBlock_8 {
   widgetClassPath=&TextBlock
   parent=HorizontalBox_14
   widget=TextBlock_8
   properties {
    text=Montage Detail:
    font {
     bForceMonospaced=false
     bMaterialIsStencil=false
     fontMaterial=null
     fontObject=&Roboto
     letterSpacing=0
     monospacedWidth=1
     size=9
     skewAmount=0
     typefaceFontName=Bold
     outlineSettings {
      bApplyOutlineToDropShadows=false
      bMiteredCorners=false
      bSeparateFillAlpha=false
      outlineMaterial=null
      outlineSize=0
      outlineColor=<rgba 0 0 0 1>
     }
    }
   }
   slot {class=HorizontalBoxSlot, properties {verticalAlignment=VAlign_Center}}
  }
  PropertyView_Montage_Size {
   widgetClassPath=&SinglePropertyView
   parent=HorizontalBox_14
   widget=PropertyView_Montage_Size
   properties {nameOverride=Size, propertyName=MontageDetails_Size}
   slot {class=HorizontalBoxSlot}
  }
  PropertyView_Montage_Position {
   widgetClassPath=&SinglePropertyView
   parent=HorizontalBox_14
   widget=PropertyView_Montage_Position
   properties {nameOverride=Position, propertyName=MontageDetails_Position, visibility=Collapsed}
   slot {class=HorizontalBoxSlot}
  }
  PropertyView_Montage_Color {
   widgetClassPath=&SinglePropertyView
   parent=HorizontalBox_14
   widget=PropertyView_Montage_Color
   properties {nameOverride=Color, propertyName=MontageDetails_Color, visibility=Collapsed}
   slot {class=HorizontalBoxSlot}
  }
  HorizontalBox_1 {
   widgetClassPath=&HorizontalBox
   parent=Box_TextOverlays
   widget=HorizontalBox_1
   properties {visibility=Collapsed}
   slot {class=VerticalBoxSlot}
  }
  TextBlock_9 {
   widgetClassPath=&TextBlock
   parent=HorizontalBox_1
   widget=TextBlock_9
   properties {
    text=View Labels:
    font {
     bForceMonospaced=false
     bMaterialIsStencil=false
     fontMaterial=null
     fontObject=&Roboto
     letterSpacing=0
     monospacedWidth=1
     size=9
     skewAmount=0
     typefaceFontName=Bold
     outlineSettings {
      bApplyOutlineToDropShadows=false
      bMiteredCorners=false
      bSeparateFillAlpha=false
      outlineMaterial=null
      outlineSize=0
      outlineColor=<rgba 0 0 0 1>
     }
    }
   }
   slot {class=HorizontalBoxSlot, properties {verticalAlignment=VAlign_Center}}
  }
  PropertyView_ViewLables_Size {
   widgetClassPath=&SinglePropertyView
   parent=HorizontalBox_1
   widget=PropertyView_ViewLables_Size
   properties {nameOverride=Size, propertyName=ViewLabels_Size}
   slot {class=HorizontalBoxSlot}
  }
  PropertyView_ViewLables_Position {
   widgetClassPath=&SinglePropertyView
   parent=HorizontalBox_1
   widget=PropertyView_ViewLables_Position
   properties {nameOverride=Position, propertyName=ViewLabels_Position}
   slot {class=HorizontalBoxSlot}
  }
  PropertyView_ViewLables_Color {
   widgetClassPath=&SinglePropertyView
   parent=HorizontalBox_1
   widget=PropertyView_ViewLables_Color
   properties {nameOverride=Color, propertyName=ViewLabels_Color}
   slot {class=HorizontalBoxSlot}
  }
  PropertyView_ViewLables_Visible {
   widgetClassPath=&SinglePropertyView
   parent=HorizontalBox_1
   widget=PropertyView_ViewLables_Visible
   properties {nameOverride=Visible, propertyName=ViewLabels_Visibility}
   slot {class=HorizontalBoxSlot}
  }
  HorizontalBox_5 {
   widgetClassPath=&HorizontalBox
   parent=Box_TextOverlays
   widget=HorizontalBox_5
   properties {visibility=Collapsed}
   slot {class=VerticalBoxSlot}
  }
  TextBlock_2 {
   widgetClassPath=&TextBlock
   parent=HorizontalBox_5
   widget=TextBlock_2
   properties {
    text=Camera Parameters:
    font {
     bForceMonospaced=false
     bMaterialIsStencil=false
     fontMaterial=null
     fontObject=&Roboto
     letterSpacing=0
     monospacedWidth=1
     size=9
     skewAmount=0
     typefaceFontName=Bold
     outlineSettings {
      bApplyOutlineToDropShadows=false
      bMiteredCorners=false
      bSeparateFillAlpha=false
      outlineMaterial=null
      outlineSize=0
      outlineColor=<rgba 0 0 0 1>
     }
    }
   }
   slot {class=HorizontalBoxSlot, properties {verticalAlignment=VAlign_Center}}
  }
  PropertyView_Cam_FOV {
   widgetClassPath=&SinglePropertyView
   parent=HorizontalBox_5
   widget=PropertyView_Cam_FOV
   properties {nameOverride=FOV, propertyName=Cameras_FOV}
   slot {class=HorizontalBoxSlot}
  }
  PropertyView_Cam_Distance {
   widgetClassPath=&SinglePropertyView
   parent=HorizontalBox_5
   widget=PropertyView_Cam_Distance
   properties {nameOverride=Distance, propertyName=Cameras_Distance}
   slot {class=HorizontalBoxSlot}
  }
  PropertyView_Cam_Distance_1 {
   widgetClassPath=&SinglePropertyView
   parent=HorizontalBox_5
   widget=PropertyView_Cam_Distance_1
   properties {nameOverride=Height, propertyName=Cameras_Height}
   slot {class=HorizontalBoxSlot}
  }
  HorizontalBox {
   widgetClassPath=&HorizontalBox
   parent=Box_TextOverlays
   widget=HorizontalBox
   slot {class=VerticalBoxSlot, properties {padding=<ltrb 0 5 0 0>}}
  }
  TextBlock_6 {
   widgetClassPath=&TextBlock
   parent=HorizontalBox
   widget=TextBlock_6
   properties {
    text="Add Overlay: "
    font {
     bForceMonospaced=false
     bMaterialIsStencil=false
     fontMaterial=null
     fontObject=&Roboto
     letterSpacing=0
     monospacedWidth=1
     size=9
     skewAmount=0
     typefaceFontName=Bold
     outlineSettings {
      bApplyOutlineToDropShadows=false
      bMiteredCorners=false
      bSeparateFillAlpha=false
      outlineMaterial=null
      outlineSize=0
      outlineColor=<rgba 0 0 0 1>
     }
    }
   }
   slot {class=HorizontalBoxSlot}
  }
  EditorUtilityButton_79 {
   widgetClassPath=&EditorUtilityButton
   parent=HorizontalBox
   widget=EditorUtilityButton_79
   properties {widgetStyle {
     clickedSlateSound {resourceObject=null}
     disabled {
      bIsDynamicallyLoaded=false
      drawAs=RoundedBox
      imageType=NoImage
      mirroring=NoMirror
      resourceName=None
      resourceObject=null
      tiling=NoTile
      imageSize=<0 0>
      margin=<ltrb 0 0 0 0>
      outlineSettings {
       bUseBrushTransparency=false
       roundingType=FixedRadius
       width=1
       color {colorUseRule=UseColor_ColorTable, specifiedColor=<rgba 0 0 0 0>}
       cornerRadii=<4 4 4 4>
      }
      tintColor {colorUseRule=UseColor_Specified, specifiedColor=<rgba 0.0395 0.0395 0.0395 1>}
      uVRegion {bIsValid=false, max=<0 0>, min=<0 0>}
     }
     disabledForeground {colorUseRule=UseColor_ColorTable, specifiedColor=<rgba 0 0 0 0>}
     hovered {
      bIsDynamicallyLoaded=false
      drawAs=RoundedBox
      imageType=NoImage
      mirroring=NoMirror
      resourceName=None
      resourceObject=null
      tiling=NoTile
      imageSize=<32 32>
      margin=<ltrb 0 0 0 0>
      outlineSettings {
       bUseBrushTransparency=false
       roundingType=FixedRadius
       width=1
       color {colorUseRule=UseColor_ColorTable, specifiedColor=<rgba 0 0 0 0>}
       cornerRadii=<4 4 4 4>
      }
      tintColor {colorUseRule=UseColor_Specified, specifiedColor=<rgba 0.0953 0.0953 0.0953 1>}
      uVRegion {bIsValid=false, max=<0 0>, min=<0 0>}
     }
     hoveredForeground {colorUseRule=UseColor_ColorTable, specifiedColor=<rgba 0 0 0 0>}
     hoveredSlateSound {resourceObject=null}
     normal {
      bIsDynamicallyLoaded=false
      drawAs=RoundedBox
      imageType=NoImage
      mirroring=NoMirror
      resourceName=None
      resourceObject=null
      tiling=NoTile
      imageSize=<32 32>
      margin=<ltrb 0 0 0 0>
      outlineSettings {
       bUseBrushTransparency=false
       roundingType=FixedRadius
       width=1
       color {colorUseRule=UseColor_ColorTable, specifiedColor=<rgba 0 0 0 0>}
       cornerRadii=<4 4 4 4>
      }
      tintColor {colorUseRule=UseColor_Specified, specifiedColor=<rgba 0.0395 0.0395 0.0395 1>}
      uVRegion {bIsValid=false, max=<0 0>, min=<0 0>}
     }
     normalForeground {colorUseRule=UseColor_ColorTable, specifiedColor=<rgba 0 0 0 0>}
     normalPadding=<ltrb 12 0 12 0>
     pressed {
      bIsDynamicallyLoaded=false
      drawAs=RoundedBox
      imageType=NoImage
      mirroring=NoMirror
      resourceName=None
      resourceObject=null
      tiling=NoTile
      imageSize=<32 32>
      margin=<ltrb 0 0 0 0>
      outlineSettings {
       bUseBrushTransparency=false
       roundingType=FixedRadius
       width=1
       color {colorUseRule=UseColor_ColorTable, specifiedColor=<rgba 0 0 0 0>}
       cornerRadii=<4 4 4 4>
      }
      tintColor {colorUseRule=UseColor_Specified, specifiedColor=<rgba 0.0284 0.0284 0.0284 1>}
      uVRegion {bIsValid=false, max=<0 0>, min=<0 0>}
     }
     pressedForeground {colorUseRule=UseColor_ColorTable, specifiedColor=<rgba 0 0 0 0>}
     pressedPadding=<ltrb 12 2.5 12 0.5>
     pressedSlateSound {resourceObject=null}
    }}
   slot {class=HorizontalBoxSlot}
  }
  TextBlock_10 {
   widgetClassPath=&TextBlock
   parent=EditorUtilityButton_79
   widget=TextBlock_10
   properties {
    justification=Center
    text=+
    font {
     bForceMonospaced=false
     bMaterialIsStencil=false
     fontMaterial=null
     fontObject=&Roboto
     letterSpacing=0
     monospacedWidth=1
     size=9
     skewAmount=0
     typefaceFontName=Bold
     outlineSettings {
      bApplyOutlineToDropShadows=false
      bMiteredCorners=false
      bSeparateFillAlpha=false
      outlineMaterial=null
      outlineSize=0
      outlineColor=<rgba 0 0 0 1>
     }
    }
   }
   slot {class=ButtonSlot, properties {horizontalAlignment=HAlign_Fill, verticalAlignment=VAlign_Fill, padding=<ltrb 0 0 0 0>}}
  }
  TextBlock_7 {
   widgetClassPath=&TextBlock
   parent=HorizontalBox
   widget=TextBlock_7
   properties {
    text="Clear All Overlays: "
    font {
     bForceMonospaced=false
     bMaterialIsStencil=false
     fontMaterial=null
     fontObject=&Roboto
     letterSpacing=0
     monospacedWidth=1
     size=9
     skewAmount=0
     typefaceFontName=Bold
     outlineSettings {
      bApplyOutlineToDropShadows=false
      bMiteredCorners=false
      bSeparateFillAlpha=false
      outlineMaterial=null
      outlineSize=0
      outlineColor=<rgba 0 0 0 1>
     }
    }
   }
   slot {class=HorizontalBoxSlot, properties {padding=<ltrb 60 0 0 0>}}
  }
  EditorUtilityButton {
   widgetClassPath=&EditorUtilityButton
   parent=HorizontalBox
   widget=EditorUtilityButton
   properties {widgetStyle {
     clickedSlateSound {resourceObject=null}
     disabled {
      bIsDynamicallyLoaded=false
      drawAs=RoundedBox
      imageType=NoImage
      mirroring=NoMirror
      resourceName=None
      resourceObject=null
      tiling=NoTile
      imageSize=<0 0>
      margin=<ltrb 0 0 0 0>
      outlineSettings {
       bUseBrushTransparency=false
       roundingType=FixedRadius
       width=1
       color {colorUseRule=UseColor_ColorTable, specifiedColor=<rgba 0 0 0 0>}
       cornerRadii=<4 4 4 4>
      }
      tintColor {colorUseRule=UseColor_Specified, specifiedColor=<rgba 0.0395 0.0395 0.0395 1>}
      uVRegion {bIsValid=false, max=<0 0>, min=<0 0>}
     }
     disabledForeground {colorUseRule=UseColor_ColorTable, specifiedColor=<rgba 0 0 0 0>}
     hovered {
      bIsDynamicallyLoaded=false
      drawAs=RoundedBox
      imageType=NoImage
      mirroring=NoMirror
      resourceName=None
      resourceObject=null
      tiling=NoTile
      imageSize=<32 32>
      margin=<ltrb 0 0 0 0>
      outlineSettings {
       bUseBrushTransparency=false
       roundingType=FixedRadius
       width=1
       color {colorUseRule=UseColor_ColorTable, specifiedColor=<rgba 0 0 0 0>}
       cornerRadii=<4 4 4 4>
      }
      tintColor {colorUseRule=UseColor_Specified, specifiedColor=<rgba 0.0953 0.0953 0.0953 1>}
      uVRegion {bIsValid=false, max=<0 0>, min=<0 0>}
     }
     hoveredForeground {colorUseRule=UseColor_ColorTable, specifiedColor=<rgba 0 0 0 0>}
     hoveredSlateSound {resourceObject=null}
     normal {
      bIsDynamicallyLoaded=false
      drawAs=RoundedBox
      imageType=NoImage
      mirroring=NoMirror
      resourceName=None
      resourceObject=null
      tiling=NoTile
      imageSize=<32 32>
      margin=<ltrb 0 0 0 0>
      outlineSettings {
       bUseBrushTransparency=false
       roundingType=FixedRadius
       width=1
       color {colorUseRule=UseColor_ColorTable, specifiedColor=<rgba 0 0 0 0>}
       cornerRadii=<4 4 4 4>
      }
      tintColor {colorUseRule=UseColor_Specified, specifiedColor=<rgba 0.0395 0.0395 0.0395 1>}
      uVRegion {bIsValid=false, max=<0 0>, min=<0 0>}
     }
     normalForeground {colorUseRule=UseColor_ColorTable, specifiedColor=<rgba 0 0 0 0>}
     normalPadding=<ltrb 12 0 12 0>
     pressed {
      bIsDynamicallyLoaded=false
      drawAs=RoundedBox
      imageType=NoImage
      mirroring=NoMirror
      resourceName=None
      resourceObject=null
      tiling=NoTile
      imageSize=<32 32>
      margin=<ltrb 0 0 0 0>
      outlineSettings {
       bUseBrushTransparency=false
       roundingType=FixedRadius
       width=1
       color {colorUseRule=UseColor_ColorTable, specifiedColor=<rgba 0 0 0 0>}
       cornerRadii=<4 4 4 4>
      }
      tintColor {colorUseRule=UseColor_Specified, specifiedColor=<rgba 0.0284 0.0284 0.0284 1>}
      uVRegion {bIsValid=false, max=<0 0>, min=<0 0>}
     }
     pressedForeground {colorUseRule=UseColor_ColorTable, specifiedColor=<rgba 0 0 0 0>}
     pressedPadding=<ltrb 12 2.5 12 0.5>
     pressedSlateSound {resourceObject=null}
    }}
   slot {class=HorizontalBoxSlot}
  }
  Image_2 {
   widgetClassPath=&Image
   bIsVariable=true
   parent=EditorUtilityButton
   widget=Image_2
   properties {brush {
     bIsDynamicallyLoaded=false
     drawAs=Image
     imageType=FullColor
     mirroring=NoMirror
     resourceName=None
     resourceObject=null
     tiling=NoTile
     imageSize=<10 10>
     margin=<ltrb 0 0 0 0>
     outlineSettings {
      bUseBrushTransparency=false
      roundingType=HalfHeightRadius
      width=0
      color {colorUseRule=UseColor_Specified, specifiedColor=<rgba 0 0 0 0>}
      cornerRadii=<0 0 0 0>
     }
     tintColor {colorUseRule=UseColor_Specified, specifiedColor=<rgba 1 1 1 1>}
     uVRegion {bIsValid=false, max=<0 0>, min=<0 0>}
    }}
   slot {class=ButtonSlot, properties {padding=<ltrb 2 2 2 2>}}
  }
  VerticalBox_textoverlayproperty {
   widgetClassPath=&VerticalBox
   bIsVariable=true
   parent=Box_TextOverlays
   widget=VerticalBox_textoverlayproperty
   slot {class=VerticalBoxSlot}
  }
  EUW_MultiCamViewer_TextDetails_103 {
   widgetClassPath=/Game/NeverCook/Animation/MultiCamViewer/Components/EUW_MultiCamViewer_TextDetails
   bIsVariable=true
   parent=VerticalBox_textoverlayproperty
   widget=EUW_MultiCamViewer_TextDetails_103
   slot {class=VerticalBoxSlot, properties {horizontalAlignment=HAlign_Left}}
  }
  EUW_MultiCamViewer_TextDetails_50 {
   widgetClassPath=&EUW_MultiCamViewer_TextDetails
   bIsVariable=true
   parent=VerticalBox_textoverlayproperty
   widget=EUW_MultiCamViewer_TextDetails_50
   slot {class=VerticalBoxSlot}
  }
  EUW_MultiCamViewer_TextDetails {
   widgetClassPath=&EUW_MultiCamViewer_TextDetails
   bIsVariable=true
   parent=VerticalBox_textoverlayproperty
   widget=EUW_MultiCamViewer_TextDetails
   slot {class=VerticalBoxSlot}
  }
 ]
}
