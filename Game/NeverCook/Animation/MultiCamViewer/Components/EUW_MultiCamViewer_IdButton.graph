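# Editor utility tile for the MultiCamViewer animation tool, showing one cosmetic's thumbnail and name.
# Clicking it switches on CosmeticType to play an emote or apply a character, Sparks instrument or pickaxe on the parent viewer.
# PreConstruct copies ItemDef into a child Property widget's typed field and points its SinglePropertyView at that field.
# Edit mode reveals that picker with confirm and delete buttons, which set or remove the viewer's ID by Index and type.
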EditorUtilityWidgetBlueprint /Game/NeverCook/Animation/MultiCamViewer/Components/EUW_MultiCamViewer_IdButton
graphs [EventGraph {nodes [
   "On Clicked (Button_0)"() {is=title K2Node_ComponentBoundEvent}
   > n1 = Cast To EUW_MultiCamViewer!(execute, Object=Get Parent() -> (Parent:Widget Object Reference)) -> (then, CastFailed, AsEUW Multi Cam Viewer:EUW Multi Cam Viewer Object Reference) {is=title K2Node_DynamicCast}
   > n2 = Switch on Int!(execute:Exec, Selection:int=Get CosmeticType() -> (CosmeticType:int=0)) -> (Default:Exec, 0:Exec, 1:Exec, 2:Exec, 3:Exec)
   n3 = Get ItemDef() -> (ItemDef:Item Definition Base Object Reference)
   n4 = Cast To AthenaDanceItemDefinition!(execute=@n2.0, Object=@n3) -> (then, CastFailed, AsAthena Dance Item Definition:Athena Dance Item Definition Object Reference) {is=title K2Node_DynamicCast}
   > PlayEmote!(execute, self:EUW Multi Cam Viewer Object Reference=@n1, EmoteAsset:Fort Montage Item Definition Base Object Reference=@n4) -> (then)
   n5 = Cast To AthenaCharacterItemDefinition!(execute=@n2.1, Object=@n3) -> (then, CastFailed, AsAthena Character Item Definition:Athena Character Item Definition Object Reference) {is=title K2Node_DynamicCast}
   > SetCosmetic!(
    execute
    self:EUW Multi Cam Viewer Object Reference=@n1
    ID:Fort Item Definition Object Reference=@n5
    CosmeticType:CosmeticPicker_CosmeticTypes Enum=NewEnumerator0
   ) -> (then)
   n6 = Cast To SparksInstrumentItemDefBase!(execute=@n2.2, Object=@n3) -> (then, CastFailed, AsSparks Instrument Item Def Base:Sparks Instrument Item Def Base Object Reference) {is=title K2Node_DynamicCast}
   > ApplySparksID!(execute, self:EUW Multi Cam Viewer Object Reference=@n1, Target:Item Definition Base Object Reference=@n6) -> (then)
   n7 = Cast To AthenaPickaxeItemDefinition!(execute=@n2.3, Object=@n3) -> (then, CastFailed, AsAthena Pickaxe Item Definition:Athena Pickaxe Item Definition Object Reference) {is=title K2Node_DynamicCast}
   > SetCosmetic(self=@n1, ID=@n7, CosmeticType=NewEnumerator2)
   "On Clicked (Button_Edit)"() {is=title K2Node_ComponentBoundEvent}
   > EditVisibility!(execute, self, Index:bool=true) -> (then)
   "On Clicked (Button_Confirm)"() {is=title K2Node_ComponentBoundEvent}
   > Confirm Button()
   "On Clicked (ButtonDelete)"() {is=title K2Node_ComponentBoundEvent}
   > n8 = Sequence()
   > n9 = Cast To EUW_MultiCamViewer(Object=Get Parent())
   > Remove ID!(
    execute
    self:EUW Multi Cam Viewer Object Reference=@n9
    Index:int=Get Index() -> (Index:int=0)
    CosmeticType:int=Get CosmeticType() -> (CosmeticType=0)
   ) -> (then)
   RemoveFromParent(execute=@n8.then_1)
   Confirm Button() {is=title K2Node_CustomEvent}
   > EditVisibility(Index=false)
   > RefreshDisplay()
   > n10 = Cast To EUW_MultiCamViewer(Object=Get Parent())
   > Set ID!(
    execute
    self:EUW Multi Cam Viewer Object Reference=@n10
    Index:int=Get Index() -> (Index=0)
    ID:Item Definition Base Object Reference=Get ItemDef()
    CosmeticType:int=Get CosmeticType() -> (CosmeticType=0)
   ) -> (then) {is=title K2Node_CallFunction}
   n11 = RefreshDisplay() {is=title K2Node_CustomEvent}
   SetAssetByObject!(
    execute=[@n22.then, @n11.then]
    self:Asset Thumbnail Widget Object Reference=Get AssetThumbnailWidget_30() -> (AssetThumbnailWidget_30:Asset Thumbnail Widget Object Reference)
    Object:Object Reference=Get ItemDef()
   ) -> (then)
   n12 = Get ItemDef()
   > "SetText (Text)"(
    self=Get TextBlock_Name() -> (TextBlock_Name:Text Object Reference)
    InText=Select(
     Option 0:Text=Get ItemName!(self:Item Definition Base Object Reference=@n12) -> (ItemName:Text)
     Option 1:Text=Get NewName() -> (NewName:Text)
     Index=Get ReplaceName() -> (ReplaceName:bool)
    ) -> (ReturnValue:Text)
   )
   n13 = EditVisibility!() -> (OutputDelegate, then, Index:bool=~) {is=title K2Node_CustomEvent}
   > Set IsBeingEdited(IsBeingEdited:bool=@n13.Index) -> (Output_Get:bool)
   > SetVisibility(
    self=[
     Get VerticalBox_1() -> (VerticalBox_1:Vertical Box Object Reference)
     Get Button_Confirm() -> (Button_Confirm:Editor Utility Button Object Reference)
     Get ButtonDelete() -> (ButtonDelete:Editor Utility Button Object Reference)
    ]
    InVisibility=Select(Option 0:ESlateVisibility Enum=Collapsed, Option 1:ESlateVisibility Enum=Visible, Index=@n13.Index) -> (ReturnValue:ESlateVisibility Enum)
   )
   > SetVisibility(
    self=[Get Button_Edit() -> (Button_Edit:Editor Utility Button Object Reference), Get Button_0() -> (Button_0:Button Object Reference)]
    InVisibility=Select(Option 0=Visible, Option 1=Collapsed, Index=@n13.Index)
   )
   Event PreConstruct()
   > n14 = Sequence!(execute) -> (then_0, then_1, then_2:Exec)
   > Set Parent!(
    execute
    Parent:EUW Multi Cam Viewer Id Button Object Reference=Self-Reference()
    self:EUW Multi Cam Viewer Property Object Reference=Get EUW_MultiCamViewer_Property() -> (EUW_MultiCamViewer_Property:EUW Multi Cam Viewer Property Object Reference)
   ) -> (then, Output_Get:EUW Multi Cam Viewer Id Button Object Reference)
   > n15 = Switch on Int(Selection=Get CosmeticType() -> (CosmeticType=0))
   n16 = Get ItemDef()
   n17 = Cast To AthenaDanceItemDefinition(execute=@n15.0, Object=@n16)
   n18 = Get EUW_MultiCamViewer_Property()
   > Set EID!(execute, EID:Athena Dance Item Definition Object Reference=@n17, self:EUW Multi Cam Viewer Property Object Reference=@n18) -> (then, Output_Get:Athena Dance Item Definition Object Reference)
   n19 = Cast To AthenaCharacterItemDefinition(execute=@n15.1, Object=@n16)
   > Set CID!(execute, CID:Athena Character Item Definition Object Reference=@n19, self:EUW Multi Cam Viewer Property Object Reference=@n18) -> (then, Output_Get:Athena Character Item Definition Object Reference)
   n20 = Cast To SparksInstrumentItemDefBase(execute=@n15.2, Object=@n16)
   > Set Sparks!(execute, Sparks:Sparks Instrument Item Def Base Object Reference=@n20, self:EUW Multi Cam Viewer Property Object Reference=@n18) -> (then, Output_Get:Sparks Instrument Item Def Base Object Reference)
   n21 = Cast To AthenaPickaxeItemDefinition(execute=@n15.3, Object=@n16)
   > Set Pickaxe!(execute, Pickaxe:Athena Pickaxe Item Definition Object Reference=@n21, self:EUW Multi Cam Viewer Property Object Reference=@n18) -> (then, Output_Get:Athena Pickaxe Item Definition Object Reference)
   n22 = SetPropertyName!(
    execute=@n14.then_1
    self:Single Property View Object Reference=Get SinglePropertyView_65!(self:EUW Multi Cam Viewer Property Object Reference=Get EUW_MultiCamViewer_Property()) -> (SinglePropertyView_65:Single Property View Object Reference)
    NewPropertyName:Name=Select!(Option 0:Name=EID, Option 1:Name=CID, Option 2:Name=Sparks, Option 3:Name=Pickaxe, Index:int=Get CosmeticType() -> (CosmeticType=0)) -> (ReturnValue:Name)
   ) -> (then)
   EditVisibility(execute=@n14.then_2, Index=Get IsBeingEdited() -> (IsBeingEdited:bool)) {is=title K2Node_CallFunction}
   > SetVisibility(
    self=Get Border_EditButton() -> (Border_EditButton:Border Object Reference)
    InVisibility=Select!(
     Option 0:ESlateVisibility Enum=Collapsed
     Option 1:ESlateVisibility Enum=SelfHitTestInvisible
     Index:bool=Get IsEditable() -> (IsEditable:bool)
    ) -> (ReturnValue:ESlateVisibility Enum)
   )
   Event Construct()
   > n23 = Is Valid(InputObject=Get Parent())
   n24 = FindParentWidgetOfType!(
    execute=@n23.Is Not Valid
    StartingWidget:Widget Object Reference=Self-Reference()
    Type:Widget Class Reference=/Game/NeverCook/Animation/MultiCamViewer/EUW_MultiCamViewer.EUW_MultiCamViewer_C
   ) -> (then, ReturnValue:EUW Multi Cam Viewer Object Reference)
   > Set Parent!(execute, Parent:Widget Object Reference=@n24) -> (then, Output_Get:Widget Object Reference)
  ]}]
variables [
 CosmeticType {type=integer}
 Index {type=integer}
 IsBeingEdited {type=boolean}
 IsEditable {type=boolean, default=true}
 ItemDef {type=ItemDefinitionBase, default=/Game/Athena/Items/Cosmetics/Characters/CID_069_Athena_Commando_F_PinkBear}
 NewName {type=string, default=CMM}
 Parent {type=Widget, default=null}
 ReplaceName {type=boolean}
]
widgets {
 info {rootWidgetClass=&CanvasPanel, widgetCount=21}
 widgets [
  CanvasPanel_3 {widgetClassPath=&CanvasPanel, widget=CanvasPanel_3}
  Button_0 {
   widgetClassPath=&Button
   parent=CanvasPanel_3
   widget=Button_0
   properties {widgetStyle {
     clickedSlateSound {resourceObject=null}
     disabled {
      bIsDynamicallyLoaded=false
      drawAs=RoundedBox
      imageType=NoImage
      mirroring=NoMirror
      resourceName=None
      resourceObject=null
      tiling=NoTile
      imageSize=<0 0>
      margin=<ltrb 0 0 0 0>
      outlineSettings {
       bUseBrushTransparency=false
       roundingType=FixedRadius
       width=1
       color {colorUseRule=UseColor_ColorTable, specifiedColor=<rgba 0 0 0 0>}
       cornerRadii=<4 4 4 4>
      }
      tintColor {colorUseRule=UseColor_Specified, specifiedColor=<rgba 0.0395 0.0395 0.0395 1>}
      uVRegion {bIsValid=false, max=<0 0>, min=<0 0>}
     }
     disabledForeground {colorUseRule=UseColor_ColorTable, specifiedColor=<rgba 0 0 0 0>}
     hovered {
      bIsDynamicallyLoaded=false
      drawAs=RoundedBox
      imageType=NoImage
      mirroring=NoMirror
      resourceName=None
      resourceObject=null
      tiling=NoTile
      imageSize=<32 32>
      margin=<ltrb 0 0 0 0>
      outlineSettings {
       bUseBrushTransparency=false
       roundingType=FixedRadius
       width=1
       color {colorUseRule=UseColor_ColorTable, specifiedColor=<rgba 0 0 0 0>}
       cornerRadii=<4 4 4 4>
      }
      tintColor {colorUseRule=UseColor_Specified, specifiedColor=<rgba 0.0953 0.0953 0.0953 1>}
      uVRegion {bIsValid=false, max=<0 0>, min=<0 0>}
     }
     hoveredForeground {colorUseRule=UseColor_ColorTable, specifiedColor=<rgba 0 0 0 0>}
     hoveredSlateSound {resourceObject=null}
     normal {
      bIsDynamicallyLoaded=false
      drawAs=RoundedBox
      imageType=NoImage
      mirroring=NoMirror
      resourceName=None
      resourceObject=null
      tiling=NoTile
      imageSize=<32 32>
      margin=<ltrb 0 0 0 0>
      outlineSettings {
       bUseBrushTransparency=false
       roundingType=FixedRadius
       width=1
       color {colorUseRule=UseColor_ColorTable, specifiedColor=<rgba 0 0 0 0>}
       cornerRadii=<4 4 4 4>
      }
      tintColor {colorUseRule=UseColor_Specified, specifiedColor=<rgba 0.0395 0.0395 0.0395 1>}
      uVRegion {bIsValid=false, max=<0 0>, min=<0 0>}
     }
     normalForeground {colorUseRule=UseColor_ColorTable, specifiedColor=<rgba 0 0 0 0>}
     normalPadding=<ltrb 3 3 3 3>
     pressed {
      bIsDynamicallyLoaded=false
      drawAs=RoundedBox
      imageType=NoImage
      mirroring=NoMirror
      resourceName=None
      resourceObject=null
      tiling=NoTile
      imageSize=<32 32>
      margin=<ltrb 0 0 0 0>
      outlineSettings {
       bUseBrushTransparency=false
       roundingType=FixedRadius
       width=1
       color {colorUseRule=UseColor_ColorTable, specifiedColor=<rgba 0 0 0 0>}
       cornerRadii=<4 4 4 4>
      }
      tintColor {colorUseRule=UseColor_Specified, specifiedColor=<rgba 0.0284 0.0284 0.0284 1>}
      uVRegion {bIsValid=false, max=<0 0>, min=<0 0>}
     }
     pressedForeground {colorUseRule=UseColor_ColorTable, specifiedColor=<rgba 0 0 0 0>}
     pressedPadding=<ltrb 3 3 3 3>
     pressedSlateSound {resourceObject=null}
    }}
   slot {class=CanvasPanelSlot, properties {bAutoSize=true, layoutData=~}}
  }
  CanvasPanel_1 {widgetClassPath=&CanvasPanel, parent=Button_0, widget=CanvasPanel_1, slot {class=ButtonSlot}}
  SizeBox_0 {
   widgetClassPath=&SizeBox
   parent=CanvasPanel_1
   widget=SizeBox_0
   properties {bOverride_HeightOverride=true, bOverride_WidthOverride=true, heightOverride=82, widthOverride=73.6119}
   slot {class=CanvasPanelSlot, properties {bAutoSize=true, layoutData=~}}
  }
  VerticalBox_0 {
   widgetClassPath=&VerticalBox
   parent=SizeBox_0
   widget=VerticalBox_0
   properties {bOverride_Cursor=true}
   slot {class=SizeBoxSlot}
  }
  AssetThumbnailWidget_30 {
   widgetClassPath=/Script/UMGEditor.AssetThumbnailWidget
   bIsVariable=true
   parent=VerticalBox_0
   widget=AssetThumbnailWidget_30
   slot {class=VerticalBoxSlot, properties {size {sizeRule=Fill, value=1}}}
  }
  Border_39 {
   widgetClassPath=&Border
   parent=VerticalBox_0
   widget=Border_39
   properties {
    clipping=ClipToBounds
    horizontalAlignment=HAlign_Center
    verticalAlignment=VAlign_Center
    background {
     bIsDynamicallyLoaded=false
     drawAs=NoDrawType
     imageType=NoImage
     mirroring=NoMirror
     resourceName=None
     resourceObject=null
     tiling=NoTile
     imageSize=<32 32>
     margin=<ltrb 0 0 0 0>
     outlineSettings {
      bUseBrushTransparency=false
      roundingType=HalfHeightRadius
      width=0
      color {colorUseRule=UseColor_Specified, specifiedColor=<rgba 0 0 0 0>}
      cornerRadii=<0 0 0 0>
     }
     tintColor {colorUseRule=UseColor_Specified, specifiedColor=<rgba 1 1 1 1>}
     uVRegion {bIsValid=false, max=<0 0>, min=<0 0>}
    }
    padding=<ltrb 0 0 0 0>
   }
   slot {class=VerticalBoxSlot}
  }
  TextBlock_Name {
   widgetClassPath=&TextBlock
   bIsVariable=true
   parent=Border_39
   widget=TextBlock_Name
   properties {
    text=Text Block
    font {
     bForceMonospaced=false
     bMaterialIsStencil=false
     fontMaterial=null
     fontObject=&Roboto
     letterSpacing=0
     monospacedWidth=1
     size=7.5
     skewAmount=0
     typefaceFontName=Bold
     outlineSettings {
      bApplyOutlineToDropShadows=false
      bMiteredCorners=false
      bSeparateFillAlpha=false
      outlineMaterial=null
      outlineSize=0
      outlineColor=<rgba 0 0 0 1>
     }
    }
   }
   slot {class=BorderSlot, properties {horizontalAlignment=HAlign_Center, verticalAlignment=VAlign_Center, padding=<ltrb 0 0 0 0>}}
  }
  HorizontalBox_0 {
   widgetClassPath=&HorizontalBox
   parent=CanvasPanel_3
   widget=HorizontalBox_0
   slot {
    class=CanvasPanelSlot
    properties {bAutoSize=true, zOrder=1, layoutData {alignment=<0 0>, anchors {maximum=<1 0>, minimum=<0 0>}, offsets=<ltrb 0 0 0 20>}}
   }
  }
  Border_EditButton {
   widgetClassPath=&Border
   bIsVariable=true
   parent=HorizontalBox_0
   widget=Border_EditButton
   properties {
    background {
     bIsDynamicallyLoaded=false
     drawAs=NoDrawType
     imageType=NoImage
     mirroring=NoMirror
     resourceName=None
     resourceObject=null
     tiling=NoTile
     imageSize=<32 32>
     margin=<ltrb 0 0 0 0>
     outlineSettings {
      bUseBrushTransparency=false
      roundingType=HalfHeightRadius
      width=0
      color {colorUseRule=UseColor_Specified, specifiedColor=<rgba 0 0 0 0>}
      cornerRadii=<0 0 0 0>
     }
     tintColor {colorUseRule=UseColor_Specified, specifiedColor=<rgba 1 1 1 1>}
     uVRegion {bIsValid=false, max=<0 0>, min=<0 0>}
    }
    padding=<ltrb 0 0 0 0>
   }
   slot {class=HorizontalBoxSlot, properties {horizontalAlignment=HAlign_Left, verticalAlignment=VAlign_Center}}
  }
  Button_Edit {
   widgetClassPath=&EditorUtilityButton
   parent=Border_EditButton
   widget=Button_Edit
   properties {widgetStyle {
     clickedSlateSound {resourceObject=null}
     disabled {
      bIsDynamicallyLoaded=false
      drawAs=RoundedBox
      imageType=NoImage
      mirroring=NoMirror
      resourceName=None
      resourceObject=null
      tiling=NoTile
      imageSize=<0 0>
      margin=<ltrb 0 0 0 0>
      outlineSettings {
       bUseBrushTransparency=false
       roundingType=FixedRadius
       width=1
       color {colorUseRule=UseColor_ColorTable, specifiedColor=<rgba 0 0 0 0>}
       cornerRadii=<4 4 4 4>
      }
      tintColor {colorUseRule=UseColor_Specified, specifiedColor=<rgba 0.0395 0.0395 0.0395 1>}
      uVRegion {bIsValid=false, max=<0 0>, min=<0 0>}
     }
     disabledForeground {colorUseRule=UseColor_ColorTable, specifiedColor=<rgba 0 0 0 0>}
     hovered {
      bIsDynamicallyLoaded=false
      drawAs=RoundedBox
      imageType=NoImage
      mirroring=NoMirror
      resourceName=None
      resourceObject=null
      tiling=NoTile
      imageSize=<32 32>
      margin=<ltrb 0 0 0 0>
      outlineSettings {
       bUseBrushTransparency=false
       roundingType=FixedRadius
       width=1
       color {colorUseRule=UseColor_ColorTable, specifiedColor=<rgba 0 0 0 0>}
       cornerRadii=<4 4 4 4>
      }
      tintColor {colorUseRule=UseColor_Specified, specifiedColor=<rgba 0.0953 0.0953 0.0953 1>}
      uVRegion {bIsValid=false, max=<0 0>, min=<0 0>}
     }
     hoveredForeground {colorUseRule=UseColor_ColorTable, specifiedColor=<rgba 0 0 0 0>}
     hoveredSlateSound {resourceObject=null}
     normal {
      bIsDynamicallyLoaded=false
      drawAs=RoundedBox
      imageType=NoImage
      mirroring=NoMirror
      resourceName=None
      resourceObject=null
      tiling=NoTile
      imageSize=<32 32>
      margin=<ltrb 0 0 0 0>
      outlineSettings {
       bUseBrushTransparency=false
       roundingType=FixedRadius
       width=1
       color {colorUseRule=UseColor_ColorTable, specifiedColor=<rgba 0 0 0 0>}
       cornerRadii=<4 4 4 4>
      }
      tintColor {colorUseRule=UseColor_Specified, specifiedColor=<rgba 0.0395 0.0395 0.0395 1>}
      uVRegion {bIsValid=false, max=<0 0>, min=<0 0>}
     }
     normalForeground {colorUseRule=UseColor_ColorTable, specifiedColor=<rgba 0 0 0 0>}
     normalPadding=<ltrb 0 0 0 0>
     pressed {
      bIsDynamicallyLoaded=false
      drawAs=RoundedBox
      imageType=NoImage
      mirroring=NoMirror
      resourceName=None
      resourceObject=null
      tiling=NoTile
      imageSize=<32 32>
      margin=<ltrb 0 0 0 0>
      outlineSettings {
       bUseBrushTransparency=false
       roundingType=FixedRadius
       width=1
       color {colorUseRule=UseColor_ColorTable, specifiedColor=<rgba 0 0 0 0>}
       cornerRadii=<4 4 4 4>
      }
      tintColor {colorUseRule=UseColor_Specified, specifiedColor=<rgba 0.0284 0.0284 0.0284 1>}
      uVRegion {bIsValid=false, max=<0 0>, min=<0 0>}
     }
     pressedForeground {colorUseRule=UseColor_ColorTable, specifiedColor=<rgba 0 0 0 0>}
     pressedPadding=<ltrb 0 0 0 0>
     pressedSlateSound {resourceObject=null}
    }}
   slot {class=BorderSlot, properties {padding=<ltrb 0 0 0 0>}}
  }
  Image_81 {
   widgetClassPath=&Image
   bIsVariable=true
   parent=Button_Edit
   widget=Image_81
   properties {brush {
     bIsDynamicallyLoaded=false
     drawAs=Image
     imageType=FullColor
     mirroring=NoMirror
     resourceName=None
     resourceObject=/Game/Athena/UI/InGame/Creative/PropertyEditor/Icon_Picker/Textures_V2/128x128/T_UI_IconLibrary_Pencil_128
     tiling=NoTile
     imageSize=<20 20>
     margin=<ltrb 0 0 0 0>
     outlineSettings {
      bUseBrushTransparency=false
      roundingType=HalfHeightRadius
      width=0
      color {colorUseRule=UseColor_Specified, specifiedColor=<rgba 0 0 0 0>}
      cornerRadii=<0 0 0 0>
     }
     tintColor {colorUseRule=UseColor_Specified, specifiedColor=<rgba 1 1 1 1>}
     uVRegion {bIsValid=false, max=<0 0>, min=<0 0>}
    }}
   slot {class=ButtonSlot, properties {padding=<ltrb 2 2 2 2>}}
  }
  Button_Confirm {
   widgetClassPath=&EditorUtilityButton
   parent=HorizontalBox_0
   widget=Button_Confirm
   properties {widgetStyle {
     clickedSlateSound {resourceObject=null}
     disabled {
      bIsDynamicallyLoaded=false
      drawAs=RoundedBox
      imageType=NoImage
      mirroring=NoMirror
      resourceName=None
      resourceObject=null
      tiling=NoTile
      imageSize=<0 0>
      margin=<ltrb 0 0 0 0>
      outlineSettings {
       bUseBrushTransparency=false
       roundingType=FixedRadius
       width=1
       color {colorUseRule=UseColor_ColorTable, specifiedColor=<rgba 0 0 0 0>}
       cornerRadii=<4 4 4 4>
      }
      tintColor {colorUseRule=UseColor_Specified, specifiedColor=<rgba 0.0395 0.0395 0.0395 1>}
      uVRegion {bIsValid=false, max=<0 0>, min=<0 0>}
     }
     disabledForeground {colorUseRule=UseColor_ColorTable, specifiedColor=<rgba 0 0 0 0>}
     hovered {
      bIsDynamicallyLoaded=false
      drawAs=RoundedBox
      imageType=NoImage
      mirroring=NoMirror
      resourceName=None
      resourceObject=null
      tiling=NoTile
      imageSize=<32 32>
      margin=<ltrb 0 0 0 0>
      outlineSettings {
       bUseBrushTransparency=false
       roundingType=FixedRadius
       width=1
       color {colorUseRule=UseColor_ColorTable, specifiedColor=<rgba 0 0 0 0>}
       cornerRadii=<4 4 4 4>
      }
      tintColor {colorUseRule=UseColor_Specified, specifiedColor=<rgba 0.0953 0.0953 0.0953 1>}
      uVRegion {bIsValid=false, max=<0 0>, min=<0 0>}
     }
     hoveredForeground {colorUseRule=UseColor_ColorTable, specifiedColor=<rgba 0 0 0 0>}
     hoveredSlateSound {resourceObject=null}
     normal {
      bIsDynamicallyLoaded=false
      drawAs=RoundedBox
      imageType=NoImage
      mirroring=NoMirror
      resourceName=None
      resourceObject=null
      tiling=NoTile
      imageSize=<32 32>
      margin=<ltrb 0 0 0 0>
      outlineSettings {
       bUseBrushTransparency=false
       roundingType=FixedRadius
       width=1
       color {colorUseRule=UseColor_ColorTable, specifiedColor=<rgba 0 0 0 0>}
       cornerRadii=<4 4 4 4>
      }
      tintColor {colorUseRule=UseColor_Specified, specifiedColor=<rgba 0.0395 0.0395 0.0395 1>}
      uVRegion {bIsValid=false, max=<0 0>, min=<0 0>}
     }
     normalForeground {colorUseRule=UseColor_ColorTable, specifiedColor=<rgba 0 0 0 0>}
     normalPadding=<ltrb 0 0 0 0>
     pressed {
      bIsDynamicallyLoaded=false
      drawAs=RoundedBox
      imageType=NoImage
      mirroring=NoMirror
      resourceName=None
      resourceObject=null
      tiling=NoTile
      imageSize=<32 32>
      margin=<ltrb 0 0 0 0>
      outlineSettings {
       bUseBrushTransparency=false
       roundingType=FixedRadius
       width=1
       color {colorUseRule=UseColor_ColorTable, specifiedColor=<rgba 0 0 0 0>}
       cornerRadii=<4 4 4 4>
      }
      tintColor {colorUseRule=UseColor_Specified, specifiedColor=<rgba 0.0284 0.0284 0.0284 1>}
      uVRegion {bIsValid=false, max=<0 0>, min=<0 0>}
     }
     pressedForeground {colorUseRule=UseColor_ColorTable, specifiedColor=<rgba 0 0 0 0>}
     pressedPadding=<ltrb 0 0 0 0>
     pressedSlateSound {resourceObject=null}
    }}
   slot {class=HorizontalBoxSlot, properties {horizontalAlignment=HAlign_Left, verticalAlignment=VAlign_Center, size {sizeRule=Fill, value=1}}}
  }
  Image {
   widgetClassPath=&Image
   bIsVariable=true
   parent=Button_Confirm
   widget=Image
   properties {brush {
     bIsDynamicallyLoaded=false
     drawAs=Image
     imageType=FullColor
     mirroring=NoMirror
     resourceName=None
     resourceObject=null
     tiling=NoTile
     imageSize=<20 20>
     margin=<ltrb 0 0 0 0>
     outlineSettings {
      bUseBrushTransparency=false
      roundingType=HalfHeightRadius
      width=0
      color {colorUseRule=UseColor_Specified, specifiedColor=<rgba 0 0 0 0>}
      cornerRadii=<0 0 0 0>
     }
     tintColor {colorUseRule=UseColor_Specified, specifiedColor=<rgba 1 1 1 1>}
     uVRegion {bIsValid=false, max=<0 0>, min=<0 0>}
    }}
   slot {class=ButtonSlot, properties {padding=<ltrb 2 2 2 2>}}
  }
  ButtonDelete {
   widgetClassPath=&EditorUtilityButton
   parent=HorizontalBox_0
   widget=ButtonDelete
   properties {widgetStyle {
     clickedSlateSound {resourceObject=null}
     disabled {
      bIsDynamicallyLoaded=false
      drawAs=RoundedBox
      imageType=NoImage
      mirroring=NoMirror
      resourceName=None
      resourceObject=null
      tiling=NoTile
      imageSize=<0 0>
      margin=<ltrb 0 0 0 0>
      outlineSettings {
       bUseBrushTransparency=false
       roundingType=FixedRadius
       width=1
       color {colorUseRule=UseColor_ColorTable, specifiedColor=<rgba 0 0 0 0>}
       cornerRadii=<4 4 4 4>
      }
      tintColor {colorUseRule=UseColor_Specified, specifiedColor=<rgba 0.0395 0.0395 0.0395 1>}
      uVRegion {bIsValid=false, max=<0 0>, min=<0 0>}
     }
     disabledForeground {colorUseRule=UseColor_ColorTable, specifiedColor=<rgba 0 0 0 0>}
     hovered {
      bIsDynamicallyLoaded=false
      drawAs=RoundedBox
      imageType=NoImage
      mirroring=NoMirror
      resourceName=None
      resourceObject=null
      tiling=NoTile
      imageSize=<32 32>
      margin=<ltrb 0 0 0 0>
      outlineSettings {
       bUseBrushTransparency=false
       roundingType=FixedRadius
       width=1
       color {colorUseRule=UseColor_ColorTable, specifiedColor=<rgba 0 0 0 0>}
       cornerRadii=<4 4 4 4>
      }
      tintColor {colorUseRule=UseColor_Specified, specifiedColor=<rgba 0.0953 0.0953 0.0953 1>}
      uVRegion {bIsValid=false, max=<0 0>, min=<0 0>}
     }
     hoveredForeground {colorUseRule=UseColor_ColorTable, specifiedColor=<rgba 0 0 0 0>}
     hoveredSlateSound {resourceObject=null}
     normal {
      bIsDynamicallyLoaded=false
      drawAs=RoundedBox
      imageType=NoImage
      mirroring=NoMirror
      resourceName=None
      resourceObject=null
      tiling=NoTile
      imageSize=<32 32>
      margin=<ltrb 0 0 0 0>
      outlineSettings {
       bUseBrushTransparency=false
       roundingType=FixedRadius
       width=1
       color {colorUseRule=UseColor_ColorTable, specifiedColor=<rgba 0 0 0 0>}
       cornerRadii=<4 4 4 4>
      }
      tintColor {colorUseRule=UseColor_Specified, specifiedColor=<rgba 0.0395 0.0395 0.0395 1>}
      uVRegion {bIsValid=false, max=<0 0>, min=<0 0>}
     }
     normalForeground {colorUseRule=UseColor_ColorTable, specifiedColor=<rgba 0 0 0 0>}
     normalPadding=<ltrb 0 0 0 0>
     pressed {
      bIsDynamicallyLoaded=false
      drawAs=RoundedBox
      imageType=NoImage
      mirroring=NoMirror
      resourceName=None
      resourceObject=null
      tiling=NoTile
      imageSize=<32 32>
      margin=<ltrb 0 0 0 0>
      outlineSettings {
       bUseBrushTransparency=false
       roundingType=FixedRadius
       width=1
       color {colorUseRule=UseColor_ColorTable, specifiedColor=<rgba 0 0 0 0>}
       cornerRadii=<4 4 4 4>
      }
      tintColor {colorUseRule=UseColor_Specified, specifiedColor=<rgba 0.0284 0.0284 0.0284 1>}
      uVRegion {bIsValid=false, max=<0 0>, min=<0 0>}
     }
     pressedForeground {colorUseRule=UseColor_ColorTable, specifiedColor=<rgba 0 0 0 0>}
     pressedPadding=<ltrb 0 0 0 0>
     pressedSlateSound {resourceObject=null}
    }}
   slot {class=HorizontalBoxSlot, properties {horizontalAlignment=HAlign_Right, verticalAlignment=VAlign_Center, size {sizeRule=Fill, value=1}}}
  }
  Image_2 {
   widgetClassPath=&Image
   bIsVariable=true
   parent=ButtonDelete
   widget=Image_2
   properties {brush {
     bIsDynamicallyLoaded=false
     drawAs=Image
     imageType=FullColor
     mirroring=NoMirror
     resourceName=None
     resourceObject=/Editor_TodoList/Textures/T_Todo_Delete
     tiling=NoTile
     imageSize=<20 20>
     margin=<ltrb 0 0 0 0>
     outlineSettings {
      bUseBrushTransparency=false
      roundingType=HalfHeightRadius
      width=0
      color {colorUseRule=UseColor_Specified, specifiedColor=<rgba 0 0 0 0>}
      cornerRadii=<0 0 0 0>
     }
     tintColor {colorUseRule=UseColor_Specified, specifiedColor=<rgba 1 1 1 1>}
     uVRegion {bIsValid=false, max=<0 0>, min=<0 0>}
    }}
   slot {class=ButtonSlot, properties {padding=<ltrb 2 2 2 2>}}
  }
  ScaleBox_1 {
   widgetClassPath=&ScaleBox
   parent=CanvasPanel_3
   widget=ScaleBox_1
   properties {stretch=UserSpecified}
   slot {
    class=CanvasPanelSlot
    properties {bAutoSize=true, layoutData {alignment=<0.5 0>, anchors {maximum=<0.5 0>, minimum=<0.5 0>}, offsets=<ltrb 0 25.1676 0 0>}}
   }
  }
  VerticalBox_1 {widgetClassPath=&VerticalBox, bIsVariable=true, parent=ScaleBox_1, widget=VerticalBox_1, slot {class=ScaleBoxSlot}}
  SinglePropertyView_65 {
   widgetClassPath=&SinglePropertyView
   parent=VerticalBox_1
   widget=SinglePropertyView_65
   properties {bAutoLoadAsset=false}
   slot {class=VerticalBoxSlot}
  }
  EUW_MultiCamViewer_Property {
   widgetClassPath=&/EUW_MultiCamViewer_Property
   bIsVariable=true
   parent=VerticalBox_1
   widget=EUW_MultiCamViewer_Property
   slot {class=VerticalBoxSlot}
  }
  Border_138 {
   widgetClassPath=&Border
   parent=CanvasPanel_3
   widget=Border_138
   properties {background {
     bIsDynamicallyLoaded=false
     drawAs=RoundedBox
     imageType=NoImage
     mirroring=NoMirror
     resourceName=None
     resourceObject=null
     tiling=NoTile
     imageSize=<32 32>
     margin=<ltrb 0 0 0 0>
     outlineSettings {
      bUseBrushTransparency=false
      roundingType=FixedRadius
      width=0
      color {colorUseRule=UseColor_Specified, specifiedColor=<rgba 0 0 0 0>}
      cornerRadii=<7 7 7 7>
     }
     tintColor {colorUseRule=UseColor_Specified, specifiedColor=<rgba 0.0104 0.0104 0.0104 0.8261>}
     uVRegion {bIsValid=false, max=<0 0>, min=<0 0>}
    }}
   slot {class=CanvasPanelSlot, properties {zOrder=-1, layoutData {alignment=<0 0>, anchors {maximum=<1 1>, minimum=<0 0>}, offsets=<ltrb 0 0 0 0>}}}
  }
 ]
}
